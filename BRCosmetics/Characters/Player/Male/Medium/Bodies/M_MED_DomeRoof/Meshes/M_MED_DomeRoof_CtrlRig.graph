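# Control Rig for the DomeRoof male medium player body that corrects clothing and accessory bones against body motion.
# It runs the shared DeformRig, then weighted parent-constraint pins, jacket/sword, lapel/collar/necklace and leg cloth passes.
# Spherical pose readers on thighs and head, arm-to-lapel distances, upperarm attributes and MovementAlpha weight additive bone offsets.
# Final fixed offsets precede smoothed skydiving, sliding and knee-slide adjustments, including moving the jacket floor-collision bone.
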ControlRigBlueprint /BRCosmetics/Characters/Player/Male/Medium/Bodies/M_MED_DomeRoof/Meshes/M_MED_DomeRoof_CtrlRig
graphs [
 New Function_ContainedGraph {nodes [
   n1 = RigVMFunctionEntryNode()
   RigVMFunction_Sequence = &RigVMFunction_Sequence(ExecuteContext=@n1.ExecuteContext)
   &RigUnit_ParentConstraint(
    AdvancedSettings=(RotationOrderForFilter=XZY)
    Child=(Type=Bone,Name="dyn_boot_l_laces_MSTR")
    Parents=((Item=(Type=Bone,Name="calf_l"),Weight=0.0124),(Item=(Type=Bone,Name="calf_twist_02_l"),Weight=0.355),(Item=(Type=Bone,Name="calf_twist_01_l"),Weight=0.6326))
    Weight=1
    ExecutePin=@RigVMFunction_Sequence.A
   )
   > &RigUnit_ParentConstraint(
    AdvancedSettings=(RotationOrderForFilter=XZY)
    Child=(Type=Bone,Name="dyn_boot_r_laces_MSTR")
    Parents=((Item=(Type=Bone,Name="calf_r"),Weight=0.0092),(Item=(Type=Bone,Name="foot_r"),Weight=0.0204),(Item=(Type=Bone,Name="calf_twist_02_r"),Weight=0.3092),(Item=(Type=Bone,Name="calf_twist_01_r"),Weight=0.6612))
    Weight=1
   )
   > &RigUnit_ParentConstraint(
    AdvancedSettings=(RotationOrderForFilter=XZY)
    Child=(Type=Bone,Name="dyn_boot_l_lip_01")
    Parents=((Item=(Type=Bone,Name="calf_twist_01_l"),Weight=0.0459),(Item=(Type=Bone,Name="calf_l"),Weight=0.2957),(Item=(Type=Bone,Name="calf_twist_02_l"),Weight=0.6584))
    Weight=1
   )
   > &RigUnit_ParentConstraint(
    AdvancedSettings=(RotationOrderForFilter=XZY)
    Child=(Type=Bone,Name="dyn_boot_l_lip_02")
    Parents=((Item=(Type=Bone,Name="calf_twist_01_l"),Weight=0.0594),(Item=(Type=Bone,Name="calf_l"),Weight=0.2128),(Item=(Type=Bone,Name="calf_twist_02_l"),Weight=0.7278))
    Weight=1
   )
   > &RigUnit_ParentConstraint(
    AdvancedSettings=(RotationOrderForFilter=XZY)
    Child=(Type=Bone,Name="dyn_boot_l_lip_03")
    Parents=((Item=(Type=Bone,Name="calf_twist_01_l"),Weight=0.0185),(Item=(Type=Bone,Name="calf_l"),Weight=0.3302),(Item=(Type=Bone,Name="calf_twist_02_l"),Weight=0.6513))
    Weight=1
   )
   > &RigUnit_ParentConstraint(
    AdvancedSettings=(RotationOrderForFilter=XZY)
    Child=(Type=Bone,Name="dyn_boot_l_lip_04")
    Parents=((Item=(Type=Bone,Name="thigh_l"),Weight=0.0038),(Item=(Type=Bone,Name="calf_twist_01_l"),Weight=0.0039),(Item=(Type=Bone,Name="calf_l"),Weight=0.4367),(Item=(Type=Bone,Name="calf_twist_02_l"),Weight=0.5555))
    Weight=1
   )
   > &RigUnit_ParentConstraint(
    AdvancedSettings=(RotationOrderForFilter=XZY)
    Child=(Type=Bone,Name="dyn_boot_r_lip_01")
    Parents=((Item=(Type=Bone,Name="calf_twist_01_r"),Weight=0.0447),(Item=(Type=Bone,Name="calf_r"),Weight=0.2988),(Item=(Type=Bone,Name="calf_twist_02_r"),Weight=0.6565))
    Weight=1
   )
   > &RigUnit_ParentConstraint(
    AdvancedSettings=(RotationOrderForFilter=XZY)
    Child=(Type=Bone,Name="dyn_boot_r_lip_02")
    Parents=((Item=(Type=Bone,Name="calf_twist_01_r"),Weight=0.0576),(Item=(Type=Bone,Name="calf_r"),Weight=0.2203),(Item=(Type=Bone,Name="calf_twist_02_r"),Weight=0.7221))
    Weight=1
   )
   > &RigUnit_ParentConstraint(
    AdvancedSettings=(RotationOrderForFilter=XZY)
    Child=(Type=Bone,Name="dyn_boot_r_lip_03")
    Parents=((Item=(Type=Bone,Name="calf_twist_01_r"),Weight=0.019),(Item=(Type=Bone,Name="calf_r"),Weight=0.3276),(Item=(Type=Bone,Name="calf_twist_02_r"),Weight=0.6534))
    Weight=1
   )
   > &RigUnit_ParentConstraint(
    AdvancedSettings=(RotationOrderForFilter=XZY)
    Child=(Type=Bone,Name="dyn_boot_r_lip_04")
    Parents=((Item=(Type=Bone,Name="calf_twist_01_r"),Weight=0.0292),(Item=(Type=Bone,Name="calf_r"),Weight=0.2769),(Item=(Type=Bone,Name="calf_twist_02_r"),Weight=0.6939))
    Weight=1
   )
   > &RigUnit_ParentConstraint(
    AdvancedSettings=(RotationOrderForFilter=XZY)
    Child=(Type=Bone,Name="dyn_lapel_l_01")
    Parents="((Item=(Type=Bone,Name=\"deform_upperarm_fr_l\"),Weight=0.0108),(Item=(Type=Bone,Name=\"spine_01\"),Weight=0.021),(Item=(Type=Bone,Name=\"spine_05\"),Weight=0.0664),(Item=(Type=Bone,Name=\"spine_02\"),Weight=0.106),(Item=(Type=Bone,Name=\"clavicle_l\"),Weight=0.1067 ...(truncated)"
    Weight=1
   )
   > &RigUnit_ParentConstraint(
    AdvancedSettings=(RotationOrderForFilter=XZY)
    Child=(Type=Bone,Name="dyn_lapel_r_01")
    Parents="((Item=(Type=Bone,Name=\"deform_upperarm_fr_r\"),Weight=0.014),(Item=(Type=Bone,Name=\"spine_01\"),Weight=0.0188),(Item=(Type=Bone,Name=\"spine_05\"),Weight=0.0657),(Item=(Type=Bone,Name=\"spine_02\"),Weight=0.1036),(Item=(Type=Bone,Name=\"clavicle_r\"),Weight=0.112 ...(truncated)"
    Weight=1
   )
   > &RigUnit_ParentConstraint(
    AdvancedSettings=(RotationOrderForFilter=XZY)
    Child=(Type=Bone,Name="deform_gun")
    Parents="((Item=(Type=Bone,Name=\"thigh_twist_01_l\"),Weight=0.0007),(Item=(Type=Bone,Name=\"spine_03\"),Weight=0.0173),(Item=(Type=Bone,Name=\"spine_01\"),Weight=0.0269),(Item=(Type=Bone,Name=\"spine_02\"),Weight=0.033),(Item=(Type=Bone,Name=\"thigh_twist_01_r\"),Weight=0.0 ...(truncated)"
    Weight=1
   )
   > &RigUnit_ParentConstraint(
    AdvancedSettings=(RotationOrderForFilter=XZY)
    Child=(Type=Bone,Name="dyn_sword_bandolier_rear_mover_MSTR")
    Parents="((Item=(Type=Bone,Name=\"dyn_jacket_rear_01\"),Weight=0.0191),(Item=(Type=Bone,Name=\"spine_04\"),Weight=0.0451),(Item=(Type=Bone,Name=\"spine_01\"),Weight=0.0978),(Item=(Type=Bone,Name=\"dyn_jacket_side_l_02\"),Weight=0.1125),(Item=(Type=Bone,Name=\"dyn_jacket_sid ...(truncated)"
    Weight=1
   )
   > &RigUnit_ParentConstraint(
    AdvancedSettings=(RotationOrderForFilter=XZY)
    Child=(Type=Bone,Name="dyn_sword_bandolier_front_mover_MSTR")
    Parents="((Item=(Type=Bone,Name=\"dyn_jacket_front_l_03\"),Weight=0.0004),(Item=(Type=Bone,Name=\"spine_04\"),Weight=0.0039),(Item=(Type=Bone,Name=\"spine_03\"),Weight=0.0421),(Item=(Type=Bone,Name=\"spine_01\"),Weight=0.0495),(Item=(Type=Bone,Name=\"spine_02\"),Weight=0.064 ...(truncated)"
    Weight=1
   )
   > &RigUnit_ParentConstraint(
    AdvancedSettings=(RotationOrderForFilter=XZY)
    Child=(Type=Bone,Name="dyn_cloth_front_l_01_MSTR")
    Parents=((Item=(Type=Bone,Name="pelvis"),Weight=0.0786),(Item=(Type=Bone,Name="thigh_l"),Weight=0.1),(Item=(Type=Bone,Name="thigh_twist_01_l"),Weight=0.8213))
    Weight=1
   )
   RigVMFunctionReturnNode(ExecuteContext=@RigVMFunction_Sequence.B)
  ]}
 New Function_ContainedGraph {nodes [
   n1 = RigVMFunctionEntryNode()
   RigVMFunction_Sequence = &RigVMFunction_Sequence(ExecuteContext=@n1.ExecuteContext)
   &RigUnit_ParentConstraint(
    AdvancedSettings=(RotationOrderForFilter=XZY)
    Child=(Type=Bone,Name="dyn_jacket_front_l_01_MSTR")
    Parents=((Item=(Type=Bone,Name="spine_03"),Weight=0.1488),(Item=(Type=Bone,Name="spine_01"),Weight=0.2049),(Item=(Type=Bone,Name="spine_02"),Weight=0.2542),(Item=(Type=Bone,Name="pelvis"),Weight=0.3921))
    Weight=1
    ExecutePin=@RigVMFunction_Sequence.A
   )
   > &RigUnit_ParentConstraint(
    AdvancedSettings=(RotationOrderForFilter=XZY)
    Child=(Type=Bone,Name="dyn_jacket_side_l_01_MSTR")
    Parents="((Item=(Type=Bone,Name=\"upperarm_twist_01_l\"),Weight=0.0016),(Item=(Type=Bone,Name=\"deform_upperarm_bk_l\"),Weight=0.0032),(Item=(Type=Bone,Name=\"clavicle_l\"),Weight=0.0166),(Item=(Type=Bone,Name=\"spine_04\"),Weight=0.0423),(Item=(Type=Bone,Name=\"spine_01\"), ...(truncated)"
    Weight=1
   )
   > &RigUnit_ParentConstraint(
    AdvancedSettings=(RotationOrderForFilter=XZY)
    Child=(Type=Bone,Name="dyn_jacket_rear_01_MSTR")
    Parents=((Item=(Type=Bone,Name="spine_04"),Weight=0.0322),(Item=(Type=Bone,Name="spine_01"),Weight=0.1988),(Item=(Type=Bone,Name="spine_03"),Weight=0.1993),(Item=(Type=Bone,Name="spine_02"),Weight=0.2508),(Item=(Type=Bone,Name="pelvis"),Weight=0.3189))
    Weight=1
   )
   > &RigUnit_ParentConstraint(
    AdvancedSettings=(RotationOrderForFilter=XZY)
    Child=(Type=Bone,Name="dyn_jacket_front_r_01_MSTR")
    Parents=((Item=(Type=Bone,Name="spine_03"),Weight=0.1442),(Item=(Type=Bone,Name="spine_01"),Weight=0.2001),(Item=(Type=Bone,Name="spine_02"),Weight=0.2493),(Item=(Type=Bone,Name="pelvis"),Weight=0.4064))
    Weight=1
   )
   > &RigUnit_ParentConstraint(
    AdvancedSettings=(RotationOrderForFilter=XZY)
    Child=(Type=Bone,Name="dyn_jacket_side_r_01_MSTR")
    Parents="((Item=(Type=Bone,Name=\"clavicle_r\"),Weight=0.0006),(Item=(Type=Bone,Name=\"spine_04\"),Weight=0.0306),(Item=(Type=Bone,Name=\"spine_01\"),Weight=0.2071),(Item=(Type=Bone,Name=\"spine_03\"),Weight=0.2079),(Item=(Type=Bone,Name=\"pelvis\"),Weight=0.2584),(Item=(Typ ...(truncated)"
    Weight=1
   )
   > RigVMFunction_Sequence_1 = RigVMAggregateNode()
   n2 = &RigUnit_SphericalPoseReader(
    ActiveRegionScaleFactors=(PositiveWidth=2,NegativeWidth=2,PositiveHeight=2,NegativeHeight=1)
    ActiveRegionSize=0.1
    Debug=(bDrawDebug=false,bDraw2D=False,bDrawLocalAxes=False,DebugScale=25,DebugSegments=20,DebugThickness=0.25)
    DriverAxis=(X=-1,Y=0,Z=0)
    DriverItem=(Type=Bone,Name="thigh_l")
    FalloffRegionScaleFactors=(PositiveWidth=2,NegativeWidth=2,PositiveHeight=4,NegativeHeight=1)
    FalloffSize=0.3
    FlipHeightScaling=false
    FlipWidthScaling=false
    OptionalParentItem=(Type=Bone,Name="pelvis")
    RotationOffset=(X=0,Y=90,Z=110)
    ExecutePin=@RigVMFunction_Sequence_1.A
   )
   > n3 = &RigUnit_SphericalPoseReader(
    ActiveRegionScaleFactors=(PositiveWidth=4,NegativeWidth=3,PositiveHeight=2.5,NegativeHeight=2)
    ActiveRegionSize=0.1
    Debug=(bDrawDebug=false,bDraw2D=False,bDrawLocalAxes=False,DebugScale=25,DebugSegments=20,DebugThickness=0.25)
    DriverAxis=(X=-1,Y=0,Z=0)
    DriverItem=(Type=Bone,Name="thigh_l")
    FalloffRegionScaleFactors=(PositiveWidth=1,NegativeWidth=4,PositiveHeight=1,NegativeHeight=0.5)
    FalloffSize=0.4
    FlipHeightScaling=False
    FlipWidthScaling=False
    OptionalParentItem=(Type=Bone,Name="pelvis")
    RotationOffset=(X=100,Y=0,Z=-35)
   )
   > n4 = &RigUnit_SphericalPoseReader(
    ActiveRegionScaleFactors=(PositiveWidth=3,NegativeWidth=0.8,PositiveHeight=2,NegativeHeight=2)
    ActiveRegionSize=0.1
    Debug=(bDrawDebug=true,bDraw2D=False,bDrawLocalAxes=False,DebugScale=25,DebugSegments=20,DebugThickness=0.25)
    DriverAxis=(X=-1,Y=0,Z=0)
    DriverItem=(Type=Bone,Name="thigh_l")
    FalloffRegionScaleFactors=(PositiveWidth=2,NegativeWidth=1.35,PositiveHeight=2,NegativeHeight=2)
    FalloffSize=0.26
    FlipHeightScaling=false
    FlipWidthScaling=false
    OptionalParentItem=(Type=Bone,Name="pelvis")
    RotationOffset=(X=180,Y=-32,Z=0)
   )
   > n5 = &RigUnit_SphericalPoseReader(
    ActiveRegionScaleFactors=(PositiveWidth=2,NegativeWidth=1.5,PositiveHeight=1.75,NegativeHeight=1)
    ActiveRegionSize=0.1
    Debug=(bDrawDebug=true,bDraw2D=False,bDrawLocalAxes=False,DebugScale=25,DebugSegments=20,DebugThickness=0.25)
    DriverAxis=(X=-1,Y=0,Z=0)
    DriverItem=(Type=Bone,Name="thigh_l")
    FalloffRegionScaleFactors=(PositiveWidth=1,NegativeWidth=1.2,PositiveHeight=1,NegativeHeight=0.5)
    FalloffSize=0.3
    FlipHeightScaling=False
    FlipWidthScaling=False
    OptionalParentItem=(Type=Bone,Name="pelvis")
    RotationOffset=(X=130,Y=0,Z=-15)
   )
   > &RigUnit_ModifyTransforms(
    ItemToModify="((Item=(Type=Bone,Name=\"dyn_jacket_front_l_01_MSTR\"),Transform=(Rotation=(X=0.129,Y=-0.1555,Z=0.0069,W=0.9794),Translation=(X=-0.2528,Y=-0.2128,Z=0.7671),Scale3D=(X=1,Y=1,Z=1))),(Item=(Type=Bone,Name=\"dyn_sword_bandolier_01_MSTR\"),Transform=(Rotation=(X=0. ...(truncated)"
    WeightMinimum=0
    Weight=@n3.OutputParam
   )
   > &RigUnit_ModifyTransforms(
    ItemToModify="((Item=(Type=Bone,Name=\"dyn_jacket_front_l_01_MSTR\"),Transform=(Rotation=(X=0.499,Y=-0.0784,Z=0.0526,W=0.8614),Translation=(X=-0.1265,Y=-4.862,Z=2.6215),Scale3D=(X=1,Y=1,Z=1))),(Item=(Type=Bone,Name=\"dyn_sword_bandolier_01_MSTR\"),Transform=(Rotation=(X=0,Y ...(truncated)"
    WeightMinimum=0
    Weight=@n5.OutputParam
   )
   > &RigUnit_ModifyTransforms(
    ItemToModify="((Item=(Type=Bone,Name=\"dyn_jacket_front_l_01_MSTR\"),Transform=(Rotation=(X=0.0464,Y=-0.3311,Z=-0.1457,W=0.9311),Translation=(X=0,Y=0,Z=0),Scale3D=(X=1,Y=1,Z=1))),(Item=(Type=Bone,Name=\"dyn_sword_bandolier_01_MSTR\"),Transform=(Rotation=(X=0,Y=-0.1305,Z=0,W ...(truncated)"
    WeightMinimum=0
    Weight=@n4.OutputParam
   )
   > &RigUnit_ModifyTransforms(
    ItemToModify="((Item=(Type=Bone,Name=\"dyn_jacket_side_l_01_MSTR\"),Transform=(Rotation=(X=-0.2557,Y=-0.4731,Z=-0.2468,W=0.8061),Translation=(X=0,Y=0,Z=0),Scale3D=(X=1,Y=1,Z=1))),(Item=(Type=Bone,Name=\"dyn_jacket_rear_01_MSTR\"),Transform=(Rotation=(X=0.0811,Y=-0.6163,Z=0. ...(truncated)"
    WeightMinimum=0
    Weight=@n2.OutputParam
   )
   n6 = &RigUnit_SphericalPoseReader(
    ActiveRegionScaleFactors=(PositiveWidth=2,NegativeWidth=2,PositiveHeight=2,NegativeHeight=1)
    ActiveRegionSize=0.1
    Debug=(bDrawDebug=false,bDraw2D=False,bDrawLocalAxes=False,DebugScale=25,DebugSegments=20,DebugThickness=0.25)
    DriverAxis=(X=1,Y=0,Z=0)
    DriverItem=(Type=Bone,Name="thigh_r")
    FalloffRegionScaleFactors=(PositiveWidth=2,NegativeWidth=2,PositiveHeight=4,NegativeHeight=1)
    FalloffSize=0.3
    FlipHeightScaling=false
    FlipWidthScaling=false
    OptionalParentItem=(Type=Bone,Name="pelvis")
    RotationOffset=(X=180,Y=90,Z=110)
    ExecutePin=@RigVMFunction_Sequence_1.B
   )
   > n7 = &RigUnit_SphericalPoseReader(
    ActiveRegionScaleFactors=(PositiveWidth=4,NegativeWidth=3,PositiveHeight=2.5,NegativeHeight=2)
    ActiveRegionSize=0.1
    Debug=(bDrawDebug=false,bDraw2D=False,bDrawLocalAxes=False,DebugScale=25,DebugSegments=20,DebugThickness=0.25)
    DriverAxis=(X=1,Y=0,Z=0)
    DriverItem=(Type=Bone,Name="thigh_r")
    FalloffRegionScaleFactors=(PositiveWidth=1,NegativeWidth=4,PositiveHeight=1,NegativeHeight=0.5)
    FalloffSize=0.41
    FlipHeightScaling=False
    FlipWidthScaling=true
    OptionalParentItem=(Type=Bone,Name="pelvis")
    RotationOffset=(X=-80,Y=0,Z=-35)
   )
   > n8 = &RigUnit_SphericalPoseReader(
    ActiveRegionScaleFactors=(PositiveWidth=3,NegativeWidth=2,PositiveHeight=2,NegativeHeight=2)
    ActiveRegionSize=0.1
    Debug=(bDrawDebug=false,bDraw2D=False,bDrawLocalAxes=False,DebugScale=25,DebugSegments=20,DebugThickness=0.25)
    DriverAxis=(X=1,Y=0,Z=0)
    DriverItem=(Type=Bone,Name="thigh_r")
    FalloffRegionScaleFactors=(PositiveWidth=2,NegativeWidth=2,PositiveHeight=2,NegativeHeight=2)
    FalloffSize=0.295
    FlipHeightScaling=False
    FlipWidthScaling=true
    OptionalParentItem=(Type=Bone,Name="pelvis")
    RotationOffset=(X=0,Y=-20,Z=0)
   )
   > n9 = &RigUnit_SphericalPoseReader(
    ActiveRegionScaleFactors=(PositiveWidth=4,NegativeWidth=3,PositiveHeight=2.5,NegativeHeight=2)
    ActiveRegionSize=0.1
    Debug=(bDrawDebug=false,bDraw2D=False,bDrawLocalAxes=False,DebugScale=25,DebugSegments=20,DebugThickness=0.25)
    DriverAxis=(X=1,Y=0,Z=0)
    DriverItem=(Type=Bone,Name="thigh_r")
    FalloffRegionScaleFactors=(PositiveWidth=1,NegativeWidth=4,PositiveHeight=1,NegativeHeight=0.5)
    FalloffSize=0.36
    FlipHeightScaling=False
    FlipWidthScaling=true
    OptionalParentItem=(Type=Bone,Name="pelvis")
    RotationOffset=(X=-35,Y=0,Z=-35)
   )
   > &RigUnit_ModifyTransforms(
    ItemToModify="((Item=(Type=Bone,Name=\"dyn_jacket_front_r_01_MSTR\"),Transform=(Rotation=(X=0.1516,Y=-0.0906,Z=0.0197,W=0.9841),Translation=(X=0.3148,Y=0.4767,Z=-1.4707),Scale3D=(X=1,Y=1,Z=1))),(Item=(Type=Bone,Name=\"deform_gun\"),Transform=(Rotation=(X=-0.0101,Y=0.2039,Z= ...(truncated)"
    WeightMinimum=0
    Weight=@n7.OutputParam
   )
   > &RigUnit_ModifyTransforms(
    ItemToModify="((Item=(Type=Bone,Name=\"dyn_jacket_front_r_01_MSTR\"),Transform=(Rotation=(X=-0.0386,Y=-0.1088,Z=-0.008,W=0.9933),Translation=(X=0.4818,Y=-2.6335,Z=-0.1491),Scale3D=(X=1,Y=1,Z=1))),(Item=(Type=Bone,Name=\"deform_gun\"),Transform=(Rotation=(X=0.0298,Y=0.0616,Z ...(truncated)"
    WeightMinimum=0
    Weight=@n9.OutputParam
   )
   > &RigUnit_ModifyTransforms(
    ItemToModify="((Item=(Type=Bone,Name=\"dyn_jacket_front_r_01_MSTR\"),Transform=(Rotation=(X=0.0158,Y=-0.2793,Z=-0.0505,W=0.9588),Translation=(X=0,Y=0,Z=0),Scale3D=(X=1,Y=1,Z=1))),(Item=(Type=Bone,Name=\"dyn_jacket_side_r_01_MSTR\"),Transform=(Rotation=(X=0.1414,Y=-0.1052,Z= ...(truncated)"
    WeightMinimum=0
    Weight=@n8.OutputParam
   )
   > &RigUnit_ModifyTransforms(
    ItemToModify="((Item=(Type=Bone,Name=\"dyn_jacket_side_r_01_MSTR\"),Transform=(Rotation=(X=-0.2948,Y=-0.4183,Z=-0.1861,W=0.8387),Translation=(X=-0.7434,Y=-3.9269,Z=0.2147),Scale3D=(X=1,Y=1,Z=1))),(Item=(Type=Bone,Name=\"dyn_jacket_rear_01_MSTR\"),Transform=(Rotation=(X=-0.1 ...(truncated)"
    WeightMinimum=0
    Weight=@n6.OutputParam
   )
   Add = &RigVMFunction_MathDoubleAdd(A=@n2.OutputParam, B=@n6.OutputParam)
   Multiply = &RigVMFunction_MathDoubleMul(B=0.5, A=@Add.Result)
   Remap = &RigVMFunction_MathDoubleRemap(SourceMaximum=1, SourceMinimum=0.5, TargetMaximum=1, TargetMinimum=0, Value=@Multiply.Result)
   &RigUnit_ModifyTransforms(
    ItemToModify="((Item=(Type=Bone,Name=\"dyn_jacket_side_r_01_MSTR\"),Transform=(Rotation=(X=-0.0774,Y=0.3216,Z=0.2207,W=0.9175),Translation=(X=0,Y=0,Z=0),Scale3D=(X=1,Y=1,Z=1))),(Item=(Type=Bone,Name=\"dyn_jacket_rear_01_MSTR\"),Transform=(Rotation=(X=0.0653,Y=0.6001,Z=0.158 ...(truncated)"
    WeightMinimum=0
    Weight=@Remap.Result
    ExecutePin=@RigVMFunction_Sequence_1.C
   )
   &RigUnit_ModifyTransforms(
    ItemToModify="((Item=(Type=Bone,Name=\"dyn_jacket_side_r_01_MSTR\"),Transform=(Rotation=(X=0,Y=-0.1045,Z=0,W=0.9945),Translation=(X=0,Y=0,Z=0),Scale3D=(X=1,Y=1,Z=1))),(Item=(Type=Bone,Name=\"dyn_jacket_rear_01_MSTR\"),Transform=(Rotation=(X=0,Y=-0.1045,Z=0,W=0.9945),Transla ...(truncated)"
    WeightMinimum=0
    ExecutePin=@RigVMFunction_Sequence_1.D
    Weight=RigVMVariableNode() {variable=MovementAlpha}.Value
   )
   RigVMFunctionReturnNode(ExecuteContext=@RigVMFunction_Sequence.B)
   RigVMVariableNode() {variable=DebugDraw}
  ]}
 New Function_ContainedGraph {nodes [
   RigVMFunctionEntryNode()
   > RigVMFunction_Sequence = RigVMAggregateNode()
   Get Normalized Distance Between = RigVMFunctionReferenceNode(
    Debug=false
    Max Distance=30
    Min Distance=0
    Source=(Type=Bone,Name="upperarm_twist_02_r")
    Target=(Type=Bone,Name="dyn_lapel_r_01")
    ExecuteContext=@RigVMFunction_Sequence.A
   )
   > Get Normalized Distance Between_1 = RigVMFunctionReferenceNode(Debug=false, Max Distance=30, Min Distance=0, Source=(Type=Bone,Name="upperarm_twist_02_l"), Target=(Type=Bone,Name="dyn_lapel_l_01"))
   n1 = &RigVMFunction_AnimEvalRichCurve(
    Curve=(EditorCurveData=(Keys=((),(Time=0.25),(Time=0.5483,Value=0.1959),(Time=0.8,Value=1),(Time=1,Value=1))))
    SourceMinimum=0
    TargetMinimum=0
    Value=@"Get Normalized Distance Between_1".Result
   )
   > &RigUnit_ModifyTransforms(
    ItemToModify="((Item=(Type=Bone,Name=\"dyn_lapel_l_01\"),Transform=(Rotation=(X=0,Y=0,Z=0.5736,W=0.8192),Translation=(X=-2.8189,Y=-0.1477,Z=0),Scale3D=(X=1,Y=1,Z=1))),(Item=(Type=Bone,Name=\"dyn_lapel_l_02\"),Transform=(Rotation=(X=0,Y=0,Z=0.3827,W=0.9239),Translation=(X=-4 ...(truncated)"
    WeightMinimum=0
    Weight=@n1.Result
   )
   RigVMFunction_AnimEvalRichCurve = &RigVMFunction_AnimEvalRichCurve(
    Curve=(EditorCurveData=(Keys=((),(Time=0.25),(Time=0.5483,Value=0.1959),(Time=0.8,Value=1),(Time=1,Value=1))))
    SourceMinimum=0
    TargetMinimum=0
    Value=@"Get Normalized Distance Between".Result
   )
   > &RigUnit_ModifyTransforms(
    ItemToModify="((Item=(Type=Bone,Name=\"dyn_lapel_r_01\"),Transform=(Rotation=(X=0.141,Y=0.1533,Z=0.683,W=0.7001),Translation=(X=4.2052,Y=-0.0485,Z=-0.6347),Scale3D=(X=1,Y=1,Z=1))),(Item=(Type=Bone,Name=\"dyn_lapel_r_02\"),Transform=(Rotation=(X=0,Y=0,Z=0.464,W=0.8858),Trans ...(truncated)"
    WeightMinimum=0
    Weight=@RigVMFunction_AnimEvalRichCurve.Result
   )
   &RigUnit_ParentConstraint(
    AdvancedSettings=(RotationOrderForFilter=XZY)
    Child=(Type=Bone,Name="dyn_shirt_collar_r_01")
    Parents="((Item=(Type=Bone,Name=\"spine_03\"),Weight=0.0007),(Item=(Type=Bone,Name=\"deform_pec_r\"),Weight=0.0008),(Item=(Type=Bone,Name=\"spine_04\"),Weight=0.0022),(Item=(Type=Bone,Name=\"head\"),Weight=0.0329),(Item=(Type=Bone,Name=\"neck_01\"),Weight=0.0608),(Item=(Type ...(truncated)"
    Weight=1
    ExecutePin=@RigVMFunction_Sequence.B
   )
   > &RigUnit_ParentConstraint(
    AdvancedSettings=(RotationOrderForFilter=XZY)
    Child=(Type=Bone,Name="dyn_shirt_collar_l_01")
    Parents=((Item=(Type=Bone,Name="head"),Weight=0.0356),(Item=(Type=Bone,Name="clavicle_l"),Weight=0.0543),(Item=(Type=Bone,Name="neck_01"),Weight=0.063),(Item=(Type=Bone,Name="neck_02"),Weight=0.1694),(Item=(Type=Bone,Name="spine_05"),Weight=0.6777))
    Weight=1
   )
   > &RigUnit_ParentConstraint(
    AdvancedSettings=(RotationOrderForFilter=XZY)
    Child=(Type=Null,Name="right_collar_aimat_null")
    Parents="((Item=(Type=Bone,Name=\"spine_03\"),Weight=0.0188),(Item=(Type=Bone,Name=\"spine_04\"),Weight=0.0253),(Item=(Type=Bone,Name=\"spine_05\"),Weight=0.0355),(Item=(Type=Bone,Name=\"upperarm_twist_02_r\"),Weight=0.1147),(Item=(Type=Bone,Name=\"clavicle_r\"),Weight=0.177 ...(truncated)"
    Weight=1
   )
   > &RigUnit_ParentConstraint(
    AdvancedSettings=(RotationOrderForFilter=XZY)
    Child=(Type=Null,Name="left_collar_aimat_null")
    Parents="((Item=(Type=Bone,Name=\"spine_03\"),Weight=0.0188),(Item=(Type=Bone,Name=\"spine_04\"),Weight=0.0253),(Item=(Type=Bone,Name=\"spine_05\"),Weight=0.0355),(Item=(Type=Bone,Name=\"upperarm_twist_02_l\"),Weight=0.1147),(Item=(Type=Bone,Name=\"clavicle_l\"),Weight=0.177 ...(truncated)"
    Weight=1
   )
   > &RigUnit_AimConstraintLocalSpaceOffset(
    AdvancedSettings=(DebugSettings=(bEnabled=false,Scale=10,WorldOffset=(Rotation=(X=0,Y=0,Z=0,W=1),Translation=(X=0,Y=0,Z=0),Scale3D=(X=1,Y=1,Z=1))),RotationOrderForFilter=XZY)
    Child=(Type=Bone,Name="dyn_shirt_collar_l_01")
    Filter=(bX=True,bY=True,bZ=True)
    Parents=((Item=(Type=Null,Name="left_collar_aimat_null"),Weight=1))
    UpAxis=(X=0,Y=1,Z=0)
    Weight=1
    WorldUp=(Target=(X=5,Y=0,Z=-2),Kind=Location,Space=(Type=Null,Name="left_collar_aimat_null"))
   )
   > &RigUnit_AimConstraintLocalSpaceOffset(
    AdvancedSettings=(DebugSettings=(bEnabled=false,Scale=10,WorldOffset=(Rotation=(X=0,Y=0,Z=0,W=1),Translation=(X=0,Y=0,Z=0),Scale3D=(X=1,Y=1,Z=1))),RotationOrderForFilter=XZY)
    AimAxis=(X=-1,Y=0,Z=0)
    Child=(Type=Bone,Name="dyn_shirt_collar_r_01")
    Filter=(bX=True,bY=True,bZ=True)
    Parents=((Item=(Type=Null,Name="right_collar_aimat_null"),Weight=1))
    UpAxis=(X=0,Y=1,Z=0)
    Weight=1
    WorldUp=(Target=(X=5,Y=0,Z=2),Kind=Location,Space=(Type=Null,Name="right_collar_aimat_null"))
   )
   RigVMFunction_MathFloatMake = /Script/RigVM.RigVMFunction_MathFloatMake(Value=0) {is=struct RigVMUnitNode}
   DISPATCH_RigDispatch_GetAnimAttribute = RigVMDispatchNode(BoneName=root, Name=deformRig_l_upperarm_up_alpha, Default=@RigVMFunction_MathFloatMake.Value) {template="DISPATCH_RigDispatch_GetAnimAttribute(in Name,in BoneName,in Default,out Value,out Found)"}
   &RigUnit_ModifyTransforms(
    ItemToModify="((Item=(Type=Bone,Name=\"dyn_lapel_l_01\"),Transform=(Rotation=(X=-0.0398,Y=0,Z=0,W=0.9992),Translation=(X=0,Y=0,Z=0),Scale3D=(X=1,Y=1,Z=1))),(Item=(Type=Bone,Name=\"dyn_lapel_l_02\"),Transform=(Rotation=(X=0,Y=0,Z=0,W=1),Translation=(X=0,Y=0,Z=0),Scale3D=(X=1 ...(truncated)"
    WeightMinimum=0
    Weight=@DISPATCH_RigDispatch_GetAnimAttribute.Value
    ExecutePin=@RigVMFunction_Sequence.C
   )
   > &RigUnit_ModifyTransforms(
    ItemToModify="((Item=(Type=Bone,Name=\"dyn_lapel_r_01\"),Transform=(Rotation=(X=0,Y=0,Z=0,W=1),Translation=(X=0,Y=0,Z=0),Scale3D=(X=1,Y=1,Z=1))),(Item=(Type=Bone,Name=\"dyn_lapel_r_02\"),Transform=(Rotation=(X=0,Y=0,Z=0,W=1),Translation=(X=0,Y=0,Z=0),Scale3D=(X=1,Y=1,Z=1))) ...(truncated)"
    WeightMinimum=0
    Weight=RigVMDispatchNode(BoneName=root, Name=deformRig_r_upperarm_up_alpha, Default=@RigVMFunction_MathFloatMake.Value) {template="DISPATCH_RigDispatch_GetAnimAttribute(in Name,in BoneName,in Default,out Value,out Found)"}.Value
   )
   &RigUnit_ParentConstraint(
    AdvancedSettings=(RotationOrderForFilter=XZY)
    Child=(Type=Bone,Name="dyn_necklace_01")
    Parents="((Item=(Type=Bone,Name=\"clavicle_l\"),Weight=0.004),(Item=(Type=Bone,Name=\"spine_03\"),Weight=0.0164),(Item=(Type=Bone,Name=\"clavicle_r\"),Weight=0.0251),(Item=(Type=Bone,Name=\"neck_01\"),Weight=0.0372),(Item=(Type=Bone,Name=\"spine_04\"),Weight=0.0733),(Item=(T ...(truncated)"
    Weight=1
    ExecutePin=@RigVMFunction_Sequence.D
   )
   > &RigUnit_AimConstraintLocalSpaceOffset(
    Child=(Type=Bone,Name="dyn_necklace_01")
    Filter=(bX=True,bY=True,bZ=True)
    Parents=((Item=(Type=Bone,Name="spine_05"),Weight=1))
    Weight=1
    WorldUp=(Target=(X=0,Y=0,Z=1),Kind=Location,Space=(Type=Bone,Name="spine_05"))
   )
   > n2 = &RigUnit_SphericalPoseReader(
    ActiveRegionScaleFactors=(PositiveWidth=1,NegativeWidth=3,PositiveHeight=2,NegativeHeight=2)
    ActiveRegionSize=0.1
    Debug=(bDrawDebug=false,bDraw2D=False,bDrawLocalAxes=False,DebugScale=25,DebugSegments=20,DebugThickness=0.25)
    DriverAxis=(X=1,Y=0,Z=0)
    DriverItem=(Type=Bone,Name="head")
    FalloffRegionScaleFactors=(PositiveWidth=1,NegativeWidth=1,PositiveHeight=1,NegativeHeight=1)
    FalloffSize=0.3
    FlipHeightScaling=False
    FlipWidthScaling=False
    OptionalParentItem=(Type=Bone,Name="spine_05")
    RotationOffset=(X=-90,Y=0,Z=20)
   )
   > &RigUnit_ModifyTransforms(
    ItemToModify="((Item=(Type=Bone,Name=\"dyn_necklace_l_side\"),Transform=(Rotation=(X=0,Y=0,Z=0,W=1),Translation=(X=0,Y=0.1378,Z=0.7617),Scale3D=(X=1,Y=1,Z=1))),(Item=(Type=Bone,Name=\"dyn_necklace_r_side\"),Transform=(Rotation=(X=0,Y=0,Z=0,W=1),Translation=(X=0,Y=-0.5622,Z= ...(truncated)"
    WeightMinimum=0
    Weight=@n2.OutputParam
   )
   > n3 = &RigUnit_SphericalPoseReader(
    ActiveRegionScaleFactors=(PositiveWidth=1,NegativeWidth=3,PositiveHeight=2,NegativeHeight=2)
    ActiveRegionSize=0.1
    Debug=(bDrawDebug=false,bDraw2D=False,bDrawLocalAxes=False,DebugScale=25,DebugSegments=20,DebugThickness=0.25)
    DriverAxis=(X=1,Y=0,Z=0)
    DriverItem=(Type=Bone,Name="head")
    FalloffRegionScaleFactors=(PositiveWidth=1,NegativeWidth=1,PositiveHeight=1,NegativeHeight=1)
    FalloffSize=0.3
    FlipHeightScaling=False
    FlipWidthScaling=False
    OptionalParentItem=(Type=Bone,Name="spine_05")
    RotationOffset=(X=90,Y=0,Z=-25)
   )
   > &RigUnit_ModifyTransforms(
    ItemToModify="((Item=(Type=Bone,Name=\"dyn_necklace_l_side\"),Transform=(Rotation=(X=0,Y=-0.1495,Z=0,W=0.9888),Translation=(X=2.0972,Y=-0.2997,Z=0.7617),Scale3D=(X=1,Y=1,Z=1))),(Item=(Type=Bone,Name=\"dyn_necklace_r_side\"),Transform=(Rotation=(X=0,Y=-0.0928,Z=0,W=0.9957),T ...(truncated)"
    WeightMinimum=0
    Weight=@n3.OutputParam
   )
   RigVMFunctionReturnNode(ExecuteContext=@RigVMFunction_Sequence.E)
  ]}
 New Function_ContainedGraph {nodes [
   RigVMFunctionEntryNode()
   > &RigUnit_RotationConstraintLocalSpaceOffset(
    AdvancedSettings=(InterpolationType=Average,RotationOrderForFilter=XZY)
    Child=(Type=Bone,Name="dyn_thigh_l_cloth_bounce_01")
    Filter=(bX=True,bY=True,bZ=True)
    Parents=((Item=(Type=Bone,Name="calf_l"),Weight=0.3),(Item=(Type=Bone,Name="thigh_twist_01_l"),Weight=1))
    Weight=0.5
   )
   > &RigUnit_RotationConstraintLocalSpaceOffset(
    AdvancedSettings=(InterpolationType=Average,RotationOrderForFilter=XZY)
    Child=(Type=Bone,Name="dyn_thigh_r_cloth_bounce_01")
    Filter=(bX=True,bY=True,bZ=True)
    Parents=((Item=(Type=Bone,Name="calf_r"),Weight=0.3),(Item=(Type=Bone,Name="thigh_twist_01_r"),Weight=1))
    Weight=0.5
   )
   > RigVMFunctionReturnNode()
  ]}
 RigVMFunctionLibrary {nodes [
   Jacket and Sword = RigVMCollapseNode(ExecuteContext=())
   Lapels and Collar and Necklace = RigVMCollapseNode(ExecuteContext=())
   Leg Cloth = RigVMCollapseNode(ExecuteContext=())
   Pins = RigVMCollapseNode(ExecuteContext=())
  ]}
 RigVMFunction_Sequence_1_1_ContainedGraph {nodes [
   n1 = RigVMFunctionEntryNode()
   n2 = &RigVMFunction_Sequence(ExecuteContext=@n1.ExecuteContext)
   n3 = &RigVMFunction_Sequence(ExecuteContext=@n2.B)
   n4 = &RigVMFunction_Sequence(ExecuteContext=@n3.B)
   RigVMFunctionReturnNode(A=@n2.A, B=@n3.A, C=@n4.A, D=@n4.B)
  ]}
 RigVMFunction_Sequence_1_ContainedGraph {nodes [
   n1 = RigVMFunctionEntryNode()
   RigVMFunction_Sequence = &RigVMFunction_Sequence(ExecuteContext=@n1.ExecuteContext)
   n2 = &RigVMFunction_Sequence(ExecuteContext=@RigVMFunction_Sequence.B)
   n3 = &RigVMFunction_Sequence(ExecuteContext=@n2.B)
   n4 = &RigVMFunction_Sequence(ExecuteContext=@n3.B)
   RigVMFunctionReturnNode(A=@RigVMFunction_Sequence.A, B=@n2.A, C=@n3.A, D=@n4.A, E=@n4.B)
  ]}
 RigVMModel {nodes [
   RigUnit_BeginExecution = &RigUnit_BeginExecution()
   FN_DeformRig_FNC = RigVMFunctionReferenceNode(
    DeformRig_ Curve="(deformRig_l_upperarm_up_alpha_72_#1=True,deformRig_r_upperarm_up_alpha_73_#2=True,deformRig_l_upperarm_fr_alpha_74_#3=True,deformRig_r_upperarm_fr_alpha_75_CB51B80C4 ...(truncated)"
    DeformRig_Bool="(upperarm_1_#4=True,elbow_3_#5=True,wrist_7_#6=True,groin_8_#7=False,glute_17_#8=False,knee_9_40C408324CA2 ...(truncated)"
    DeformRig_Debug=(upperarm_debug_18_#4=False,wrist_debug_20_#6=False,lat_debug_26_#9=False,groinGlute_debug_27_#7=False)
    DeformRig_Value="(upperarm_up_18_#10=(Rotation=(X=0.0,Y=-0.0,Z=0.0,W=1.0),Translation=(X=0.0,Y=0.0,Z=8.0),Scale3D=(X=0.70,Y=1.0,Z=1.0)),upperarm_fwd_fr_19_#11=(Rotation=(X=0.0,Y=-0.0,Z=0.0,W=1.0),Translation=(X=0.0, ...(truncated)"
    ExecuteContext=@RigUnit_BeginExecution.ExecutePin
   )
   > Pins = RigVMFunctionReferenceNode()
   > Jacket and Sword = RigVMFunctionReferenceNode()
   > Lapels and Collar and Necklace = RigVMFunctionReferenceNode()
   > Leg Cloth = RigVMFunctionReferenceNode()
   > &RigUnit_ModifyTransforms(
    ItemToModify="((Item=(Type=Bone,Name=\"dyn_jacket_front_l_01_MSTR\"),Transform=(Rotation=(X=0,Y=-0.0175,Z=0,W=0.9998),Translation=(X=0,Y=0,Z=1),Scale3D=(X=1,Y=1,Z=1))),(Item=(Type=Bone,Name=\"dyn_jacket_front_r_01_MSTR\"),Transform=(Rotation=(X=0,Y=0,Z=0,W=1),Translation=(X ...(truncated)"
    Weight=1
    WeightMinimum=0
   )
   > &RigUnit_ModifyTransforms(
    ItemToModify=((Item=(Type=Bone,Name="deform_gun"),Transform=(Rotation=(X=0,Y=0,Z=0,W=1),Translation=(X=0,Y=0,Z=0),Scale3D=(X=0.95,Y=0.95,Z=0.95))))
    Weight=1
    WeightMinimum=0
   )
   RigVMFunction_MathBoolToFloat = &RigVMFunction_MathBoolToFloat(Value=RigVMVariableNode() {variable=IsSkydiving}.Value)
   AlphaInterp = &RigVMFunction_AlphaInterp(InterpSpeedDecreasing=10, InterpSpeedIncreasing=10, Scale=1, bClampResult=true, Value=@RigVMFunction_MathBoolToFloat.Result)
   > &RigUnit_ModifyTransforms(
    ItemToModify="((Item=(Type=Bone,Name=\"dyn_sword_bandolier_01_MSTR\"),Transform=(Rotation=(X=0,Y=-0.3827,Z=0,W=0.9239),Translation=(X=5,Y=0,Z=0),Scale3D=(X=1,Y=1,Z=1))),(Item=(Type=Bone,Name=\"dyn_jacket_front_l_01_MSTR\"),Transform=(Rotation=(X=0,Y=-0.2588,Z=0,W=0.9659),Tr ...(truncated)"
    WeightMinimum=0
    Weight=@AlphaInterp.Result
   )
   n1 = &RigVMFunction_AlphaInterp(
    InterpSpeedDecreasing=10
    InterpSpeedIncreasing=10
    Scale=1
    bClampResult=true
    Value=&RigVMFunction_MathBoolToFloat(Value=RigVMVariableNode() {variable=IsSliding}.Value).Result
   )
   > &RigUnit_ModifyTransforms(
    ItemToModify="((Item=(Type=Bone,Name=\"root_jacket_floor_collision\"),Transform=(Rotation=(X=0,Y=0,Z=0,W=1),Translation=(X=0,Y=-2,Z=2),Scale3D=(X=1,Y=1,Z=1))),(Item=(Type=Bone,Name=\"dyn_jacket_side_r_01_MSTR\"),Transform=(Rotation=(X=0,Y=-0.4081,Z=0,W=0.9129),Translation=( ...(truncated)"
    WeightMinimum=0
    Weight=@n1.Result
   )
   n2 = &RigVMFunction_AlphaInterp(
    InterpSpeedDecreasing=10
    InterpSpeedIncreasing=10
    Scale=1
    bClampResult=true
    Value=&RigVMFunction_MathBoolToFloat(Value=RigVMVariableNode() {variable=IsKneeSliding}.Value).Result
   )
   > &RigUnit_ModifyTransforms(
    ItemToModify=((Item=(Type=Bone,Name="root_jacket_floor_collision"),Transform=(Rotation=(X=0,Y=0,Z=0,W=1),Translation=(X=0,Y=-2,Z=2),Scale3D=(X=1,Y=1,Z=1))))
    WeightMinimum=0
    Weight=@n2.Result
   )
  ]}
]
hierarchy [
 root {loc=<0 0 0>, rot=<0 0 0 1>}
 pelvis {parent=root, loc=<0 -0.7071 93.1191>, rot=<0.0127 0.707 -0.0127 -0.707>}
 spine_01 {parent=pelvis, loc=<6.7494 -0.6388 0>, rot=<0 0 0.034 -0.9994>}
 spine_02 {parent=spine_01, loc=<7.2518 -0.035 0>, rot=<0 0 0.0238 -0.9997>}
 spine_03 {parent=spine_02, loc=<6.0104 0 0>, rot=<0 0 0.0149 -0.9999>}
 spine_04 {parent=spine_03, loc=<8.1024 0 0>, rot=<0 0 -0.0132 -0.9999>}
 spine_05 {parent=spine_04, loc=<10.0146 0 0>, rot=<0 0 -0.0015 -1>}
 attach_cape {parent=spine_05, loc=<9.017 -11.714 0>, rot=<0.0616 -0.7044 -0.0616 -0.7044>}
 attach_backpack {parent=spine_05, loc=<-0.123 -13.473 0>, rot=<-0.0093 -0.707 0.0093 -0.707>}
 clavicle_l {parent=spine_05, loc=<4.3575 -0.1466 -2.3223>, rot=<0.0465 -0.7289 -0.009 -0.6829>}
 upperarm_l {parent=clavicle_l, loc=<15.5111 0 0>, rot=<0.0226 -0.3234 -0.0088 -0.9459>}
 lowerarm_l {parent=upperarm_l, loc=<26.6514 0 0>, rot=<0.0391 -0.0099 -0.2453 -0.9686>}
 hand_l {parent=lowerarm_l, loc=<22.3491 0 0>, rot=<-0.7081 -0.0507 0.0975 -0.6975>}
 thumb_01_l {parent=hand_l, loc=<1.3118 -0.3191 -2.459>, rot=<0.6546 -0.4042 0.0664 -0.6354>}
 thumb_02_l {parent=thumb_01_l, loc=<5.9336 -0.0448 0.0401>, rot=<-0.0037 0.0213 0.298 -0.9543>}
 thumb_03_l {parent=thumb_02_l, loc=<2.2223 0.3682 -0.0907>, rot=<0.0118 -0.013 0.0066 -0.9998>}
 index_metacarpal_l {parent=hand_l, loc=<1.6535 0.0183 -2.96>, rot=<0.0158 -0.0124 -0.0173 -0.9996>}
 index_01_l {parent=index_metacarpal_l, loc=<8.3008 0.0357 -0.0524>, rot=<-0.0044 0.0485 0.0924 -0.9945>}
 index_02_l {parent=index_01_l, loc=<3.3587 -0.3004 0.0452>, rot=<-0.002 0.0027 0.3225 -0.9466>}
 index_03_l {parent=index_02_l, loc=<2.5578 -0.123 -0.0252>, rot=<-0.002 0.0067 0.2558 -0.9667>}
 middle_metacarpal_l {parent=hand_l, loc=<2.1856 -0.1725 -0.5984>, rot=<0.006 0.0168 -0.0329 -0.9993>}
 middle_01_l {parent=middle_metacarpal_l, loc=<7.3275 -0.2447 0.0129>, rot=<-0.0208 0.0042 0.0858 -0.9961>}
 middle_02_l {parent=middle_01_l, loc=<4.0123 0.1401 0.1175>, rot=<-0.0218 0.0258 0.3512 -0.9357>}
 middle_03_l {parent=middle_02_l, loc=<2.6077 0.0534 0.1373>, rot=<0.0082 0.0188 0.1708 -0.9851>}
 ring_metacarpal_l {parent=hand_l, loc=<2.2074 -0.3625 1.3638>, rot=<-0.0284 0.0722 -0.0458 -0.9959>}
 ring_01_l {parent=ring_metacarpal_l, loc=<7.0295 -0.6197 -0.1446>, rot=<-0.0489 -0.0252 0.1892 -0.9804>}
 ring_02_l {parent=ring_01_l, loc=<3.6174 0.4647 -0.1027>, rot=<0 0 0.2806 -0.9598>}
 ring_03_l {parent=ring_02_l, loc=<2.4152 0.114 0.0199>, rot=<0 0 0.1189 -0.9929>}
 pinky_metacarpal_l {parent=hand_l, loc=<1.7616 -0.9753 3.3814>, rot=<-0.0397 0.0814 -0.069 -0.9935>}
 pinky_01_l {parent=pinky_metacarpal_l, loc=<7.1709 -0.4831 -0.1094>, rot=<-0.0654 -0.0035 0.2242 -0.9723>}
 pinky_02_l {parent=pinky_01_l, loc=<2.8274 0.0893 -0.1043>, rot=<0.0037 0.0011 0.2854 -0.9584>}
 pinky_03_l {parent=pinky_02_l, loc=<2.2046 0.0307 -0.0078>, rot=<-0.0002 0.0098 0.1445 -0.9895>}
 deform_wrist_dn_base_l {parent=hand_l, loc=<0 -2.2342 0>, rot=<0 0 0 -1>}
 deform_wrist_dn_l {parent=deform_wrist_dn_base_l, loc=<0 0 0>, rot=<0 0 0 1>}
 deform_wrist_up_base_l {parent=hand_l, loc=<0 2.2682 0>, rot=<0 0 0 -1>}
 deform_wrist_up_l {parent=deform_wrist_up_base_l, loc=<0 0 0>, rot=<0 0 0 1>}
 lowerarm_twist_02_l {parent=lowerarm_l, loc=<9.6199 0 0>, rot=<0 0 0 -1>}
 lowerarm_twist_01_l {parent=lowerarm_l, loc=<17.8549 0 0>, rot=<0 0 0 -1>}
 deform_elbow_in_base_l {parent=lowerarm_l, loc=<0.2112 5.259 0.8274>, rot=<0 0 0 -1>}
 deform_elbow_in_l {parent=deform_elbow_in_base_l, loc=<0 0 0>, rot=<0 0 0 1>}
 deform_elbow_out_base_l {parent=lowerarm_l, loc=<0 -4 0>, rot=<0 0 0 -1>}
 deform_elbow_out_l {parent=deform_elbow_out_base_l, loc=<0 0 0>, rot=<0 0 0 1>}
 upperarm_twist_01_l {parent=upperarm_l, loc=<0.0041 0 0>, rot=<0 0 0 1>}
 deform_upperarm_bk_base_l {parent=upperarm_twist_01_l, loc=<1.0813 -7.2193 -1.9453>, rot=<0 0 0 1>}
 deform_upperarm_bk_l {parent=deform_upperarm_bk_base_l, loc=<0 0 0>, rot=<0 0 0 1>}
 deform_upperarm_fr_base_l {parent=upperarm_twist_01_l, loc=<-0.3691 7.0506 -2.765>, rot=<0 0 0 1>}
 deform_upperarm_fr_l {parent=deform_upperarm_fr_base_l, loc=<0 0 0>, rot=<0 0 0 1>}
 deform_upperarm_up_base_l {parent=upperarm_twist_01_l, loc=<-3.2216 -1.1391 3.3311>, rot=<0 0 0 1>}
 deform_upperarm_up_l {parent=deform_upperarm_up_base_l, loc=<0 0 0>, rot=<0 0 0 1>}
 upperarm_twist_02_l {parent=upperarm_l, loc=<13.7941 0 0>, rot=<0 0 0 1>}
 neck_01 {parent=spine_05, loc=<11.3138 0.1529 0>, rot=<0 0 -0.2003 -0.9797>}
 neck_02 {parent=neck_01, loc=<4.5627 0 0>, rot=<0 0 0.0504 -0.9987>}
 head {parent=neck_02, loc=<5.3914 0 0>, rot=<0 0 0.115 -0.9934>}
 clavicle_r {parent=spine_05, loc=<4.3575 -0.1466 2.3223>, rot=<0.7289 0.0465 -0.6829 0.009>}
 upperarm_r {parent=clavicle_r, loc=<-15.5111 0 0>, rot=<0.0226 -0.3234 -0.0088 -0.9459>}
 lowerarm_r {parent=upperarm_r, loc=<-26.6514 0 0>, rot=<0.0391 -0.0099 -0.2453 -0.9686>}
 hand_r {parent=lowerarm_r, loc=<-22.3491 0 0>, rot=<-0.7081 -0.0507 0.0975 -0.6975>}
 thumb_01_r {parent=hand_r, loc=<-1.3118 0.3191 2.459>, rot=<0.6546 -0.4042 0.0664 -0.6354>}
 thumb_02_r {parent=thumb_01_r, loc=<-5.9336 0.0448 -0.0401>, rot=<-0.0037 0.0213 0.298 -0.9543>}
 thumb_03_r {parent=thumb_02_r, loc=<-2.2223 -0.3682 0.0907>, rot=<0.0118 -0.013 0.0066 -0.9998>}
 index_metacarpal_r {parent=hand_r, loc=<-1.6535 -0.0183 2.96>, rot=<0.0158 -0.0124 -0.0173 -0.9996>}
 index_01_r {parent=index_metacarpal_r, loc=<-8.3008 -0.0357 0.0524>, rot=<-0.0044 0.0485 0.0924 -0.9945>}
 index_02_r {parent=index_01_r, loc=<-3.3587 0.3004 -0.0452>, rot=<-0.002 0.0027 0.3225 -0.9466>}
 index_03_r {parent=index_02_r, loc=<-2.5578 0.123 0.0252>, rot=<-0.002 0.0067 0.2558 -0.9667>}
 middle_metacarpal_r {parent=hand_r, loc=<-2.1856 0.1725 0.5984>, rot=<0.006 0.0168 -0.0329 -0.9993>}
 middle_01_r {parent=middle_metacarpal_r, loc=<-7.3275 0.2447 -0.0129>, rot=<-0.0208 0.0042 0.0858 -0.9961>}
 middle_02_r {parent=middle_01_r, loc=<-4.0123 -0.1401 -0.1175>, rot=<-0.0218 0.0258 0.3512 -0.9357>}
 middle_03_r {parent=middle_02_r, loc=<-2.6077 -0.0534 -0.1373>, rot=<0.0082 0.0188 0.1708 -0.9851>}
 ring_metacarpal_r {parent=hand_r, loc=<-2.2074 0.3625 -1.3638>, rot=<-0.0284 0.0722 -0.0458 -0.9959>}
 ring_01_r {parent=ring_metacarpal_r, loc=<-7.0295 0.6197 0.1446>, rot=<-0.0489 -0.0252 0.1892 -0.9804>}
 ring_02_r {parent=ring_01_r, loc=<-3.6174 -0.4647 0.1027>, rot=<0 0 0.2806 -0.9598>}
 ring_03_r {parent=ring_02_r, loc=<-2.4152 -0.114 -0.0199>, rot=<0 0 0.1189 -0.9929>}
 pinky_metacarpal_r {parent=hand_r, loc=<-1.7616 0.9753 -3.3814>, rot=<-0.0397 0.0814 -0.069 -0.9935>}
 pinky_01_r {parent=pinky_metacarpal_r, loc=<-7.1709 0.4831 0.1094>, rot=<-0.0654 -0.0035 0.2242 -0.9723>}
 pinky_02_r {parent=pinky_01_r, loc=<-2.8274 -0.0893 0.1043>, rot=<0.0037 0.0011 0.2854 -0.9584>}
 pinky_03_r {parent=pinky_02_r, loc=<-2.2046 -0.0307 0.0078>, rot=<-0.0002 0.0098 0.1445 -0.9895>}
 deform_wrist_dn_base_r {parent=hand_r, loc=<-0.0004 2.2345 -0.0001>, rot=<0 0 0 -1>}
 deform_wrist_dn_r {parent=deform_wrist_dn_base_r, loc=<0 0 0>, rot=<0 0 0 1>}
 deform_wrist_up_base_r {parent=hand_r, loc=<0.0003 -2.2684 0.0001>, rot=<0 0 0 -1>}
 deform_wrist_up_r {parent=deform_wrist_up_base_r, loc=<0 0 0>, rot=<0 0 0 1>}
 lowerarm_twist_02_r {parent=lowerarm_r, loc=<-9.6199 0 0>, rot=<0 0 0 -1>}
 lowerarm_twist_01_r {parent=lowerarm_r, loc=<-17.8549 0 0>, rot=<0 0 0 -1>}
 deform_elbow_in_base_r {parent=lowerarm_r, loc=<-0.2111 -5.2591 -0.8275>, rot=<0 0 0 -1>}
 deform_elbow_in_r {parent=deform_elbow_in_base_r, loc=<-0.0001 0 0.0002>, rot=<0 0 0 1>}
 deform_elbow_out_base_r {parent=lowerarm_r, loc=<0 4 -0.0001>, rot=<0 0 0 -1>}
 deform_elbow_out_r {parent=deform_elbow_out_base_r, loc=<0 0 0>, rot=<0 0 0 1>}
 upperarm_twist_01_r {parent=upperarm_r, loc=<-0.0041 0 0>, rot=<0 0 0 -1>}
 deform_upperarm_bk_base_r {parent=upperarm_twist_01_r, loc=<-1.0814 7.2192 1.9455>, rot=<0 0 0 1>}
 deform_upperarm_bk_r {parent=deform_upperarm_bk_base_r, loc=<0 0 0>, rot=<0 0 0 1>}
 deform_upperarm_fr_base_r {parent=upperarm_twist_01_r, loc=<0.3689 -7.0506 2.7652>, rot=<0 0 0 1>}
 deform_upperarm_fr_r {parent=deform_upperarm_fr_base_r, loc=<0 0 0>, rot=<0 0 0 1>}
 deform_upperarm_up_base_r {parent=upperarm_twist_01_r, loc=<3.2214 1.1391 -3.3308>, rot=<0 0 0 1>}
 deform_upperarm_up_r {parent=deform_upperarm_up_base_r, loc=<0 0 0>, rot=<0 0 0 -1>}
 upperarm_twist_02_r {parent=upperarm_r, loc=<-13.7941 0 0>, rot=<0 0 0 -1>}
 dyn_necklace_01 {parent=spine_05, loc=<7.7623 6.6237 -0.0613>, rot=<-0.3024 -0.9532 0 0>}
 dyn_necklace_02 {parent=dyn_necklace_01, loc=<3.1109 0.0066 -0.0613>, rot=<0 0.0191 0 -0.9998>}
 dyn_necklace_03 {parent=dyn_necklace_02, loc=<0.7617 -0.0125 -0.0291>, rot=<0 0.0079 0.0026 -1>}
 dyn_necklace_l_side {parent=spine_05, loc=<10.6124 5.1129 -4.7787>, rot=<-0.4044 -0.6956 -0.5893 0.0736>}
 dyn_necklace_r_side {parent=spine_05, loc=<10.6124 5.1129 4.7787>, rot=<-0.6956 0.4044 -0.0736 -0.5893>}
 deform_pec_aim_base_l {parent=spine_04, loc=<7.8732 12.6388 -2.4123>, rot=<0.0102 -0.4602 0.3979 -0.7936>}
 deform_pec_aim_l {parent=deform_pec_aim_base_l, loc=<0 0 0>, rot=<0 0 0 1>}
 deform_pec_base_l {parent=deform_pec_aim_l, loc=<10.766 0 0>, rot=<0 0 0 1>}
 deform_pec_l {parent=deform_pec_base_l, loc=<0.055 0 0>, rot=<0 0 0 1>}
 deform_pec_aim_base_r {parent=spine_04, loc=<7.8727 12.6388 2.4123>, rot=<-0.4602 -0.0102 0.7936 0.3979>}
 deform_pec_aim_r {parent=deform_pec_aim_base_r, loc=<0 0 0>, rot=<0 0 0 1>}
 deform_pec_base_r {parent=deform_pec_aim_r, loc=<-10.7662 -0.0006 -0.0007>, rot=<0 0 0 1>}
 deform_pec_r {parent=deform_pec_base_r, loc=<0 0 0>, rot=<0 0 0 1>}
 dyn_lapel_l_01 {parent=spine_04, loc=<3.957 12.5489 -8.945>, rot=<0.0456 -0.5804 0.1384 -0.8012>}
 dyn_lapel_l_02 {parent=dyn_lapel_l_01, loc=<4.5815 0 0>, rot=<0 0 -0.2843 -0.9587>}
 dyn_lapel_r_01 {parent=spine_04, loc=<3.9567 12.1265 8.945>, rot=<-0.5804 -0.0456 0.8012 0.1384>}
 dyn_lapel_r_02 {parent=dyn_lapel_r_01, loc=<-5.1328 -0.0001 -0.0005>, rot=<0 0 -0.2843 -0.9587>}
 dyn_shirt_collar_l_01 {parent=spine_04, loc=<19.9872 6.2141 -6.7948>, rot=<-0.2491 -0.7348 0.3613 -0.5172>}
 dyn_shirt_collar_l_02 {parent=dyn_shirt_collar_l_01, loc=<3.6343 0 0>, rot=<0 0 0 1>}
 dyn_shirt_collar_r_01 {parent=spine_04, loc=<19.9876 6.2142 6.7948>, rot=<0.7348 -0.2491 -0.5172 -0.3613>}
 dyn_shirt_collar_r_02 {parent=dyn_shirt_collar_r_01, loc=<-3.6346 0.0007 0.0005>, rot=<0 0 0 1>}
 dyn_sword_bandolier_01_MSTR {parent=spine_03, loc=<2.2133 -0.4912 -3.7979>, rot=<0 -0.9649 0 -0.2627>}
 dyn_sword_bandolier_01 {parent=dyn_sword_bandolier_01_MSTR, loc=<0 0 0>, rot=<0 0 0 1>}
 dyn_sword_bandolier_02 {parent=dyn_sword_bandolier_01, loc=<16.814 0 0>, rot=<0 0 0 1>}
 dyn_sword_bandolier_03 {parent=dyn_sword_bandolier_02, loc=<10.9046 0 0>, rot=<0 0 -0.2593 -0.9658>}
 dyn_sword {parent=dyn_sword_bandolier_03, loc=<5.7044 -0.5092 -0.7228>, rot=<0.119 -0.1244 0.5719 -0.802>}
 dyn_sword_HIDE {parent=dyn_sword, loc=<0 0 0>, rot=<0 0 0 1>}
 dyn_sword_bandolier_rear_mover_MSTR {parent=dyn_sword_bandolier_01, loc=<7.0813 -10.3402 5.0107>, rot=<0.2905 -0.0366 -0.2486 -0.9233>}
 dyn_sword_bandolier_rear_mover {parent=dyn_sword_bandolier_rear_mover_MSTR, loc=<0 0 0>, rot=<0 0 0 1>}
 dyn_sword_bandolier_front_mover_MSTR {parent=dyn_sword_bandolier_01, loc=<12.3448 12.0492 -1.6246>, rot=<-0.058 0.0176 0.1271 -0.99>}
 dyn_sword_bandolier_front_mover {parent=dyn_sword_bandolier_front_mover_MSTR, loc=<0 0 0>, rot=<0 0 0 1>}
 dyn_jacket_front_l_01_MSTR {parent=spine_02, loc=<0.3177 3.4492 -14.8722>, rot=<0.0743 -0.9839 -0.162 -0.0122>}
 dyn_jacket_front_l_01 {parent=dyn_jacket_front_l_01_MSTR, loc=<0 0 0>, rot=<0 0 0 1>}
 dyn_jacket_front_l_02 {parent=dyn_jacket_front_l_01, loc=<10.6454 0 0>, rot=<0 0.0922 0 -0.9957>}
 dyn_jacket_front_l_03 {parent=dyn_jacket_front_l_02, loc=<14.9328 0 0>, rot=<0 0.0079 0.0026 -1>}
 dyn_jacket_front_l_04 {parent=dyn_jacket_front_l_03, loc=<13.3543 0 0>, rot=<0.0001 0.0305 0.002 -0.9995>}
 dyn_jacket_front_l_05 {parent=dyn_jacket_front_l_04, loc=<11.7723 0 0>, rot=<0 0 0 1>}
 dyn_jacket_side_l_01_MSTR {parent=spine_02, loc=<1.9361 -10.5571 -10.2478>, rot=<0.0658 -0.8587 0.4926 -0.1251>}
 dyn_jacket_side_l_01 {parent=dyn_jacket_side_l_01_MSTR, loc=<0 0 0>, rot=<0 0 0 1>}
 dyn_jacket_side_l_02 {parent=dyn_jacket_side_l_01, loc=<7.5291 0.0851 -0.7574>, rot=<0 -0.0048 0 -1>}
 dyn_jacket_side_l_03 {parent=dyn_jacket_side_l_02, loc=<14.9328 0 0>, rot=<0 0.0389 0 -0.9992>}
 dyn_jacket_side_l_04 {parent=dyn_jacket_side_l_03, loc=<13.3543 0 0>, rot=<0 0 0 1>}
 dyn_jacket_side_l_05 {parent=dyn_jacket_side_l_04, loc=<11.7723 0 0>, rot=<0 0 0 1>}
 dyn_jacket_rear_01_MSTR {parent=spine_02, loc=<1.1912 -11.6769 0>, rot=<0.0601 -0.7045 0.7045 -0.0601>}
 dyn_jacket_rear_01 {parent=dyn_jacket_rear_01_MSTR, loc=<0 0 0>, rot=<0 0 0 1>}
 dyn_jacket_rear_02 {parent=dyn_jacket_rear_01, loc=<7.0531 0 0>, rot=<0 0.0426 0 -0.9991>}
 dyn_jacket_rear_03 {parent=dyn_jacket_rear_02, loc=<14.9328 0 0>, rot=<0 0 0 1>}
 dyn_jacket_rear_04 {parent=dyn_jacket_rear_03, loc=<13.3543 0 0>, rot=<0 0 0 1>}
 dyn_jacket_rear_05 {parent=dyn_jacket_rear_04, loc=<11.7723 0 0>, rot=<0 0 0 1>}
 dyn_jacket_front_r_01_MSTR {parent=spine_02, loc=<0.3172 3.4491 14.8722>, rot=<0.9839 0.0743 -0.0122 0.162>}
 dyn_jacket_front_r_01 {parent=dyn_jacket_front_r_01_MSTR, loc=<0 0 0>, rot=<0 0 0 1>}
 dyn_jacket_front_r_02 {parent=dyn_jacket_front_r_01, loc=<-10.645 0 0.0001>, rot=<0 0.0922 0 -0.9957>}
 dyn_jacket_front_r_03 {parent=dyn_jacket_front_r_02, loc=<-14.9328 0 0>, rot=<0 0.0079 0.0026 -1>}
 dyn_jacket_front_r_04 {parent=dyn_jacket_front_r_03, loc=<-13.3543 0 0>, rot=<0.0001 0.0305 0.002 -0.9995>}
 dyn_jacket_front_r_05 {parent=dyn_jacket_front_r_04, loc=<-11.7722 0 0>, rot=<0 0 0 1>}
 dyn_jacket_side_r_01_MSTR {parent=spine_02, loc=<1.9365 -10.557 10.2478>, rot=<0.8587 0.0658 -0.1251 -0.4926>}
 dyn_jacket_side_r_01 {parent=dyn_jacket_side_r_01_MSTR, loc=<0 0 0>, rot=<0 0 0 1>}
 dyn_jacket_side_r_02 {parent=dyn_jacket_side_r_01, loc=<-7.5293 -0.0851 0.7574>, rot=<0 -0.0048 0 -1>}
 dyn_jacket_side_r_03 {parent=dyn_jacket_side_r_02, loc=<-14.9331 0 0>, rot=<0 0.0389 0 -0.9992>}
 dyn_jacket_side_r_04 {parent=dyn_jacket_side_r_03, loc=<-13.3542 0.0001 0>, rot=<0 0 0 1>}
 dyn_jacket_side_r_05 {parent=dyn_jacket_side_r_04, loc=<-11.7723 0 0>, rot=<0 0 0 1>}
 dyn_cloth_front_l_01_MSTR {parent=spine_02, loc=<-30.7705 6.9415 -14.849>, rot=<0.0238 -0.937 -0.3384 -0.0834>}
 dyn_cloth_front_l_01 {parent=dyn_cloth_front_l_01_MSTR, loc=<0 0 0>, rot=<0 0 0 1>}
 dyn_cloth_front_l_02 {parent=dyn_cloth_front_l_01, loc=<13.3543 0 0>, rot=<0.0001 0.0305 0.002 -0.9995>}
 dyn_cloth_front_l_03 {parent=dyn_cloth_front_l_02, loc=<11.7723 0 0>, rot=<0 0 0 1>}
 thigh_l {parent=pelvis, loc=<-2.5693 -0.0109 -7.7501>, rot=<0.0012 0.0572 0.0155 -0.9982>}
 calf_l {parent=thigh_l, loc=<-39.0484 0 0>, rot=<0.0006 -0.0161 -0.0391 -0.9991>}
 foot_l {parent=calf_l, loc=<-42.561 0.1077 0.4292>, rot=<0.0418 -0.0455 0.0343 -0.9975>}
 ball_l {parent=foot_l, loc=<-6.1599 13.8264 0.6806>, rot=<0.0001 -0.0001 -0.7063 -0.7079>}
 calf_twist_02_l {parent=calf_l, loc=<-18.8064 -0.0967 0.0397>, rot=<0.0237 0.0095 -0.0035 -0.9997>}
 calf_twist_01_l {parent=calf_l, loc=<-35.813 -0.0967 0.0397>, rot=<0.0237 0.0095 -0.0035 -0.9997>}
 dyn_boot_l_laces_MSTR {parent=calf_l, loc=<-32.9118 -7.222 0.242>, rot=<0.0427 -0.998 -0.0039 0.0473>}
 dyn_boot_l_lace_a_01 {parent=dyn_boot_l_laces_MSTR, loc=<-0.1113 0.1656 0.6389>, rot=<-0.0299 0.1761 0.0182 -0.9837>}
 dyn_boot_l_lace_a_02 {parent=dyn_boot_l_lace_a_01, loc=<4.0929 0.067 -0.1742>, rot=<0.0112 -0.0605 -0.0241 -0.9978>}
 dyn_boot_l_lace_b_01 {parent=dyn_boot_l_laces_MSTR, loc=<0.1113 -0.1656 -0.6389>, rot=<0.0299 -0.1761 -0.0182 -0.9837>}
 dyn_boot_l_lace_b_02 {parent=dyn_boot_l_lace_b_01, loc=<3.8983 0.0435 0.2086>, rot=<-0.0084 0.0738 -0.016 -0.9971>}
 dyn_boot_l_lip_01 {parent=calf_l, loc=<-16.3052 -10.6696 -0.6785>, rot=<0.1213 -0.9924 -0.0172 0.0094>}
 dyn_boot_l_lip_02 {parent=calf_l, loc=<-17.1373 -1.3412 -6.5015>, rot=<0.0675 -0.7056 -0.701 -0.0783>}
 dyn_boot_l_lip_03 {parent=calf_l, loc=<-16.3965 6.4646 -0.812>, rot=<0.017 -0.0194 0.9836 0.1788>}
 dyn_boot_l_lip_04 {parent=calf_l, loc=<-15.5013 -1.9445 5.4594>, rot=<0.053 -0.6354 0.7656 0.0856>}
 thigh_twist_01_l {parent=thigh_l, loc=<-21.6303 0 0>, rot=<0.0013 0 0 -1>}
 dyn_thigh_l_cloth_bounce_01 {parent=thigh_l, loc=<-31.6282 0 0>, rot=<0.0006 -0.0161 -0.0391 -0.9991>}
 dyn_thigh_l_cloth_bounce_02 {parent=dyn_thigh_l_cloth_bounce_01, loc=<-14.6439 0 7.4126>, rot=<0 0 0 1>}
 dyn_thigh_l_cloth_bounce_03 {parent=dyn_thigh_l_cloth_bounce_01, loc=<-13.1366 0 -8.3835>, rot=<0 0 0 1>}
 thigh_r {parent=pelvis, loc=<-2.5693 -0.0109 7.7501>, rot=<0.0572 -0.0012 0.9982 0.0155>}
 calf_r {parent=thigh_r, loc=<39.0484 0 0>, rot=<0.0006 -0.0161 -0.0391 -0.9991>}
 foot_r {parent=calf_r, loc=<42.561 -0.1077 -0.4292>, rot=<0.0418 -0.0455 0.0343 -0.9975>}
 ball_r {parent=foot_r, loc=<6.1599 -13.8264 -0.6806>, rot=<0.0001 -0.0001 -0.7063 -0.7079>}
 calf_twist_02_r {parent=calf_r, loc=<18.8064 0.0967 -0.0397>, rot=<0.0237 0.0095 -0.0035 -0.9997>}
 calf_twist_01_r {parent=calf_r, loc=<35.813 0.0967 -0.0397>, rot=<0.0237 0.0095 -0.0035 -0.9997>}
 dyn_boot_r_laces_MSTR {parent=calf_r, loc=<32.835 7.0399 -0.4587>, rot=<0.9974 0.0563 -0.0453 -0.0019>}
 dyn_boot_r_lace_a_01 {parent=dyn_boot_r_laces_MSTR, loc=<0.1926 -0.3425 0.6385>, rot=<-0.0295 0.1493 -0.0367 -0.9877>}
 dyn_boot_r_lace_a_02 {parent=dyn_boot_r_lace_a_01, loc=<4.2212 0 0>, rot=<0 0 0 1>}
 dyn_boot_r_lace_b_01 {parent=dyn_boot_r_laces_MSTR, loc=<-0.1926 0.3425 -0.6385>, rot=<0.0295 -0.1493 0.0367 -0.9877>}
 dyn_boot_r_lace_b_02 {parent=dyn_boot_r_lace_b_01, loc=<4.5842 0 0>, rot=<0 0 0 1>}
 dyn_boot_r_lip_01 {parent=calf_r, loc=<16.3052 10.6696 0.6785>, rot=<0.1213 -0.9924 -0.0172 0.0094>}
 dyn_boot_r_lip_02 {parent=calf_r, loc=<17.1373 1.3412 6.5015>, rot=<0.0675 -0.7056 -0.701 -0.0783>}
 dyn_boot_r_lip_03 {parent=calf_r, loc=<16.3965 -6.4646 0.8121>, rot=<0.017 -0.0194 0.9835 0.1788>}
 dyn_boot_r_lip_04 {parent=calf_r, loc=<15.5013 1.9445 -5.4594>, rot=<0.053 -0.6354 0.7656 0.0856>}
 thigh_twist_01_r {parent=thigh_r, loc=<21.6303 0 0>, rot=<0.0013 0 0 -1>}
 dyn_thigh_r_cloth_bounce_01 {parent=thigh_r, loc=<31.6282 0 0>, rot=<0.0006 -0.0161 -0.0391 -0.9991>}
 dyn_thigh_r_cloth_bounce_02 {parent=dyn_thigh_r_cloth_bounce_01, loc=<14.6439 0 -7.4126>, rot=<0 0 0 -1>}
 dyn_thigh_r_cloth_bounce_03 {parent=dyn_thigh_r_cloth_bounce_01, loc=<13.1367 0 8.3835>, rot=<0 0 0 -1>}
 deform_gun {parent=pelvis, loc=<4.0753 9.3542 9.8474>, rot=<-0.2572 -0.7602 -0.1981 -0.5628>}
 root_jacket_floor_collision {parent=root, loc=<0 0 0>, rot=<0 0 0 1>}
 left_collar_aimat_null {type=Null, loc=<18.592 0.7587 136.6967>, rot=<0.0284 -0.7065 -0.0284 0.7065>}
 right_collar_aimat_null {type=Null, loc=<-18.592 0.7587 136.6967>, rot=<0.0284 -0.7065 -0.0284 0.7065>}
]
variables [
 DebugDraw {type=bool, default=False}
 IsKneeSliding {type=bool, default=False}
 IsSkydiving {type=bool, default=False}
 IsSliding {type=bool, default=False}
 MovementAlpha {type=double, default=0}
 RigVMFunction_MathQuaternionMake_W {type=float, default=1}
]
