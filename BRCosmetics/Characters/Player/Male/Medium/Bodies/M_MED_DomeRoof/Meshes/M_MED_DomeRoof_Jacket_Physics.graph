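# Physics asset for the DomeRoof male player jacket, defining bodies for secondary motion of the jacket and sword bandolier.
# Five-segment front, side and rear capsule chains use limited joints from spine_02, with front chains anchored through MSTR bodies.
# Free cross-links join neighboring panels lower down; a bandolier chain from spine_03 carries the sword and links to the left front panel.
# Leg, pelvis, spine and gun capsules, plus a large floor sphere, provide collision volumes.
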PhysicsAsset /BRCosmetics/Characters/Player/Male/Medium/Bodies/M_MED_DomeRoof/Meshes/M_MED_DomeRoof_Jacket_Physics
bodies {
 calf_l [{shapeType=Capsule, length=35.7321, radius=13.2401, center=<-27.6983 0.9394 0.4158>, rotation=<rot -0.021 -97.5766 -89.5816>}]
 calf_r [{shapeType=Capsule, length=36.2946, radius=13.6776, center=<28.1118 -1.0157 -0.4189>, rotation=<rot -0.0234 -97.2527 90.4183>}]
 deform_gun [{
   shapeType=Capsule
   length=10.4049
   radius=2.1645
   shapeName=deform_gun_capsule
   center=<-2.1698 0.0399 -12.582>
   rotation=<rot -3.201 -50.4375 2.9995>
  }]
 dyn_jacket_front_l_01 [{shapeType=Capsule, length=7, radius=6, shapeName=dyn_jacket_front_l_01_capsule, center=<4 0 7>, rotation=<rot 0 0 -90>}]
 dyn_jacket_front_l_01_MSTR [{shapeType=Capsule, length=0.505, radius=0.505, shapeName=dyn_jacket_front_l_01_MSTR_capsule, rotation=<rot 0 -90 0>}]
 dyn_jacket_front_l_02 [{shapeType=Capsule, length=11.5625, radius=6.5, shapeName=dyn_jacket_front_l_02_capsule, center=<5 0 7>, rotation=<rot 0 0 -90>}]
 dyn_jacket_front_l_03 [{shapeType=Capsule, length=11.875, radius=6.75, shapeName=dyn_jacket_front_l_03_capsule, center=<4 0 6>, rotation=<rot 0 0 -90>}]
 dyn_jacket_front_l_04 [{shapeType=Capsule, length=13.25, radius=7.8125, shapeName=dyn_jacket_front_l_04_capsule, center=<4 0 7.5>, rotation=<rot 0 0 -90>}]
 dyn_jacket_front_l_05 [{shapeType=Capsule, length=19.3125, radius=10.5625, shapeName=dyn_jacket_front_l_05_capsule, center=<4 0 7.5>, rotation=<rot 0 0 -90>}]
 dyn_jacket_front_r_01 [{shapeType=Capsule, length=14.625, radius=6.9375, shapeName=dyn_jacket_front_l_01_capsule, center=<-4 0 -6>, rotation=<rot 0 0 90>}]
 dyn_jacket_front_r_01_MSTR [{shapeType=Capsule, length=0.505, radius=0.505, shapeName=dyn_jacket_front_r_01_MSTR_capsule, rotation=<rot 0 -90 0>}]
 dyn_jacket_front_r_02 [{shapeType=Capsule, length=15.125, radius=7.6562, shapeName=dyn_jacket_front_l_02_capsule, center=<-5 0 -7>, rotation=<rot 0 0 90>}]
 dyn_jacket_front_r_03 [{
   shapeType=Capsule
   length=23.25
   radius=7.7812
   shapeName=dyn_jacket_front_l_03_capsule
   center=<-4 0.0993 -7.5534>
   rotation=<rot 0.0001 0.0002 87.8686>
  }]
 dyn_jacket_front_r_04 [{shapeType=Capsule, length=23.9062, radius=7.375, shapeName=dyn_jacket_front_l_04_capsule, center=<-4 0 -7.5>, rotation=<rot 0 0 90>}]
 dyn_jacket_front_r_05 [{shapeType=Capsule, length=25.0312, radius=7.6875, shapeName=dyn_jacket_front_l_05_capsule, center=<-4 0 -6>, rotation=<rot 0 0 90>}]
 dyn_jacket_rear_01 [{shapeType=Capsule, length=7, radius=6, shapeName=dyn_jacket_rear_01_capsule, center=<4 0 6>, rotation=<rot 0 0 -90>}]
 dyn_jacket_rear_02 [{shapeType=Capsule, length=7, radius=6, shapeName=dyn_jacket_rear_02_capsule, center=<4 0 6>, rotation=<rot 0 0 -90>}]
 dyn_jacket_rear_03 [{shapeType=Capsule, length=17.4375, radius=8.125, shapeName=dyn_jacket_rear_03_capsule, center=<4 0 5>, rotation=<rot 0 0 -90>}]
 dyn_jacket_rear_04 [{shapeType=Capsule, length=20, radius=10.375, shapeName=dyn_jacket_rear_04_capsule, center=<4 0 4.5>, rotation=<rot 0 0 -90>}]
 dyn_jacket_rear_05 [{shapeType=Capsule, length=15, radius=11, shapeName=dyn_jacket_rear_05_capsule, center=<4 0 6>, rotation=<rot 0 0 -90>}]
 dyn_jacket_side_l_01 [{shapeType=Capsule, length=7, radius=6, shapeName=dyn_jacket_side_l_01_capsule, center=<4 0 6>, rotation=<rot 0 0 -90>}]
 dyn_jacket_side_l_02 [{shapeType=Capsule, length=7, radius=6, shapeName=dyn_jacket_side_l_02_capsule, center=<4 0 6>, rotation=<rot 0 0 -90>}]
 dyn_jacket_side_l_03 [{shapeType=Capsule, length=14, radius=6.5, shapeName=dyn_jacket_side_l_03_capsule, center=<4 0 5>, rotation=<rot 0 0 -90>}]
 dyn_jacket_side_l_04 [{shapeType=Capsule, length=10.5625, radius=12.4375, shapeName=dyn_jacket_side_l_04_capsule, center=<4 0 5>, rotation=<rot 0 0 -90>}]
 dyn_jacket_side_l_05 [{shapeType=Capsule, length=15, radius=11, shapeName=dyn_jacket_side_l_05_capsule, center=<4 0 6>, rotation=<rot 0 0 -90>}]
 dyn_jacket_side_r_01 [{shapeType=Capsule, length=7, radius=6, shapeName=dyn_jacket_side_l_01_capsule, center=<-4 0 -6>, rotation=<rot 0 0 90>}]
 dyn_jacket_side_r_02 [{shapeType=Capsule, length=7, radius=6, shapeName=dyn_jacket_side_l_02_capsule, center=<-4 0 -6>, rotation=<rot 0 0 90>}]
 dyn_jacket_side_r_03 [{shapeType=Capsule, length=14, radius=7.5, shapeName=dyn_jacket_side_l_03_capsule, center=<-4 0 -5>, rotation=<rot 0 0 90>}]
 dyn_jacket_side_r_04 [{shapeType=Capsule, length=11.75, radius=11.25, shapeName=dyn_jacket_side_l_04_capsule, center=<-4 0 -5>, rotation=<rot 0 0 90>}]
 dyn_jacket_side_r_05 [{shapeType=Capsule, length=16.125, radius=9.875, shapeName=dyn_jacket_side_l_05_capsule, center=<-4 0 -6>, rotation=<rot 0 0 90>}]
 dyn_sword [{
   shapeType=Capsule
   length=36.2006
   radius=5.8683
   shapeName=dyn_sword_capsule
   center=<11.2445 -1.3428 2.0473>
   rotation=<rot 0 -94.7849 -89.181>
  }]
 dyn_sword_bandolier_01 [{
   shapeType=Capsule
   length=16.4448
   radius=9.0626
   shapeName=dyn_sword_bandolier_01_capsule
   center=<12.0819 0.5626 8.1082>
   rotation=<rot -17.8864 -2.8606 -87.8555>
  }]
 dyn_sword_bandolier_01_MSTR [{shapeType=Capsule, length=0.505, radius=0.505, shapeName=dyn_sword_bandolier_01_MSTR_capsule, rotation=<rot 0 -90 0>}]
 dyn_sword_bandolier_02 [{
   shapeType=Capsule
   length=9.3811
   radius=7.5398
   shapeName=dyn_sword_bandolier_02_capsule
   center=<7.0276 2.7549 4.9326>
   rotation=<rot -2.4855 0.6988 -71.4869>
  }]
 dyn_sword_bandolier_03 [{
   shapeType=Capsule
   length=39.3784
   radius=6.2877
   shapeName=dyn_sword_bandolier_03_capsule
   center=<22.522 -5.6727 13.4221>
   rotation=<rot 14.9538 -120.0282 -82.271>
  }]
 dyn_sword_bandolier_front_mover [{
   shapeType=Capsule
   length=7.4003
   radius=3.8543
   shapeName=dyn_sword_bandolier_front_mover_capsule
   center=<0.4328 3.1379 0.1333>
   rotation=<rot 5.8377 -82.2428 -92.1543>
  }]
 dyn_sword_bandolier_rear_mover [{
   shapeType=Capsule
   length=13.7789
   radius=4.905
   shapeName=dyn_sword_bandolier_rear_mover_capsule
   center=<1.4567 -2.6228 -0.5306>
   rotation=<rot -1.4433 -114.8232 -94.9397>
  }]
 pelvis [{shapeType=Capsule, length=7.1875, radius=11.8957}]
 root_jacket_floor_collision [{shapeType=Sphere, radius=117, shapeName=root_sphere, center=<0 0 -101.5537>}]
 spine_02 [{shapeType=Capsule, length=5.8107, radius=13.3438, shapeName=spine_02_capsule, rotation=<rot 0 -90 0>}]
 spine_03 [{shapeType=Capsule, length=7.3184, radius=12.4408, shapeName=spine_03_capsule}]
 thigh_l [
  {shapeType=Capsule, length=30.9586, radius=13.8161, center=<-21.3123 -1.3853 0.7052>, rotation=<rot -6.8551 85.2464 -87.776>}
  {shapeType=Capsule, length=4.5625, radius=15, shapeName=thigh_l_capsule, center=<-43.3107 -2.5304 0.3766>, rotation=<rot 85.0328 0 0.0001>}
 ]
 thigh_r [
  {shapeType=Capsule, length=31.6149, radius=13.1599, center=<21.3123 1.3853 -0.7052>, rotation=<rot -6.8551 85.2464 92.224>}
  {
   shapeType=Capsule
   length=4.25
   radius=13.6875
   shapeName=thigh_l_capsule
   center=<35.3164 -1.2345 -1.281>
   rotation=<rot 78.5202 -112.7984 66.7853>
  }
 ]
}
constraints [
 {bone1Name=calf_l, bone2Name=thigh_l, swing1LimitDegrees=45, swing2LimitDegrees=45, twistLimitDegrees=45}
 {
  bone1Name=dyn_jacket_front_l_01_MSTR
  bone2Name=spine_02
  swing1LimitDegrees=45
  swing1Motion=Free
  swing2LimitDegrees=45
  swing2Motion=Free
  twistLimitDegrees=45
  twistMotion=Free
 }
 {bone1Name=dyn_jacket_front_l_02, bone2Name=dyn_jacket_front_l_01, swing1LimitDegrees=6, swing2LimitDegrees=18, twistLimitDegrees=14}
 {bone1Name=dyn_jacket_front_l_03, bone2Name=dyn_jacket_front_l_02, swing1LimitDegrees=4, swing2LimitDegrees=32, twistLimitDegrees=8}
 {bone1Name=dyn_jacket_front_l_04, bone2Name=dyn_jacket_front_l_03, swing1LimitDegrees=16, swing2LimitDegrees=16, twistLimitDegrees=6}
 {bone1Name=dyn_jacket_front_l_05, bone2Name=dyn_jacket_front_l_04, swing1LimitDegrees=16, swing2LimitDegrees=16, twistLimitDegrees=6}
 {
  bone1Name=dyn_jacket_front_r_01_MSTR
  bone2Name=spine_02
  swing1LimitDegrees=45
  swing1Motion=Free
  swing2LimitDegrees=45
  swing2Motion=Free
  twistLimitDegrees=45
  twistMotion=Free
 }
 {bone1Name=dyn_jacket_rear_01, bone2Name=spine_02, swing1LimitDegrees=10, swing2LimitDegrees=10, twistLimitDegrees=10}
 {bone1Name=dyn_jacket_rear_02, bone2Name=dyn_jacket_rear_01, swing1LimitDegrees=10, swing2LimitDegrees=10, twistLimitDegrees=10}
 {bone1Name=dyn_jacket_rear_03, bone2Name=dyn_jacket_rear_02, swing1LimitDegrees=12, swing2LimitDegrees=32, twistLimitDegrees=8}
 {bone1Name=dyn_jacket_rear_04, bone2Name=dyn_jacket_rear_03, swing1LimitDegrees=12, swing2LimitDegrees=32, twistLimitDegrees=8}
 {bone1Name=dyn_jacket_rear_05, bone2Name=dyn_jacket_rear_04, swing1LimitDegrees=12, swing2LimitDegrees=12, twistLimitDegrees=8}
 {bone1Name=dyn_jacket_side_l_01, bone2Name=spine_02, swing1LimitDegrees=10, swing2LimitDegrees=10, twistLimitDegrees=10}
 {bone1Name=dyn_jacket_side_l_02, bone2Name=dyn_jacket_side_l_01, swing1LimitDegrees=10, swing2LimitDegrees=10, twistLimitDegrees=10}
 {bone1Name=dyn_jacket_side_l_03, bone2Name=dyn_jacket_side_l_02, swing1LimitDegrees=12, swing2LimitDegrees=32, twistLimitDegrees=8}
 {bone1Name=dyn_jacket_side_l_04, bone2Name=dyn_jacket_side_l_03, swing1LimitDegrees=12, swing2LimitDegrees=32, twistLimitDegrees=8}
 {bone1Name=dyn_jacket_side_l_05, bone2Name=dyn_jacket_side_l_04, swing1LimitDegrees=12, swing2LimitDegrees=12, twistLimitDegrees=8}
 {bone1Name=calf_r, bone2Name=thigh_r, swing1LimitDegrees=45, swing2LimitDegrees=45, twistLimitDegrees=45}
 {bone1Name=dyn_jacket_front_r_02, bone2Name=dyn_jacket_front_r_01, swing1LimitDegrees=10, swing2LimitDegrees=10, twistLimitDegrees=10}
 {bone1Name=dyn_jacket_front_l_01, bone2Name=dyn_jacket_front_l_01_MSTR, swing1LimitDegrees=6, swing2LimitDegrees=18, twistLimitDegrees=14}
 {bone1Name=dyn_jacket_front_r_01, bone2Name=dyn_jacket_front_r_01_MSTR, swing1LimitDegrees=10, swing2LimitDegrees=10, twistLimitDegrees=10}
 {bone1Name=dyn_jacket_side_r_01, bone2Name=spine_02, swing1LimitDegrees=10, swing2LimitDegrees=10, twistLimitDegrees=10}
 {bone1Name=dyn_jacket_side_r_02, bone2Name=dyn_jacket_side_r_01, swing1LimitDegrees=10, swing2LimitDegrees=10, twistLimitDegrees=10}
 {bone1Name=dyn_jacket_side_r_03, bone2Name=dyn_jacket_side_r_02, swing1LimitDegrees=12, swing2LimitDegrees=32, twistLimitDegrees=8}
 {bone1Name=dyn_jacket_side_r_04, bone2Name=dyn_jacket_side_r_03, swing1LimitDegrees=12, swing2LimitDegrees=32, twistLimitDegrees=8}
 {bone1Name=dyn_jacket_side_r_05, bone2Name=dyn_jacket_side_r_04, swing1LimitDegrees=12, swing2LimitDegrees=12, twistLimitDegrees=8}
 {bone1Name=dyn_jacket_front_r_05, bone2Name=dyn_jacket_front_r_04, swing1LimitDegrees=8, swing2LimitDegrees=16, twistLimitDegrees=6}
 {bone1Name=dyn_jacket_front_r_04, bone2Name=dyn_jacket_front_r_03, swing1LimitDegrees=8, swing2LimitDegrees=16, twistLimitDegrees=6}
 {bone1Name=dyn_jacket_front_r_03, bone2Name=dyn_jacket_front_r_02, swing1LimitDegrees=12, swing2LimitDegrees=32, twistLimitDegrees=8}
 {
  bone1Name=dyn_jacket_front_r_03
  bone2Name=dyn_jacket_side_r_03
  swing1LimitDegrees=45
  swing1Motion=Free
  swing2LimitDegrees=45
  swing2Motion=Free
  twistLimitDegrees=45
  twistMotion=Free
 }
 {
  bone1Name=dyn_jacket_front_l_04
  bone2Name=dyn_jacket_side_l_04
  swing1LimitDegrees=45
  swing1Motion=Free
  swing2LimitDegrees=45
  swing2Motion=Free
  twistLimitDegrees=45
  twistMotion=Free
 }
 {
  bone1Name=dyn_jacket_front_r_04
  bone2Name=dyn_jacket_side_r_04
  swing1LimitDegrees=45
  swing1Motion=Free
  swing2LimitDegrees=45
  swing2Motion=Free
  twistLimitDegrees=45
  twistMotion=Free
 }
 {
  bone1Name=dyn_jacket_side_r_04
  bone2Name=dyn_jacket_rear_04
  swing1LimitDegrees=45
  swing1Motion=Free
  swing2LimitDegrees=45
  swing2Motion=Free
  twistLimitDegrees=45
  twistMotion=Free
 }
 {
  bone1Name=dyn_jacket_side_l_04
  bone2Name=dyn_jacket_rear_04
  swing1LimitDegrees=45
  swing1Motion=Free
  swing2LimitDegrees=45
  swing2Motion=Free
  twistLimitDegrees=45
  twistMotion=Free
 }
 {
  bone1Name=thigh_l
  bone2Name=pelvis
  swing1LimitDegrees=45
  swing1Motion=Free
  swing2LimitDegrees=45
  swing2Motion=Free
  twistLimitDegrees=45
  twistMotion=Free
 }
 {
  bone1Name=dyn_jacket_rear_05
  bone2Name=dyn_jacket_side_l_05
  swing1LimitDegrees=45
  swing1Motion=Free
  swing2LimitDegrees=45
  swing2Motion=Free
  twistLimitDegrees=45
  twistMotion=Free
 }
 {
  bone1Name=dyn_jacket_rear_05
  bone2Name=dyn_jacket_side_r_05
  swing1LimitDegrees=45
  swing1Motion=Free
  swing2LimitDegrees=45
  swing2Motion=Free
  twistLimitDegrees=45
  twistMotion=Free
 }
 {
  bone1Name=deform_gun
  bone2Name=pelvis
  swing1LimitDegrees=45
  swing1Motion=Free
  swing2LimitDegrees=45
  swing2Motion=Free
  twistLimitDegrees=45
  twistMotion=Free
 }
 {
  bone1Name=thigh_r
  bone2Name=pelvis
  swing1LimitDegrees=45
  swing1Motion=Free
  swing2LimitDegrees=45
  swing2Motion=Free
  twistLimitDegrees=45
  twistMotion=Free
 }
 {bone1Name=dyn_sword_bandolier_03, bone2Name=dyn_sword_bandolier_02, swing1LimitDegrees=12, swing2LimitDegrees=12, twistLimitDegrees=2}
 {bone1Name=dyn_sword_bandolier_01, bone2Name=dyn_sword_bandolier_01_MSTR, swing1LimitDegrees=12, swing2LimitDegrees=12, twistLimitDegrees=2}
 {bone1Name=dyn_sword_bandolier_02, bone2Name=dyn_sword_bandolier_01, swing1LimitDegrees=12, swing2LimitDegrees=12, twistLimitDegrees=2}
 {
  bone1Name=dyn_sword_bandolier_01_MSTR
  bone2Name=spine_03
  swing1LimitDegrees=45
  swing1Motion=Free
  swing2LimitDegrees=45
  swing2Motion=Free
  twistLimitDegrees=45
  twistMotion=Free
 }
 {bone1Name=dyn_sword, bone2Name=dyn_sword_bandolier_03, swing1LimitDegrees=4, swing2LimitDegrees=2, swing2Motion=Locked, twistLimitDegrees=1}
 {bone1Name=dyn_sword_bandolier_02, bone2Name=dyn_jacket_front_l_02, swing1LimitDegrees=10, swing2LimitDegrees=45, twistLimitDegrees=10}
 {
  bone1Name=dyn_jacket_front_r_05
  bone2Name=dyn_jacket_side_r_05
  swing1LimitDegrees=45
  swing1Motion=Free
  swing2LimitDegrees=45
  swing2Motion=Free
  twistLimitDegrees=45
  twistMotion=Free
 }
 {
  bone1Name=dyn_jacket_front_l_05
  bone2Name=dyn_jacket_side_l_05
  swing1LimitDegrees=45
  swing1Motion=Free
  swing2LimitDegrees=45
  swing2Motion=Free
  twistLimitDegrees=45
  twistMotion=Free
 }
 {
  bone1Name=spine_02
  bone2Name=pelvis
  swing1LimitDegrees=45
  swing1Motion=Free
  swing2LimitDegrees=45
  swing2Motion=Free
  twistLimitDegrees=45
  twistMotion=Free
 }
 {
  bone1Name=spine_03
  bone2Name=spine_02
  swing1LimitDegrees=45
  swing1Motion=Free
  swing2LimitDegrees=45
  swing2Motion=Free
  twistLimitDegrees=45
  twistMotion=Free
 }
 {
  bone1Name=dyn_sword_bandolier_front_mover
  bone2Name=dyn_sword_bandolier_01
  swing1LimitDegrees=2.5
  swing2LimitDegrees=2.5
  twistLimitDegrees=2.5
 }
 {bone1Name=dyn_sword_bandolier_rear_mover, bone2Name=spine_02, swing1LimitDegrees=2.5, swing2LimitDegrees=2.5, twistLimitDegrees=2.5}
]
