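# Post-copy animation Blueprint for the M_MED_DomeRoof male body cosmetic, adding secondary motion to clothing and accessories.
# It copies the attached parent's pose into a Control Rig fed by movement, skydiving and sliding states, then caches it.
# A pelvis-space RigidBody jacket simulation follows, then chained spring AnimDynamics for cloth, boots, lapels, collar and necklace, interleaved with additive bone offsets.
# The EventGraph passes noise and velocity-interpolation settings to material WPO functions.
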AnimBlueprint /BRCosmetics/Characters/Player/Male/Medium/Bodies/M_MED_DomeRoof/Meshes/M_MED_DomeRoof_AnimBP
target_skeleton=&/M_MED_DomeRoof_Skeleton
graphs [
 AnimGraph {nodes [
   n1 = RigidBody!(
    ComponentPose=Local To Component(LocalPose=Use cached pose 'PostDeformationRig'())
    Alpha
    OverrideWorldGravity:Vector="0,0,0"
    ExternalForce
    SimSpaceSettings:Sim Space Settings Structure=(WorldAlpha=1,VelocityScaleZ=0.5,DampingAlpha=1,MaxLinearVelocity=250,MaxAngularVelocity=100,MaxLinearAcceleration=250,MaxAngularAcceleration=100,ExternalLinearDragV=(X=0,Y=0,Z=0),ExternalLinearVelocity=(X=0,Y=0,Z=0),ExternalAngularVelocity=(X=0,Y=0,Z=0))
   ) -> (Pose) {
    bOverrideWorldGravity=true
    overridePhysicsAsset=&/M_MED_DomeRoof_Jacket_Physics
    simulationSpace=BaseBoneSpace
    alphaBoolBlend {blendOption=Linear}
    baseBoneRef {boneName=pelvis}
    componentAppliedLinearAccClamp=<1400 1400 1400>
    componentLinearAccScale=<0.5 0.5 0.5>
    componentLinearVelScale=<0.5 0.5 0.5>
    simSpaceSettings {maxAngularAcceleration=100, maxAngularVelocity=100, maxLinearAcceleration=250, maxLinearVelocity=250, velocityScaleZ=0.5, worldAlpha=1}
   }
   n2 = Anim Dynamics - Bone: dyn_cloth_front_l_01!(ComponentPose=@n1, bAlphaBoolEnabled:bool=True, ExternalForce="0,0,0") -> (Pose) {
    alphaInputType=Bool
    angularDampingOverride=1
    angularSpringConstant=15000
    bAngularSpring=true
    bOverrideAngularDamping=true
    bUseGravityOverride=true
    gravityScale=0.35
    simulationSpace=RootRelative
    alphaBoolBlend {blendOption=Linear}
    boundBone {boneName=dyn_cloth_front_l_01}
    componentLinearAccScale=<1 1 1>
    componentLinearVelScale=<1 1 1>
    gravityOverride {z=-980}
    physicsBodyDefinitions [{
      boundBone {boneName=dyn_cloth_front_l_01}
      boxExtents=<20 30 30>
      constraintSetup {angularLimitsMax=<15 0 15>, angularLimitsMin=<-15 -15 -15>, angularTarget=<1 0 0>}
      localJointOffset=<4 0 0>
     }]
   }
   n3 = Anim Dynamics - Bone: dyn_cloth_front_l_02!(ComponentPose=@n2, bAlphaBoolEnabled:bool=True, ExternalForce="0,0,0") -> (Pose) {
    alphaInputType=Bool
    angularDampingOverride=1
    angularSpringConstant=15000
    bAngularSpring=true
    bOverrideAngularDamping=true
    bUseGravityOverride=true
    gravityScale=0.35
    simulationSpace=RootRelative
    alphaBoolBlend {blendOption=Linear}
    boundBone {boneName=dyn_cloth_front_l_02}
    componentLinearAccScale=<1 1 1>
    componentLinearVelScale=<1 1 1>
    gravityOverride {z=-980}
    physicsBodyDefinitions [{
      boundBone {boneName=dyn_cloth_front_l_02}
      boxExtents=<20 30 30>
      constraintSetup {angularLimitsMax=<15 0 15>, angularLimitsMin=<-15 -15 -15>, angularTarget=<1 0 0>}
      localJointOffset=<4 0 0>
     }]
   }
   n4 = Anim Dynamics - Bone: dyn_cloth_front_l_03!(ComponentPose=@n3, bAlphaBoolEnabled:bool=True, ExternalForce="0,0,0") -> (Pose) {
    alphaInputType=Bool
    angularDampingOverride=1
    angularSpringConstant=15000
    bAngularSpring=true
    bOverrideAngularDamping=true
    bUseGravityOverride=true
    gravityScale=0.35
    simulationSpace=RootRelative
    alphaBoolBlend {blendOption=Linear}
    boundBone {boneName=dyn_cloth_front_l_03}
    componentLinearAccScale=<1 1 1>
    componentLinearVelScale=<1 1 1>
    gravityOverride {z=-980}
    physicsBodyDefinitions [{
      boundBone {boneName=dyn_cloth_front_l_03}
      boxExtents=<20 30 30>
      constraintSetup {angularLimitsMax=<30 30 30>, angularLimitsMin=<-30 -30 -30>, angularTarget=<1 0 0>}
      localJointOffset=<4 0 0>
     }]
   }
   n5 = "Transform (Modify) Bone - Bone: dyn_cloth_front_l_01"!(ComponentPose=@n4, bAlphaBoolEnabled:bool=True, Translation="-0.0002,0,1", Rotation="5,-0.016,-0.006", Scale="1,1,1") -> (Pose) {
    alphaInputType=Bool
    rotationMode=BMM_Additive
    rotationSpace=BCS_BoneSpace
    translationMode=BMM_Additive
    translationSpace=BCS_BoneSpace
    alphaBoolBlend {blendOption=Linear}
    boneToModify {boneName=dyn_cloth_front_l_01}
    rotation=<rot 5 -0.016 -0.006>
    translation {x=-0.0002, z=1}
   }
   n6 = "Transform (Modify) Bone - Bone: dyn_sword_bandolier_02"(ComponentPose=@n5, Alpha=1, Translation="0.006,0.262,1.874", Rotation="0.098,0.270,-0.730", Scale="1,1,1") {
    rotationMode=BMM_Additive
    rotationSpace=BCS_BoneSpace
    scaleSpace=BCS_BoneSpace
    translationMode=BMM_Additive
    translationSpace=BCS_BoneSpace
    alphaBoolBlend {blendOption=Linear}
    boneToModify {boneName=dyn_sword_bandolier_02}
    rotation=<rot 0.098 0.27 -0.73>
    translation=<0.006 0.262 1.874>
   }
   n7 = "Transform (Modify) Bone - Bone: dyn_jacket_front_r_01_MSTR"(ComponentPose=@n6, Alpha=1, Translation="0,0,-0.5", Rotation="2,0,0", Scale="1,1,1") {
    rotationMode=BMM_Additive
    rotationSpace=BCS_BoneSpace
    translationMode=BMM_Additive
    translationSpace=BCS_BoneSpace
    alphaBoolBlend {blendOption=Linear}
    boneToModify {boneName=dyn_jacket_front_r_01_MSTR}
    rotation {pitch=2}
    translation {z=-0.5}
   }
   n8 = "Transform (Modify) Bone - Bone: dyn_jacket_side_r_03"(ComponentPose=@n7, Alpha=1, Translation="0,-1,-1", Rotation="0,0,0", Scale="1,1,1") {
    rotationMode=BMM_Additive
    rotationSpace=BCS_BoneSpace
    translationMode=BMM_Additive
    translationSpace=BCS_BoneSpace
    alphaBoolBlend {blendOption=Linear}
    boneToModify {boneName=dyn_jacket_side_r_03}
    translation {y=-1, z=-1}
   }
   n9 = "Transform (Modify) Bone - Bone: dyn_jacket_front_l_01"(ComponentPose=@n8, Alpha=1, Translation="-0.056,0.755,0", Rotation="-1.312,-0.095,4.033", Scale="1,0.95,1") {
    rotationMode=BMM_Additive
    rotationSpace=BCS_BoneSpace
    scaleMode=BMM_Additive
    scaleSpace=BCS_BoneSpace
    translationMode=BMM_Additive
    translationSpace=BCS_BoneSpace
    alphaBoolBlend {blendOption=Linear}
    boneToModify {boneName=dyn_jacket_front_l_01}
    rotation=<rot -1.312 -0.095 4.033>
    scale {y=0.95}
    translation=<-0.056 0.755>
   }
   n10 = "Transform (Modify) Bone - Bone: dyn_jacket_front_l_02"(ComponentPose=@n9, Alpha=1, Translation="-0.265,-3.554,0.841", Rotation="1.086,-0.016,-0.008", Scale="1,1,1") {
    rotationMode=BMM_Additive
    rotationSpace=BCS_BoneSpace
    translationMode=BMM_Additive
    translationSpace=BCS_BoneSpace
    alphaBoolBlend {blendOption=Linear}
    boneToModify {boneName=dyn_jacket_front_l_02}
    rotation=<rot 1.086 -0.016 -0.008>
    translation=<-0.265 -3.554 0.841>
   }
   n11 = "Transform (Modify) Bone - Bone: dyn_jacket_front_l_03"(ComponentPose=@n10, Alpha=1, Translation="0,0,0", Rotation="0,0,0", Scale="1,1,1") {
    rotationMode=BMM_Additive
    rotationSpace=BCS_BoneSpace
    translationMode=BMM_Additive
    translationSpace=BCS_BoneSpace
    alphaBoolBlend {blendOption=Linear}
    boneToModify {boneName=dyn_jacket_front_l_03}
   }
   n12 = "Transform (Modify) Bone - Bone: dyn_jacket_front_l_04"(ComponentPose=@n11, Alpha=1, Translation="0,0,0", Rotation="0,0,0", Scale="1,1,1") {
    rotationMode=BMM_Additive
    rotationSpace=BCS_BoneSpace
    translationMode=BMM_Additive
    translationSpace=BCS_BoneSpace
    alphaBoolBlend {blendOption=Linear}
    boneToModify {boneName=dyn_jacket_front_l_04}
   }
   n13 = "Transform (Modify) Bone - Bone: dyn_jacket_front_l_05"(ComponentPose=@n12, Alpha=1, Translation="0,0,0", Rotation="0,0,0", Scale="1,1,1") {
    rotationMode=BMM_Additive
    rotationSpace=BCS_BoneSpace
    translationMode=BMM_Additive
    translationSpace=BCS_BoneSpace
    alphaBoolBlend {blendOption=Linear}
    boneToModify {boneName=dyn_jacket_front_l_05}
   }
   n14 = "Transform (Modify) Bone - Bone: dyn_jacket_side_l_03"(ComponentPose=@n13, Alpha=1, Translation="0,1,1", Rotation="0,0,0", Scale="1,1,1") {
    rotationMode=BMM_Additive
    rotationSpace=BCS_BoneSpace
    translationMode=BMM_Additive
    translationSpace=BCS_BoneSpace
    alphaBoolBlend {blendOption=Linear}
    boneToModify {boneName=dyn_jacket_side_l_03}
    translation {y=1, z=1}
   }
   n15 = "Transform (Modify) Bone - Bone: dyn_sword_bandolier_rear_mover_MSTR"(ComponentPose=@n14, Alpha=1, Translation="0.2041,-1.5,0", Rotation="0,0,0", Scale="1,1,1") {
    rotationMode=BMM_Additive
    rotationSpace=BCS_BoneSpace
    translationMode=BMM_Additive
    translationSpace=BCS_BoneSpace
    alphaBoolBlend {blendOption=Linear}
    boneToModify {boneName=dyn_sword_bandolier_rear_mover_MSTR}
    translation=<0.2041 -1.5>
   }
   n16 = "Transform (Modify) Bone - Bone: dyn_sword_bandolier_front_mover_MSTR"(ComponentPose=@n15, Alpha=1, Translation="0,1,0", Rotation="0,0,0", Scale="1,1,1") {
    rotationMode=BMM_Additive
    rotationSpace=BCS_BoneSpace
    translationMode=BMM_Additive
    translationSpace=BCS_BoneSpace
    alphaBoolBlend {blendOption=Linear}
    boneToModify {boneName=dyn_sword_bandolier_front_mover_MSTR}
    translation {y=1}
   }
   n17 = Anim Dynamics - Bone: dyn_boot_l_lip_01!(SimulationSpace:AnimPhysSimSpaceType Enum, ComponentPose=@n16, Alpha=1, ExternalForce="0,0,0") -> (Pose) {
    angularDampingOverride=1
    angularSpringConstant=20000
    bAngularSpring=true
    bOverrideAngularDamping=true
    bUseGravityOverride=true
    gravityScale=0.35
    numSolverIterationsPreUpdate=1
    simulationSpace=RootRelative
    alphaBoolBlend {blendOption=Linear}
    boundBone {boneName=dyn_boot_l_lip_01}
    componentLinearAccScale=<1 1 1>
    componentLinearVelScale=<1 1 1>
    gravityOverride {z=-980}
    physicsBodyDefinitions [{
      boundBone {boneName=dyn_boot_l_lip_01}
      boxExtents=<20 30 30>
      constraintSetup {angularLimitsMax=<15 15 15>, angularLimitsMin=<-15 -15 -15>, angularTarget=<1 0 0>}
      localJointOffset=<4 0 0>
     }]
   }
   n18 = Anim Dynamics - Bone: dyn_boot_l_lip_02!(SimulationSpace:AnimPhysSimSpaceType Enum, ComponentPose=@n17, Alpha=1, ExternalForce="0,0,0") -> (Pose) {
    angularDampingOverride=1
    angularSpringConstant=20000
    bAngularSpring=true
    bOverrideAngularDamping=true
    bUseGravityOverride=true
    gravityScale=0.35
    numSolverIterationsPreUpdate=1
    simulationSpace=RootRelative
    alphaBoolBlend {blendOption=Linear}
    boundBone {boneName=dyn_boot_l_lip_02}
    componentLinearAccScale=<1 1 1>
    componentLinearVelScale=<1 1 1>
    gravityOverride {z=-980}
    physicsBodyDefinitions [{
      boundBone {boneName=dyn_boot_l_lip_02}
      boxExtents=<20 30 30>
      constraintSetup {angularLimitsMax=<15 15 15>, angularLimitsMin=<-15 -15 -15>, angularTarget=<1 0 0>}
      localJointOffset=<4 0 0>
     }]
   }
   n19 = Anim Dynamics - Bone: dyn_boot_l_lip_03!(SimulationSpace:AnimPhysSimSpaceType Enum, ComponentPose=@n18, Alpha=1, ExternalForce="0,0,0") -> (Pose) {
    angularDampingOverride=1
    angularSpringConstant=20000
    bAngularSpring=true
    bOverrideAngularDamping=true
    bUseGravityOverride=true
    gravityScale=0.35
    numSolverIterationsPreUpdate=1
    simulationSpace=RootRelative
    alphaBoolBlend {blendOption=Linear}
    boundBone {boneName=dyn_boot_l_lip_03}
    componentLinearAccScale=<1 1 1>
    componentLinearVelScale=<1 1 1>
    gravityOverride {z=-980}
    physicsBodyDefinitions [{
      boundBone {boneName=dyn_boot_l_lip_03}
      boxExtents=<20 30 30>
      constraintSetup {angularLimitsMax=<15 15 15>, angularLimitsMin=<-15 -15 -15>, angularTarget=<1 0 0>}
      localJointOffset=<4 0 0>
     }]
   }
   n20 = Anim Dynamics - Bone: dyn_boot_l_lip_04!(SimulationSpace:AnimPhysSimSpaceType Enum, ComponentPose=@n19, Alpha=1, ExternalForce="0,0,0") -> (Pose) {
    angularDampingOverride=1
    angularSpringConstant=20000
    bAngularSpring=true
    bOverrideAngularDamping=true
    bUseGravityOverride=true
    gravityScale=0.35
    numSolverIterationsPreUpdate=1
    simulationSpace=RootRelative
    alphaBoolBlend {blendOption=Linear}
    boundBone {boneName=dyn_boot_l_lip_04}
    componentLinearAccScale=<1 1 1>
    componentLinearVelScale=<1 1 1>
    gravityOverride {z=-980}
    physicsBodyDefinitions [{
      boundBone {boneName=dyn_boot_l_lip_04}
      boxExtents=<20 30 30>
      constraintSetup {angularLimitsMax=<15 15 15>, angularLimitsMin=<-15 -15 -15>, angularTarget=<1 0 0>}
      localJointOffset=<4 0 0>
     }]
   }
   n21 = Anim Dynamics - Bone: dyn_boot_r_lip_01!(SimulationSpace:AnimPhysSimSpaceType Enum, ComponentPose=@n20, Alpha=1, ExternalForce="0,0,0") -> (Pose) {
    angularDampingOverride=1
    angularSpringConstant=20000
    bAngularSpring=true
    bOverrideAngularDamping=true
    bUseGravityOverride=true
    gravityScale=0.35
    numSolverIterationsPreUpdate=1
    simulationSpace=RootRelative
    alphaBoolBlend {blendOption=Linear}
    boundBone {boneName=dyn_boot_r_lip_01}
    componentLinearAccScale=<1 1 1>
    componentLinearVelScale=<1 1 1>
    gravityOverride {z=-980}
    physicsBodyDefinitions [{
      boundBone {boneName=dyn_boot_r_lip_01}
      boxExtents=<20 30 30>
      constraintSetup {angularLimitsMax=<15 15 15>, angularLimitsMin=<-15 -15 -15>, angularTarget=<1 0 0>}
      localJointOffset=<-4 0 0>
     }]
   }
   n22 = Anim Dynamics - Bone: dyn_boot_r_lip_02!(SimulationSpace:AnimPhysSimSpaceType Enum, ComponentPose=@n21, Alpha=1, ExternalForce="0,0,0") -> (Pose) {
    angularDampingOverride=1
    angularSpringConstant=20000
    bAngularSpring=true
    bOverrideAngularDamping=true
    bUseGravityOverride=true
    gravityScale=0.35
    numSolverIterationsPreUpdate=1
    simulationSpace=RootRelative
    alphaBoolBlend {blendOption=Linear}
    boundBone {boneName=dyn_boot_r_lip_02}
    componentLinearAccScale=<1 1 1>
    componentLinearVelScale=<1 1 1>
    gravityOverride {z=-980}
    physicsBodyDefinitions [{
      boundBone {boneName=dyn_boot_r_lip_02}
      boxExtents=<20 30 30>
      constraintSetup {angularLimitsMax=<15 15 15>, angularLimitsMin=<-15 -15 -15>, angularTarget=<1 0 0>}
      localJointOffset=<-4 0 0>
     }]
   }
   n23 = Anim Dynamics - Bone: dyn_boot_r_lip_03!(SimulationSpace:AnimPhysSimSpaceType Enum, ComponentPose=@n22, Alpha=1, ExternalForce="0,0,0") -> (Pose) {
    angularDampingOverride=1
    angularSpringConstant=20000
    bAngularSpring=true
    bOverrideAngularDamping=true
    bUseGravityOverride=true
    gravityScale=0.35
    numSolverIterationsPreUpdate=1
    simulationSpace=RootRelative
    alphaBoolBlend {blendOption=Linear}
    boundBone {boneName=dyn_boot_r_lip_03}
    componentLinearAccScale=<1 1 1>
    componentLinearVelScale=<1 1 1>
    gravityOverride {z=-980}
    physicsBodyDefinitions [{
      boundBone {boneName=dyn_boot_r_lip_03}
      boxExtents=<20 30 30>
      constraintSetup {angularLimitsMax=<15 15 15>, angularLimitsMin=<-15 -15 -15>, angularTarget=<1 0 0>}
      localJointOffset=<-4 0 0>
     }]
   }
   n24 = Anim Dynamics - Bone: dyn_boot_r_lip_04!(SimulationSpace:AnimPhysSimSpaceType Enum, ComponentPose=@n23, Alpha=1, ExternalForce="0,0,0") -> (Pose) {
    angularDampingOverride=1
    angularSpringConstant=20000
    bAngularSpring=true
    bOverrideAngularDamping=true
    bUseGravityOverride=true
    gravityScale=0.35
    numSolverIterationsPreUpdate=1
    simulationSpace=RootRelative
    alphaBoolBlend {blendOption=Linear}
    boundBone {boneName=dyn_boot_r_lip_04}
    componentLinearAccScale=<1 1 1>
    componentLinearVelScale=<1 1 1>
    gravityOverride {z=-980}
    physicsBodyDefinitions [{
      boundBone {boneName=dyn_boot_r_lip_04}
      boxExtents=<20 30 30>
      constraintSetup {angularLimitsMax=<15 15 15>, angularLimitsMin=<-15 -15 -15>, angularTarget=<1 0 0>}
      localJointOffset=<-4 0 0>
     }]
   }
   n25 = Anim Dynamics - Bone: dyn_boot_r_lace_a_01!(SimulationSpace:AnimPhysSimSpaceType Enum, ComponentPose=@n24, Alpha=1, ExternalForce="0,0,0") -> (Pose) {
    angularDampingOverride=1
    angularSpringConstant=20000
    bAngularSpring=true
    bOverrideAngularDamping=true
    bUseGravityOverride=true
    gravityScale=0.35
    numSolverIterationsPreUpdate=1
    simulationSpace=RootRelative
    alphaBoolBlend {blendOption=Linear}
    boundBone {boneName=dyn_boot_r_lace_a_01}
    componentLinearAccScale=<1 1 1>
    componentLinearVelScale=<1 1 1>
    gravityOverride {z=-980}
    physicsBodyDefinitions [{
      boundBone {boneName=dyn_boot_r_lace_a_01}
      boxExtents=<20 30 30>
      constraintSetup {angularLimitsMax=<15 15 15>, angularLimitsMin=<-15 -15 -15>, angularTarget=<1 0 0>}
      localJointOffset=<4 0 0>
     }]
   }
   n26 = Anim Dynamics - Bone: dyn_boot_r_lace_a_02!(SimulationSpace:AnimPhysSimSpaceType Enum, ComponentPose=@n25, Alpha=1, ExternalForce="0,0,0") -> (Pose) {
    angularDampingOverride=1
    angularSpringConstant=20000
    bAngularSpring=true
    bOverrideAngularDamping=true
    bUseGravityOverride=true
    gravityScale=0.35
    numSolverIterationsPreUpdate=1
    simulationSpace=RootRelative
    alphaBoolBlend {blendOption=Linear}
    boundBone {boneName=dyn_boot_r_lace_a_02}
    componentLinearAccScale=<1 1 1>
    componentLinearVelScale=<1 1 1>
    gravityOverride {z=-980}
    physicsBodyDefinitions [{
      boundBone {boneName=dyn_boot_r_lace_a_02}
      boxExtents=<20 30 30>
      constraintSetup {angularLimitsMax=<15 15 15>, angularLimitsMin=<-15 -15 -15>, angularTarget=<1 0 0>}
      localJointOffset=<4 0 0>
     }]
   }
   n27 = Anim Dynamics - Bone: dyn_boot_r_lace_b_01!(SimulationSpace:AnimPhysSimSpaceType Enum, ComponentPose=@n26, Alpha=1, ExternalForce="0,0,0") -> (Pose) {
    angularDampingOverride=1
    angularSpringConstant=20000
    bAngularSpring=true
    bOverrideAngularDamping=true
    bUseGravityOverride=true
    gravityScale=0.35
    numSolverIterationsPreUpdate=1
    simulationSpace=RootRelative
    alphaBoolBlend {blendOption=Linear}
    boundBone {boneName=dyn_boot_r_lace_b_01}
    componentLinearAccScale=<1 1 1>
    componentLinearVelScale=<1 1 1>
    gravityOverride {z=-980}
    physicsBodyDefinitions [{
      boundBone {boneName=dyn_boot_r_lace_b_01}
      boxExtents=<20 30 30>
      constraintSetup {angularLimitsMax=<15 15 15>, angularLimitsMin=<-15 -15 -15>, angularTarget=<1 0 0>}
      localJointOffset=<4 0 0>
     }]
   }
   n28 = Anim Dynamics - Bone: dyn_boot_r_lace_b_02!(SimulationSpace:AnimPhysSimSpaceType Enum, ComponentPose=@n27, Alpha=1, ExternalForce="0,0,0") -> (Pose) {
    angularDampingOverride=1
    angularSpringConstant=20000
    bAngularSpring=true
    bOverrideAngularDamping=true
    bUseGravityOverride=true
    gravityScale=0.35
    numSolverIterationsPreUpdate=1
    simulationSpace=RootRelative
    alphaBoolBlend {blendOption=Linear}
    boundBone {boneName=dyn_boot_r_lace_b_02}
    componentLinearAccScale=<1 1 1>
    componentLinearVelScale=<1 1 1>
    gravityOverride {z=-980}
    physicsBodyDefinitions [{
      boundBone {boneName=dyn_boot_r_lace_b_02}
      boxExtents=<20 30 30>
      constraintSetup {angularLimitsMax=<15 15 15>, angularLimitsMin=<-15 -15 -15>, angularTarget=<1 0 0>}
      localJointOffset=<4 0 0>
     }]
   }
   n29 = Anim Dynamics - Bone: dyn_boot_l_lace_a_01!(SimulationSpace:AnimPhysSimSpaceType Enum, ComponentPose=@n28, Alpha=1, ExternalForce="0,0,0") -> (Pose) {
    angularDampingOverride=1
    angularSpringConstant=20000
    bAngularSpring=true
    bOverrideAngularDamping=true
    bUseGravityOverride=true
    gravityScale=0.35
    numSolverIterationsPreUpdate=1
    simulationSpace=RootRelative
    alphaBoolBlend {blendOption=Linear}
    boundBone {boneName=dyn_boot_l_lace_a_01}
    componentLinearAccScale=<1 1 1>
    componentLinearVelScale=<1 1 1>
    gravityOverride {z=-980}
    physicsBodyDefinitions [{
      boundBone {boneName=dyn_boot_l_lace_a_01}
      boxExtents=<20 30 30>
      constraintSetup {angularLimitsMax=<15 15 15>, angularLimitsMin=<-15 -15 -15>, angularTarget=<1 0 0>}
      localJointOffset=<4 0 0>
     }]
   }
   n30 = Anim Dynamics - Bone: dyn_boot_l_lace_a_02!(SimulationSpace:AnimPhysSimSpaceType Enum, ComponentPose=@n29, Alpha=1, ExternalForce="0,0,0") -> (Pose) {
    angularDampingOverride=1
    angularSpringConstant=20000
    bAngularSpring=true
    bOverrideAngularDamping=true
    bUseGravityOverride=true
    gravityScale=0.35
    numSolverIterationsPreUpdate=1
    simulationSpace=RootRelative
    alphaBoolBlend {blendOption=Linear}
    boundBone {boneName=dyn_boot_l_lace_a_02}
    componentLinearAccScale=<1 1 1>
    componentLinearVelScale=<1 1 1>
    gravityOverride {z=-980}
    physicsBodyDefinitions [{
      boundBone {boneName=dyn_boot_l_lace_a_02}
      boxExtents=<20 30 30>
      constraintSetup {angularLimitsMax=<15 15 15>, angularLimitsMin=<-15 -15 -15>, angularTarget=<1 0 0>}
      localJointOffset=<4 0 0>
     }]
   }
   n31 = Anim Dynamics - Bone: dyn_boot_l_lace_b_01!(SimulationSpace:AnimPhysSimSpaceType Enum, ComponentPose=@n30, Alpha=1, ExternalForce="0,0,0") -> (Pose) {
    angularDampingOverride=1
    angularSpringConstant=20000
    bAngularSpring=true
    bOverrideAngularDamping=true
    bUseGravityOverride=true
    gravityScale=0.35
    numSolverIterationsPreUpdate=1
    simulationSpace=RootRelative
    alphaBoolBlend {blendOption=Linear}
    boundBone {boneName=dyn_boot_l_lace_b_01}
    componentLinearAccScale=<1 1 1>
    componentLinearVelScale=<1 1 1>
    gravityOverride {z=-980}
    physicsBodyDefinitions [{
      boundBone {boneName=dyn_boot_l_lace_b_01}
      boxExtents=<20 30 30>
      constraintSetup {angularLimitsMax=<15 15 15>, angularLimitsMin=<-15 -15 -15>, angularTarget=<1 0 0>}
      localJointOffset=<4 0 0>
     }]
   }
   n32 = Anim Dynamics - Bone: dyn_boot_l_lace_b_02!(SimulationSpace:AnimPhysSimSpaceType Enum, ComponentPose=@n31, Alpha=1, ExternalForce="0,0,0") -> (Pose) {
    angularDampingOverride=1
    angularSpringConstant=20000
    bAngularSpring=true
    bOverrideAngularDamping=true
    bUseGravityOverride=true
    gravityScale=0.35
    numSolverIterationsPreUpdate=1
    simulationSpace=RootRelative
    alphaBoolBlend {blendOption=Linear}
    boundBone {boneName=dyn_boot_l_lace_b_02}
    componentLinearAccScale=<1 1 1>
    componentLinearVelScale=<1 1 1>
    gravityOverride {z=-980}
    physicsBodyDefinitions [{
      boundBone {boneName=dyn_boot_l_lace_b_02}
      boxExtents=<20 30 30>
      constraintSetup {angularLimitsMax=<15 15 15>, angularLimitsMin=<-15 -15 -15>, angularTarget=<1 0 0>}
      localJointOffset=<4 0 0>
     }]
   }
   n33 = Anim Dynamics - Bone: dyn_lapel_l_01(ComponentPose=@n32, Alpha=1, ExternalForce="0,0,0") {
    angularDampingOverride=1
    angularSpringConstant=20000
    bAngularSpring=true
    bOverrideAngularDamping=true
    bUseGravityOverride=true
    gravityScale=0.35
    numSolverIterationsPreUpdate=1
    simulationSpace=RootRelative
    alphaBoolBlend {blendOption=Linear}
    boundBone {boneName=dyn_lapel_l_01}
    componentLinearAccScale=<1 1 1>
    componentLinearVelScale=<1 1 1>
    gravityOverride {z=-980}
    physicsBodyDefinitions [{
      boundBone {boneName=dyn_lapel_l_01}
      boxExtents=<35 35 35>
      constraintSetup {angularLimitsMax=<10 10 10>, angularLimitsMin=<-10 -10 -10>, angularTarget=<1 0 0>}
      localJointOffset=<4 0 0>
     }]
   }
   n34 = Anim Dynamics - Bone: dyn_lapel_l_02(ComponentPose=@n33, Alpha=1, ExternalForce="0,0,0") {
    angularDampingOverride=1
    angularSpringConstant=20000
    bAngularSpring=true
    bOverrideAngularDamping=true
    bUseGravityOverride=true
    gravityScale=0.35
    numSolverIterationsPreUpdate=1
    simulationSpace=RootRelative
    alphaBoolBlend {blendOption=Linear}
    boundBone {boneName=dyn_lapel_l_02}
    componentLinearAccScale=<1 1 1>
    componentLinearVelScale=<1 1 1>
    gravityOverride {z=-980}
    physicsBodyDefinitions [{
      boundBone {boneName=dyn_lapel_l_02}
      boxExtents=<35 35 35>
      constraintSetup {angularLimitsMax=<10 10 10>, angularLimitsMin=<-10 -10 -10>, angularTarget=<1 0 0>}
      localJointOffset=<4 0 0>
     }]
   }
   n35 = Anim Dynamics - Bone: dyn_lapel_r_01(ComponentPose=@n34, Alpha=1, ExternalForce="0,0,0") {
    angularDampingOverride=1
    angularSpringConstant=20000
    bAngularSpring=true
    bOverrideAngularDamping=true
    bUseGravityOverride=true
    gravityScale=0.35
    numSolverIterationsPreUpdate=1
    simulationSpace=RootRelative
    alphaBoolBlend {blendOption=Linear}
    boundBone {boneName=dyn_lapel_r_01}
    componentLinearAccScale=<1 1 1>
    componentLinearVelScale=<1 1 1>
    gravityOverride {z=-980}
    physicsBodyDefinitions [{
      boundBone {boneName=dyn_lapel_r_01}
      boxExtents=<35 35 35>
      constraintSetup {angularLimitsMax=<10 10 10>, angularLimitsMin=<-10 -10 -10>, angularTarget=<1 0 0>}
      localJointOffset=<-4 0 0>
     }]
   }
   n36 = Anim Dynamics - Bone: dyn_lapel_r_02(ComponentPose=@n35, Alpha=1, ExternalForce="0,0,0") {
    angularDampingOverride=1
    angularSpringConstant=20000
    bAngularSpring=true
    bOverrideAngularDamping=true
    bUseGravityOverride=true
    gravityScale=0.35
    numSolverIterationsPreUpdate=1
    simulationSpace=RootRelative
    alphaBoolBlend {blendOption=Linear}
    boundBone {boneName=dyn_lapel_r_02}
    componentLinearAccScale=<1 1 1>
    componentLinearVelScale=<1 1 1>
    gravityOverride {z=-980}
    physicsBodyDefinitions [{
      boundBone {boneName=dyn_lapel_r_02}
      boxExtents=<35 35 35>
      constraintSetup {angularLimitsMax=<10 10 10>, angularLimitsMin=<-10 -10 -10>, angularTarget=<1 0 0>}
      localJointOffset=<-4 0 0>
     }]
   }
   n37 = Anim Dynamics - Bone: dyn_shirt_collar_l_01(ComponentPose=@n36, Alpha=1, ExternalForce="0,0,0") {
    angularDampingOverride=1
    angularSpringConstant=20000
    bAngularSpring=true
    bOverrideAngularDamping=true
    bUseGravityOverride=true
    gravityScale=0.35
    numSolverIterationsPreUpdate=1
    simulationSpace=RootRelative
    alphaBoolBlend {blendOption=Linear}
    boundBone {boneName=dyn_shirt_collar_l_01}
    componentLinearAccScale=<1 1 1>
    componentLinearVelScale=<1 1 1>
    gravityOverride {z=-980}
    physicsBodyDefinitions [{
      boundBone {boneName=dyn_shirt_collar_l_01}
      boxExtents=<20 30 30>
      constraintSetup {angularLimitsMax=<5 5 5>, angularLimitsMin=<-5 -5 -5>, angularTarget=<1 0 0>}
      localJointOffset=<4 0 0>
     }]
    planarLimits [{drivingBone {boneName=clavicle_l}, planeTransform {location=<2.2366 4.8687 5.0651>, rotation=<rot -6.6861 -0.8151 27.5835>}}]
   }
   n38 = Anim Dynamics - Bone: dyn_shirt_collar_l_02(ComponentPose=@n37, Alpha=1, ExternalForce="0,0,0") {
    angularDampingOverride=1
    angularSpringConstant=20000
    bAngularSpring=true
    bOverrideAngularDamping=true
    bUseGravityOverride=true
    gravityScale=0.35
    numSolverIterationsPreUpdate=1
    simulationSpace=RootRelative
    alphaBoolBlend {blendOption=Linear}
    boundBone {boneName=dyn_shirt_collar_l_02}
    componentLinearAccScale=<1 1 1>
    componentLinearVelScale=<1 1 1>
    gravityOverride {z=-980}
    physicsBodyDefinitions [{
      boundBone {boneName=dyn_shirt_collar_l_02}
      boxExtents=<20 30 30>
      constraintSetup {angularLimitsMax=<5 5 5>, angularLimitsMin=<-5 -5 -5>, angularTarget=<1 0 0>}
      localJointOffset=<4 0 0>
     }]
    planarLimits [{drivingBone {boneName=clavicle_l}, planeTransform {location=<6.4395 5.1132 4.5922>, rotation=<rot 0.7012 8.3308 35.4036>}}]
   }
   n39 = Anim Dynamics - Bone: dyn_shirt_collar_r_01(ComponentPose=@n38, Alpha=1, ExternalForce="0,0,0") {
    angularDampingOverride=1
    angularSpringConstant=20000
    bAngularSpring=true
    bOverrideAngularDamping=true
    bUseGravityOverride=true
    bUsePlanarLimit=false
    gravityScale=0.35
    numSolverIterationsPreUpdate=1
    simulationSpace=RootRelative
    alphaBoolBlend {blendOption=Linear}
    boundBone {boneName=dyn_shirt_collar_r_01}
    componentLinearAccScale=<1 1 1>
    componentLinearVelScale=<1 1 1>
    gravityOverride {z=-980}
    physicsBodyDefinitions [{
      boundBone {boneName=dyn_shirt_collar_r_01}
      boxExtents=<20 30 30>
      constraintSetup {angularLimitsMax=<5 5 5>, angularLimitsMin=<0 0 0>, angularTarget=<1 0 0>}
      localJointOffset=<-4 0 0>
     }]
    planarLimits [{drivingBone {boneName=clavicle_r}, planeTransform {location=<-2.2366 4.8687 5.0651>, rotation=<rot -6.6861 -0.8151 27.5835>}}]
   }
   n40 = Anim Dynamics - Bone: dyn_shirt_collar_r_02(ComponentPose=@n39, Alpha=1, ExternalForce="0,0,0") {
    angularDampingOverride=1
    angularSpringConstant=20000
    bAngularSpring=true
    bOverrideAngularDamping=true
    bUseGravityOverride=true
    gravityScale=0.35
    numSolverIterationsPreUpdate=1
    simulationSpace=RootRelative
    alphaBoolBlend {blendOption=Linear}
    boundBone {boneName=dyn_shirt_collar_r_02}
    componentLinearAccScale=<1 1 1>
    componentLinearVelScale=<1 1 1>
    gravityOverride {z=-980}
    physicsBodyDefinitions [{
      boundBone {boneName=dyn_shirt_collar_r_02}
      boxExtents=<20 30 30>
      constraintSetup {angularLimitsMax=<5 5 5>, angularLimitsMin=<-5 -5 -5>, angularTarget=<1 0 0>}
      localJointOffset=<-4 0 0>
     }]
    planarLimits [{drivingBone {boneName=clavicle_r}, planeTransform {location=<-8.1209 -1.1167 -1.9325>, rotation=<rot 16.1833 177.0925 145.7715>}}]
   }
   n41 = "Transform (Modify) Bone - Bone: dyn_shirt_collar_r_01"(ComponentPose=@n40, Alpha=1, Translation="0,0,0", Rotation="0,8,0", Scale="1,1,1") {
    rotationMode=BMM_Additive
    rotationSpace=BCS_BoneSpace
    translationMode=BMM_Additive
    translationSpace=BCS_BoneSpace
    alphaBoolBlend {blendOption=Linear}
    boneToModify {boneName=dyn_shirt_collar_r_01}
    rotation {yaw=8}
   }
   n42 = "Transform (Modify) Bone - Bone: dyn_shirt_collar_l_01"(ComponentPose=@n41, Alpha=1, Translation="0,0,0", Rotation="0,8,0", Scale="1,1,1") {
    rotationMode=BMM_Additive
    rotationSpace=BCS_BoneSpace
    translationMode=BMM_Additive
    translationSpace=BCS_BoneSpace
    alphaBoolBlend {blendOption=Linear}
    boneToModify {boneName=dyn_shirt_collar_l_01}
    rotation {yaw=8}
   }
   n43 = Anim Dynamics - Bone: dyn_thigh_l_cloth_bounce_01(ComponentPose=@n42, Alpha=1, ExternalForce="0,0,0") {
    angularDampingOverride=0.99
    angularSpringConstant=15000
    bAngularSpring=true
    bLinearSpring=true
    bOverrideAngularDamping=true
    bUseGravityOverride=true
    gravityScale=0.35
    linearSpringConstant=0.5
    numSolverIterationsPreUpdate=1
    simulationSpace=RootRelative
    alphaBoolBlend {blendOption=Linear}
    boundBone {boneName=dyn_thigh_l_cloth_bounce_01}
    componentAppliedLinearAccClamp=<25000 25000 25000>
    componentLinearAccScale=<1 1 1>
    componentLinearVelScale=<1 1 1>
    gravityOverride {z=-1800}
    physicsBodyDefinitions [{
      boundBone {boneName=dyn_thigh_l_cloth_bounce_01}
      boxExtents=<20 40 40>
      constraintSetup {
       angularLimitsMax=<10 10 10>
       angularLimitsMin=<-10 -10 -10>
       angularTarget=<1 0 0>
       linearAxesMax=<0.5 0.5 0.5>
       linearAxesMin=<-0.5 -0.5 -0.5>
      }
      localJointOffset=<-12 0 0>
     }]
   }
   n44 = Anim Dynamics - Bone: dyn_thigh_l_cloth_bounce_02(ComponentPose=@n43, Alpha=1, ExternalForce="0,0,0") {
    angularDampingOverride=0.99
    angularSpringConstant=15000
    bAngularSpring=true
    bLinearSpring=true
    bOverrideAngularDamping=true
    bOverrideLinearDamping=true
    bUseGravityOverride=true
    gravityScale=0.35
    linearDampingOverride=0.95
    linearSpringConstant=0.5
    numSolverIterationsPreUpdate=1
    simulationSpace=RootRelative
    alphaBoolBlend {blendOption=Linear}
    boundBone {boneName=dyn_thigh_l_cloth_bounce_02}
    componentAppliedLinearAccClamp=<25000 25000 25000>
    componentLinearAccScale=<1 1 1>
    componentLinearVelScale=<1 1 1>
    gravityOverride {z=-1800}
    physicsBodyDefinitions [{
      boundBone {boneName=dyn_thigh_l_cloth_bounce_02}
      boxExtents=<30 40 40>
      constraintSetup {angularLimitsMax=<10 10 10>, angularLimitsMin=<-10 -10 -10>, angularTarget=<1 0 0>, linearAxesMax=<3 3 3>, linearAxesMin=<-3 -3 -3>}
      localJointOffset=<-4 0 0>
     }]
   }
   n45 = Anim Dynamics - Bone: dyn_thigh_l_cloth_bounce_03(ComponentPose=@n44, Alpha=1, ExternalForce="0,0,0") {
    angularDampingOverride=0.99
    angularSpringConstant=15000
    bAngularSpring=true
    bLinearSpring=true
    bOverrideAngularDamping=true
    bOverrideLinearDamping=true
    bUseGravityOverride=true
    gravityScale=0.35
    linearDampingOverride=0.95
    linearSpringConstant=0.5
    numSolverIterationsPreUpdate=1
    simulationSpace=RootRelative
    alphaBoolBlend {blendOption=Linear}
    boundBone {boneName=dyn_thigh_l_cloth_bounce_03}
    componentAppliedLinearAccClamp=<25000 25000 25000>
    componentLinearAccScale=<1 1 1>
    componentLinearVelScale=<1 1 1>
    gravityOverride {z=-1800}
    physicsBodyDefinitions [{
      boundBone {boneName=dyn_thigh_l_cloth_bounce_03}
      boxExtents=<30 40 40>
      constraintSetup {angularLimitsMax=<10 10 10>, angularLimitsMin=<-10 -10 -10>, angularTarget=<1 0 0>, linearAxesMax=<3 3 3>, linearAxesMin=<-3 -3 -3>}
      localJointOffset=<-4 0 0>
     }]
   }
   n46 = Anim Dynamics - Bone: dyn_thigh_r_cloth_bounce_01(ComponentPose=@n45, Alpha=1, ExternalForce="0,0,0") {
    angularDampingOverride=0.99
    angularSpringConstant=15000
    bAngularSpring=true
    bLinearSpring=true
    bOverrideAngularDamping=true
    bUseGravityOverride=true
    gravityScale=0.35
    linearSpringConstant=0.5
    numSolverIterationsPreUpdate=1
    simulationSpace=RootRelative
    alphaBoolBlend {blendOption=Linear}
    boundBone {boneName=dyn_thigh_r_cloth_bounce_01}
    componentAppliedLinearAccClamp=<25000 25000 25000>
    componentLinearAccScale=<1 1 1>
    componentLinearVelScale=<1 1 1>
    gravityOverride {z=-1800}
    physicsBodyDefinitions [{
      boundBone {boneName=dyn_thigh_r_cloth_bounce_01}
      boxExtents=<20 40 40>
      constraintSetup {
       angularLimitsMax=<10 10 10>
       angularLimitsMin=<-10 -10 -10>
       angularTarget=<1 0 0>
       linearAxesMax=<0.5 0.5 0.5>
       linearAxesMin=<-0.5 -0.5 -0.5>
      }
      localJointOffset=<12 0 0>
     }]
   }
   n47 = Anim Dynamics - Bone: dyn_thigh_r_cloth_bounce_02(ComponentPose=@n46, Alpha=1, ExternalForce="0,0,0") {
    angularDampingOverride=0.99
    angularSpringConstant=15000
    bAngularSpring=true
    bLinearSpring=true
    bOverrideAngularDamping=true
    bOverrideLinearDamping=true
    bUseGravityOverride=true
    gravityScale=0.35
    linearDampingOverride=0.95
    linearSpringConstant=0.5
    numSolverIterationsPreUpdate=1
    simulationSpace=RootRelative
    alphaBoolBlend {blendOption=Linear}
    boundBone {boneName=dyn_thigh_r_cloth_bounce_02}
    componentAppliedLinearAccClamp=<25000 25000 25000>
    componentLinearAccScale=<1 1 1>
    componentLinearVelScale=<1 1 1>
    gravityOverride {z=-1800}
    physicsBodyDefinitions [{
      boundBone {boneName=dyn_thigh_r_cloth_bounce_02}
      boxExtents=<30 40 40>
      constraintSetup {angularLimitsMax=<10 10 10>, angularLimitsMin=<-10 -10 -10>, angularTarget=<1 0 0>, linearAxesMax=<3 3 3>, linearAxesMin=<-3 -3 -3>}
      localJointOffset=<4 0 0>
     }]
   }
   n48 = Anim Dynamics - Bone: dyn_thigh_r_cloth_bounce_03(ComponentPose=@n47, Alpha=1, ExternalForce="0,0,0") {
    angularDampingOverride=0.99
    angularSpringConstant=15000
    bAngularSpring=true
    bLinearSpring=true
    bOverrideAngularDamping=true
    bOverrideLinearDamping=true
    bUseGravityOverride=true
    gravityScale=0.35
    linearDampingOverride=0.95
    linearSpringConstant=0.5
    numSolverIterationsPreUpdate=1
    simulationSpace=RootRelative
    alphaBoolBlend {blendOption=Linear}
    boundBone {boneName=dyn_thigh_r_cloth_bounce_03}
    componentAppliedLinearAccClamp=<25000 25000 25000>
    componentLinearAccScale=<1 1 1>
    componentLinearVelScale=<1 1 1>
    gravityOverride {z=-1800}
    physicsBodyDefinitions [{
      boundBone {boneName=dyn_thigh_r_cloth_bounce_03}
      boxExtents=<30 40 40>
      constraintSetup {angularLimitsMax=<10 10 10>, angularLimitsMin=<-10 -10 -10>, angularTarget=<1 0 0>, linearAxesMax=<3 3 3>, linearAxesMin=<-3 -3 -3>}
      localJointOffset=<4 0 0>
     }]
   }
   n49 = Anim Dynamics - Bone: dyn_necklace_01(ComponentPose=@n48, Alpha=1, ExternalForce="0,0,0") {
    angularDampingOverride=1
    angularSpringConstant=20000
    bAngularSpring=true
    bOverrideAngularDamping=true
    bUseGravityOverride=true
    gravityScale=0.35
    simulationSpace=RootRelative
    alphaBoolBlend {blendOption=Linear}
    boundBone {boneName=dyn_necklace_01}
    componentLinearAccScale=<1 1 1>
    componentLinearVelScale=<1 1 1>
    gravityOverride {z=-980}
    physicsBodyDefinitions [{
      boundBone {boneName=dyn_necklace_01}
      boxExtents=<20 30 30>
      constraintSetup {angularLimitsMax=<5 5 5>, angularLimitsMin=<-5 -5 -5>, angularTarget=<1 0 0>}
      localJointOffset=<2 0 0>
     }]
    planarLimits [{drivingBone {boneName=spine_05}, planeTransform {location=<7.4338 8.9086 0.5977>, rotation=<rot 4.5615 25.4268 89.663>}}]
   }
   n50 = Anim Dynamics - Bone: dyn_necklace_02(ComponentPose=@n49, Alpha=1, ExternalForce="0,0,0") {
    angularDampingOverride=1
    angularSpringConstant=20000
    bAngularSpring=true
    bOverrideAngularDamping=true
    bUseGravityOverride=true
    gravityScale=0.35
    simulationSpace=RootRelative
    alphaBoolBlend {blendOption=Linear}
    boundBone {boneName=dyn_necklace_02}
    componentLinearAccScale=<1 1 1>
    componentLinearVelScale=<1 1 1>
    gravityOverride {z=-980}
    physicsBodyDefinitions [{
      boundBone {boneName=dyn_necklace_02}
      boxExtents=<20 30 30>
      constraintSetup {angularLimitsMax=<5 5 5>, angularLimitsMin=<-5 -5 0>, angularTarget=<1 0 0>}
      localJointOffset=<2 0 0>
     }]
    planarLimits [{drivingBone {boneName=spine_05}, planeTransform {location=<2.8957 8.471 -0.0209>, rotation=<rot 2.5186 -26.7766 86.1944>}}]
   }
   n51 = Anim Dynamics - Bone: dyn_necklace_03(ComponentPose=@n50, Alpha=1, ExternalForce="0,0,0") {
    angularDampingOverride=1
    angularSpringConstant=20000
    bAngularSpring=true
    bOverrideAngularDamping=true
    bUseGravityOverride=true
    gravityScale=0.35
    simulationSpace=RootRelative
    alphaBoolBlend {blendOption=Linear}
    boundBone {boneName=dyn_necklace_03}
    componentLinearAccScale=<1 1 1>
    componentLinearVelScale=<1 1 1>
    gravityOverride {z=-980}
    physicsBodyDefinitions [{
      boundBone {boneName=dyn_necklace_03}
      boxExtents=<20 30 30>
      constraintSetup {angularLimitsMax=<5 5 5>, angularLimitsMin=<-5 -5 0>, angularTarget=<1 0 0>}
      localJointOffset=<2 0 0>
     }]
    planarLimits [{drivingBone {boneName=spine_05}, planeTransform {location=<2.8957 8.471 -0.0209>, rotation=<rot 2.5186 -26.7766 86.1943>}}]
   }
   Output Pose(Result=Component To Local(ComponentPose=@n51)) {layerGroup=None}
   n52 = Control Rig!(
    Source=Copy Pose From Mesh() {bUseAttachedParent=true}
    Alpha
    MovementAlpha:double=Get MovementAlpha() -> (MovementAlpha:double)
    IsSkydiving:bool=Get IsSkydivingNotParachuting() -> (IsSkydivingNotParachuting:bool)
    IsSliding:bool=Get bIsBaseballSliding()
    IsKneeSliding:bool=Get bIsKneeSliding()
   ) -> (Pose) {
    bTransferInputCurves=false
    alphaBoolBlend {blendOption=Linear}
    controlRigAssetReference {blueprintRigClass=&/M_MED_DomeRoof_CtrlRig.M_MED_DomeRoof_CtrlRig_C}
   }
   Save cached pose 'PostDeformationRig'(Pose=@n52) {cacheName=PostDeformationRig}
  ]}
 EventGraph {nodes [
   n1 = Get Material() -> (Material:Material Instance Dynamic Object Reference)
   n2 = Get Overlay Material() -> (Overlay Material:Material Instance Dynamic Object Reference)
   n3 = Get Skeletal Mesh() -> (Skeletal Mesh:Skeletal Mesh Component Object Reference)
   Set WPO Globals Function(
    Material=@n1
    Overlay Material=@n2
    Skeletal Mesh=@n3
    "Local Noise Transform  Location"="Get Local Noise Transform  Location"() -> ("Local Noise Transform  Location":Vector)
    Local Noise Offset=Get Local Noise Offset()
   )
   n4 = Get Velocity Interp Speed()
   > Set WPO Regional Function!(
    execute
    Material:Material Instance Dynamic Object Reference=@n1
    Overlay Material:Material Instance Dynamic Object Reference=@n2
    Skeletal Mesh:Skeletal Mesh Component Object Reference=@n3
    Velocity Transform:Name=foot_r
    Velocity Interp Speed:double=@n4
    Collision Transform:Name=calf_r
    Collision Transform Axis:EAxis Enum
    Invert Collision Transform Axis:bool=false
    Delta Time X:double=0
    Previous Velocity Transform Location:"Vector (by ref)"=Get R1 Previous Velocity Transform Location() -> (R1 Previous Velocity Transform Location:Vector)
    Previous Velocity Transform Velocity:"Vector (by ref)"=Get R1 Previous Velocity Transform Velocity() -> (R1 Previous Velocity Transform Velocity:Vector)
    Region:int=1
   ) -> (then)
   > Set WPO Regional Function(
    Material=@n1
    Overlay Material=@n2
    Skeletal Mesh=@n3
    Velocity Transform=foot_l
    Velocity Interp Speed=@n4
    Collision Transform=calf_l
    Invert Collision Transform Axis=true
    Delta Time X=0
    Previous Velocity Transform Location=Get R2 Previous Velocity Transform Location() -> (R2 Previous Velocity Transform Location:Vector)
    Previous Velocity Transform Velocity=Get R2 Previous Velocity Transform Velocity() -> (R2 Previous Velocity Transform Velocity:Vector)
    Region=2
   )
   > n5 = Branch(Condition=Get bIsSkydiveFloating())
   > n6 = Set WPO Pho Wind Intensity(WPO Pho Wind Intensity=1)
   SetScalarParameterValue(
    execute=[@n7.then, @n6.then]
    self=[Get Overlay Material(), Get Material()]
    ParameterName=ClothSim_WindStrength
    Value=Get WPO Pho Wind Intensity()
   )
   n7 = Set WPO Pho Wind Intensity(execute=@n5.else, WPO Pho Wind Intensity=0)
   n8 = Event BlueprintUpdateAnimation()
   > n9 = Sequence!(execute) -> (then_0, then_1, then_2:Exec, then_3:Exec, then_4:Exec)
   n10 = Branch(execute=@n9.then_2, Condition=Get bClothEnabled())
   > Set UseAnimDynamicsCloth(UseAnimDynamicsCloth:bool=false) -> (Output_Get:bool)
   Set UseAnimDynamicsCloth(execute=@n10.else, UseAnimDynamicsCloth=true)
   Hide Sword(execute=@n9.then_3)
   n11 = Branch(execute=@n9.then_4, Condition=Get bIsInFrontEnd())
   > Set Simulation Space(Simulation Space:AnimPhysSimSpaceType Enum=World) -> (Output_Get:AnimPhysSimSpaceType Enum)
   > Set BootLipAlpha(BootLipAlpha:double=0.2) -> (Output_Get:double)
   Set Simulation Space(execute=@n11.else)
   > Set BootLipAlpha(BootLipAlpha=1)
   n12 = Gravity Override State Function DEPRECATED(
    execute=@n9.then_1
    CharacterPartAnimInstance=Self-Reference()
    GameTimeInSeconds=GetGameTimeInSeconds()
    GravityOverrideParams=Get Gravity Override Params()
    RBAN Params=Get RBAN Params()
   )
   > Set Gravity Override(Gravity Override=@n12)
   > Set RBAN Settings Out(RBAN Settings Out:RBANSetting Struct Structure=@n12.RBAN SettingsOut) -> (Output_Get:RBANSetting Struct Structure)
   > Cloth Controller Function DEPRECATED!(
    execute
    CharacterPartAnimInstance:Custom Character Part Anim Instance Object Reference=Self-Reference()
    ClothingAssetName:String=M_MED_DomeRoof_Clothing_0
    Cloth Paramaters:Cloth Params Struct Structure=Get Cloth Paramaters() -> (Cloth Paramaters:Cloth Params Struct Structure)
    Cloth Core Settings:Cloth Core Setting Struct Structure=Get Cloth Core Settings() -> (Cloth Core Settings:Cloth Core Setting Struct Structure)
    ClothUpAxis:Vector="1,0,0"
    OnGround_Moving_SpeedLimit:double=2
    OnGround_Moving_AngularVel_Max:double=0.05
    OnGround_Moving_FictitiousVel_Max:double=0.6
    LinVelScale_FwdBwd_Clamp:double=0
    LinVelScale_UpDwn_Clamp:double=0.7
    DeltaTime:double=@n8.DeltaTimeX
    GameTimeInSeconds:double=GetGameTimeInSeconds()
    WindGustsParams:Wind Gust Params Struct Structure=Get Wind Gusts Params() -> (Wind Gusts Params:Wind Gust Params Struct Structure)
    WindGustLocalSettings:Wind Gust Local Setting Struct Structure=Get Wind Gust Local Settings() -> (Wind Gust Local Settings:Wind Gust Local Setting Struct Structure)
   ) -> (
    then
    CurrentStateOut:String
    Cloth Settings Out:Cloth Setting Struct Structure
    Cloth Core Settings Out:Cloth Core Setting Struct Structure
    WindGustSettings Out:Wind Gust Setting Struct Structure
    WindGustLocalSettings Out:Wind Gust Local Setting Struct Structure
    GustStrength - WindVelocityOut:Vector
    LinearVelVector-DELETE LATER:Vector
   )
   > n13 = PlayerMovement(AnimInstance=Self-Reference(), XDotProductCoefficient=25, YDotProductCoefficient=25, ZDotProductCoefficient=25)
   > n14 = Branch(Condition=Get bIsCrouching())
   > Set MovementAlpha(MovementAlpha:double=0) -> (Output_Get:double)
   Set MovementAlpha(execute=@n14.else, MovementAlpha="Absolute (Float)"(A=float * float(A=@n13.YMovement, B=0.045) {is=title K2Node_PromotableOperator}))
   Set IsDomeRoofWeapon(
    execute=@n9.then_0
    IsDomeRoofWeapon:bool=PickaxeItemDefinitionHasAnyMetaTag(Tags=Make Array("[0]"=(TagName="Cosmetics.MetaTags.DomeRoof")), bRequiresEquipped=true)
   ) -> (Output_Get:bool)
   Event OnBeginTargeting() {is=title K2Node_Event}
   n15 = Get CurrentWeapon() -> (CurrentWeapon:Fort Weapon Object Reference)
   > n16 = Cast To FortWeaponRanged!(execute, Object=@n15) -> (then, CastFailed, AsFort Weapon Ranged:Fort Weapon Ranged Object Reference) {is=title K2Node_DynamicCast}
   > Branch(Condition=AND Boolean(A=IsUsingScopeOrFirstPersonCamera!(self:Fort Weapon Ranged Object Reference=@n16) -> (ReturnValue:bool), B=Get bClothEnabled()))
   > Set ClothBlendWeight!(execute, ClothBlendWeight:float=0, self:Skeletal Mesh Component Object Reference=GetPartSkeletalMeshComponent()) -> (then, Output_Get:float)
   Event OnWeaponChanged()
   > Hide Sword()
   Event OnBeginSkydiving()
   > Set IsSkydivingNotParachuting(IsSkydivingNotParachuting:bool=true) -> (Output_Get:bool)
   > Hide Sword()
   Event OnEndSkydiving()
   > Set IsSkydivingNotParachuting(IsSkydivingNotParachuting=false)
   > Hide Sword()
   Event OnParachuteOpened()
   > Set IsSkydivingNotParachuting(IsSkydivingNotParachuting=false)
   Event OnParachuteClosed()
   > n17 = Branch(Condition=Get bIsSkyDiving())
   > Set IsSkydivingNotParachuting(IsSkydivingNotParachuting=true)
   Set IsSkydivingNotParachuting(execute=@n17.else, IsSkydivingNotParachuting=false)
   Event OnCharacterAndPickaxeCustomizationLoaded!() -> (
    OutputDelegate
    then
    CustomBaseMesh:Skeletal Mesh Component Object Reference
    CustomParts:Array of Custom Character Part Object References
    CustomMeshParts:Array of Skeletal Mesh Component Budgeted Object References
   ) {is=title K2Node_Event}
   > Hide Sword()
   Event OnEndTargeting() {is=title K2Node_Event}
   > Delay(Duration=1)
   > n18 = Cast To FortWeaponRanged(Object=@n15)
   > n19 = Branch(Condition=IsUsingScopeOrFirstPersonCamera(self=@n18))
   ForceClothNextUpdateTeleportAndReset(execute=[@n19.else, @n18.CastFailed], self:Skeletal Mesh Component Object Reference=GetPartSkeletalMeshComponent())
   > Set ClothBlendWeight(ClothBlendWeight=1, self=GetPartSkeletalMeshComponent())
  ]}
 Hide Sword {nodes [
   Hide Sword() {is=title K2Node_FunctionEntry}
   > n1 = Branch(Condition=PickaxeItemDefinitionHasAnyMetaTag(Tags=Make Array("[0]"=(TagName="Cosmetics.MetaTags.DomeRoof")), bRequiresEquipped=true))
   > n2 = Branch(Condition=OR Boolean!(A=Get bIsSkyDiving(), B=Get bIsTacticalSprinting(), C:bool=Get bIsOnGrindRail()) -> (ReturnValue))
   n3 = Get Dynamic Material Instance(execute=[@n2.then, @n1.else], SkeletalMeshComponent=GetPartSkeletalMeshComponent())
   > SetScalarParameterValue(self=@n3, ParameterName=Hide Element 01, Value=0)
   n4 = Get Dynamic Material Instance(execute=@n2.else, SkeletalMeshComponent=GetPartSkeletalMeshComponent())
   > SetScalarParameterValue(self=@n4, ParameterName=Hide Element 01, Value=1)
  ]}
]
variables [
 BootLipAlpha {type=number}
 Cloth Core Settings {
  type=ClothCoreSettingStruct
  default {
   collisionThickness=0.5
   dampingCoefficient=0.125
   frictionCoefficient=0.2
   localDampingCoefficient=0.1
   numberIterations=2
   numberSubsteps=1
   selfCollisionThickness=0
   animDriveDamping=<0 1>
   animDriveStiffness=<0 1>
   areaStiffness=<1 1>
   bendingStiffness=<1 1>
   edgeStiffness=<1 1>
   tetherStiffness=<1 1>
  }
 }
 Cloth Paramaters {
  type=ClothParamsStruct
  default {
   dBNO {
    angularVelocityScale=0.6
    fictitiousAngularScale=0.1
    jointName=pelvis
    dragCoefficient=<0.2 1>
    gravityOverride=<-980 400 0>
    liftCoefficient=<0 0.5>
    linearVelocityScale=<0.3 0.3 0.1>
   }
   dBNO_Carried {
    angularVelocityScale=0.6
    fictitiousAngularScale=0.1
    jointName=pelvis
    dragCoefficient=<0.2 1>
    gravityOverride=<-980 -200 0>
    liftCoefficient=<0 0.5>
    linearVelocityScale=<0.3 0.3 0.1>
   }
   dBNO_Impulse {
    angularVelocityScale=0
    fictitiousAngularScale=0
    jointName=pelvis
    dragCoefficient=<0 0.2>
    gravityOverride=<-3000 -2000 0>
    liftCoefficient=<0 0.1>
    linearVelocityScale=<0.3 0.1 0.3>
   }
   dBNO_RollBwd {
    angularVelocityScale=0
    fictitiousAngularScale=0
    jointName=pelvis
    dragCoefficient=<0 0.2>
    gravityOverride=<-2000 0 0>
    liftCoefficient=<0 0.1>
    linearVelocityScale=<0.3 0.1 0.3>
   }
   dBNO_RollFwd {
    angularVelocityScale=0
    fictitiousAngularScale=0
    jointName=pelvis
    dragCoefficient=<0 0.2>
    gravityOverride=<-980 0 0>
    liftCoefficient=<0 0.1>
    linearVelocityScale=<0.3 0.1 0.3>
   }
   dBNO_Tumbling {
    angularVelocityScale=0
    fictitiousAngularScale=0
    jointName=pelvis
    dragCoefficient=<0 0.2>
    gravityOverride=<-2500 -300 0>
    liftCoefficient=<0 0.1>
    linearVelocityScale=<0.3 0.1 0.3>
   }
   emoteOrMelee {
    angularVelocityScale=0.6
    fictitiousAngularScale=0.1
    jointName=root
    dragCoefficient=<2.5 2.5>
    gravityOverride=<0 0 -980>
    liftCoefficient=<0.5 0.5>
    linearVelocityScale=<0.6 0.6 0.5>
   }
   front_End {
    angularVelocityScale=0.05
    fictitiousAngularScale=0
    jointName=root
    dragCoefficient=<0.2 0.2>
    gravityOverride=<0 -200 -980>
    liftCoefficient=<0.1 0.1>
    linearVelocityScale=<0.1 0.1 0.1>
   }
   onGround_Crouch_Moving {
    angularVelocityScale=0.1
    fictitiousAngularScale=0.3
    jointName=root
    dragCoefficient=<2 2>
    gravityOverride=<0 0 -980>
    liftCoefficient=<0.5 0.5>
    linearVelocityScale=<0.6 0.6 0.5>
   }
   onGround_Falling {
    angularVelocityScale=0.6
    fictitiousAngularScale=0
    jointName=root
    dragCoefficient=<0 0.2>
    gravityOverride=<0 -500 -980>
    liftCoefficient=<0 0.1>
    linearVelocityScale=<2 1 0.3>
   }
   onGround_Moving {
    angularVelocityScale=0.1
    fictitiousAngularScale=0.1
    jointName=root
    dragCoefficient=<0.5 0.5>
    gravityOverride=<0 0 -1100>
    liftCoefficient=<0.2 0.2>
    linearVelocityScale=<0.3 0.3 0.25>
   }
   onGround_Standing {
    angularVelocityScale=0.1
    fictitiousAngularScale=0.5
    jointName=root
    dragCoefficient=<1.5 1.5>
    gravityOverride=<0 0 -980>
    liftCoefficient=<0.5 0.5>
    linearVelocityScale=<0.6 0.6 0.5>
   }
   rail_Grinding {
    angularVelocityScale=0.2
    fictitiousAngularScale=0
    jointName=root
    dragCoefficient=<0.05 0.1>
    gravityOverride=<50 -100 -980>
    liftCoefficient=<0 0.1>
    linearVelocityScale=<0.1 0.1 0.1>
   }
   riding_Moving {
    angularVelocityScale=0.05
    fictitiousAngularScale=1
    jointName=root
    dragCoefficient=<0.5 0.5>
    gravityOverride=<0 0 -980>
    liftCoefficient=<0.5 0.5>
    linearVelocityScale=<0.4 0.6 0.3>
   }
   riding_Moving_Targeting {
    angularVelocityScale=0.05
    fictitiousAngularScale=0.7
    jointName=root
    dragCoefficient=<0.5 0.5>
    gravityOverride=<0 0 -980>
    liftCoefficient=<0.5 0.5>
    linearVelocityScale=<0.2 0.6 0.2>
   }
   riding_Standing {
    angularVelocityScale=0.05
    fictitiousAngularScale=1
    jointName=root
    dragCoefficient=<0.5 0.5>
    gravityOverride=<0 -450 -980>
    liftCoefficient=<0.5 0.5>
    linearVelocityScale=<0.6 0.6 0.3>
   }
   skydive_DivingDown {
    angularVelocityScale=0.1
    fictitiousAngularScale=0
    jointName=pelvis
    dragCoefficient=<0.3 0.3>
    gravityOverride=<-450 -100 0>
    liftCoefficient=<0.9 0.9>
    linearVelocityScale=<0.1 0.1 0.1>
   }
   skydive_FloatingDown {
    angularVelocityScale=0.1
    fictitiousAngularScale=0
    jointName=root
    dragCoefficient=<0.3 0.3>
    gravityOverride=<0 -800 200>
    liftCoefficient=<1.3 1.3>
    linearVelocityScale=<0.2 0.2 0.1>
   }
   skydive_GoingUp {
    angularVelocityScale=0.1
    fictitiousAngularScale=0
    jointName=root
    dragCoefficient=<0.01 0.01>
    gravityOverride=<0 -2000 400>
    liftCoefficient=<0.05 0.05>
    linearVelocityScale=<0.1 0.1 0.01>
   }
   skydive_Parachuting {
    angularVelocityScale=0.6
    fictitiousAngularScale=0.1
    jointName=root
    dragCoefficient=<0.8 0.8>
    gravityOverride=<0 -600 -500>
    liftCoefficient=<0.8 0.8>
    linearVelocityScale=<0.2 0.2 0.3>
   }
   skydive_Surfing {
    angularVelocityScale=0.6
    fictitiousAngularScale=0.1
    jointName=root
    dragCoefficient=<1.5 1.5>
    gravityOverride=<0 -600 -500>
    liftCoefficient=<1 1>
    linearVelocityScale=<0.1 0.1 0.3>
   }
   swimming {
    angularVelocityScale=0.6
    fictitiousAngularScale=0
    jointName=root
    dragCoefficient=<2.5 2.5>
    gravityOverride=<0 0 0>
    liftCoefficient=<0.1 0.1>
    linearVelocityScale=<0.6 0.6 0.5>
   }
   tactical_Sprint {
    angularVelocityScale=0.1
    fictitiousAngularScale=0
    jointName=root
    dragCoefficient=<0.05 0.15>
    gravityOverride=<0 0 -1100>
    liftCoefficient=<0.05 0.15>
    linearVelocityScale=<0.05 0.05 0.05>
   }
   ziplining {
    angularVelocityScale=0.1
    fictitiousAngularScale=0
    jointName=root
    dragCoefficient=<0 0.1>
    gravityOverride=<0 0 -980>
    liftCoefficient=<0 0.1>
    linearVelocityScale=<0.1 0.1 0.25>
   }
  }
 }
 ClothWeightTarget {type=number, default=1}
 Current State Out {type=string, default=""}
 Gravity Override {type=Vector}
 Gravity Override Params {
  type=GravityOverrideParamsStruct
  default {
   onGround_Moving {gravityOverride=<0 -200 -700>}
   onGround_Standing {gravityOverride=<100 -100 -980>, windAmplitude=<50 -20 50>, windFrequency=<1 2.5 2.5>, windOffset=<-25 -80 -45>}
   skydive_DivingDown {jointName=root, gravityOverride=<0 -500 400>, windAmplitude=<240 240 240>, windFrequency=<4 50 20>, windOffset=<0 -200 300>}
   skydive_FloatingDown {gravityOverride=<0 -300 50>, windAmplitude=<200 -250 200>, windOffset=<0 0 0>}
   skydive_GoingUp {gravityOverride=<0 -300 50>, windAmplitude=<200 -250 200>}
   skydive_Parachuting {windFrequency=<1.5 5 5>}
   ziplining {gravityOverride=<0 -450 -500>, windAmplitude=<20 50 10>, windFrequency=<1 5 7>, windOffset=<-200 -400 -200>}
  }
 }
 IsDomeRoofWeapon {type=boolean}
 IsSkydivingNotParachuting {type=boolean}
 Local Noise Offset {type=Vector}
 "Local Noise Transform  Location" {type=Vector}
 Material {type=MaterialInstanceDynamic}
 MovementAlpha {type=number}
 Overlay Material {type=MaterialInstanceDynamic}
 PrevClothWeight {type=number}
 R1 Previous Velocity Transform Location {type=Vector}
 R1 Previous Velocity Transform Velocity {type=Vector}
 R2 Previous Velocity Transform Location {type=Vector}
 R2 Previous Velocity Transform Velocity {type=Vector}
 RBAN Params {
  type=RBANParamsStruct
  default {
   dBNO {
    componentLinearAccScale=<0.3 0.3 0.5>
    componentLinearVelScale=<0.3 0.3 0.5>
    simSpaceSettings {
     maxAngularAcceleration=200
     maxAngularVelocity=200
     maxLinearAcceleration=50
     maxLinearVelocity=50
     velocityScaleZ=0.5
     externalLinearDragV=<0 0 0>
    }
   }
   dBNO_Carried {
    componentLinearAccScale=<0.3 0.3 0.5>
    componentLinearVelScale=<0.3 0.3 0.5>
    simSpaceSettings {
     maxAngularAcceleration=200
     maxAngularVelocity=200
     maxLinearAcceleration=50
     maxLinearVelocity=50
     velocityScaleZ=0.5
     externalLinearDragV=<0 0 0>
    }
   }
   emoteOrMelee {
    componentLinearAccScale=<0.1 0.1 0.35>
    componentLinearVelScale=<0.1 0.1 0.35>
    simSpaceSettings {
     maxAngularAcceleration=150
     maxAngularVelocity=150
     maxLinearAcceleration=30
     maxLinearVelocity=30
     velocityScaleZ=0.5
     externalLinearDragV=<0 0 0>
    }
   }
   front_End {
    componentLinearAccScale=<0.3 0.3 0.5>
    componentLinearVelScale=<0.3 0.3 0.5>
    simSpaceSettings {
     maxAngularAcceleration=200
     maxAngularVelocity=200
     maxLinearAcceleration=50
     maxLinearVelocity=50
     velocityScaleZ=0.05
     worldAlpha=0.1
     externalLinearDragV=<0 0 0>
    }
   }
   onGround_Crouch_Moving {
    componentLinearAccScale=<0.3 0.3 0.5>
    componentLinearVelScale=<0.3 0.3 0.5>
    simSpaceSettings {
     maxAngularAcceleration=200
     maxAngularVelocity=200
     maxLinearAcceleration=50
     maxLinearVelocity=50
     velocityScaleZ=0.5
     externalLinearDragV=<0 0 0>
    }
   }
   onGround_Falling {
    componentLinearAccScale=<0.3 0.3 0.5>
    componentLinearVelScale=<0.3 0.3 0.5>
    simSpaceSettings {
     maxAngularAcceleration=200
     maxAngularVelocity=200
     maxLinearAcceleration=50
     maxLinearVelocity=50
     velocityScaleZ=0.5
     externalLinearDragV=<0 0 0>
    }
   }
   onGround_Moving {
    componentLinearAccScale=<0.3 0.3 0.3>
    componentLinearVelScale=<0.3 0.3 0.3>
    simSpaceSettings {
     maxAngularAcceleration=200
     maxAngularVelocity=200
     maxLinearAcceleration=50
     maxLinearVelocity=50
     velocityScaleZ=1
     externalLinearDragV=<0.1 0.1 0.1>
    }
   }
   onGround_Standing {
    componentLinearAccScale=<0.05 0.05 0.15>
    componentLinearVelScale=<0.05 0.05 0.15>
    simSpaceSettings {
     maxAngularAcceleration=150
     maxAngularVelocity=150
     maxLinearAcceleration=30
     maxLinearVelocity=30
     velocityScaleZ=0.2
     worldAlpha=0.2
     externalLinearDragV=<0 0 0>
    }
   }
   rail_Grinding {
    componentLinearAccScale=<0.3 0.3 0.5>
    componentLinearVelScale=<0.3 0.3 0.5>
    simSpaceSettings {
     maxAngularAcceleration=200
     maxAngularVelocity=200
     maxLinearAcceleration=50
     maxLinearVelocity=50
     velocityScaleZ=0.5
     worldAlpha=1
     externalLinearDragV=<0 0 0>
    }
   }
   riding_Moving {
    componentLinearAccScale=<0.3 0.3 0.5>
    componentLinearVelScale=<0.3 0.3 0.5>
    simSpaceSettings {
     maxAngularAcceleration=200
     maxAngularVelocity=200
     maxLinearAcceleration=50
     maxLinearVelocity=50
     velocityScaleZ=0.5
     externalLinearDragV=<0 0 0>
    }
   }
   riding_Moving_Targeting {
    componentLinearAccScale=<0.3 0.3 0.5>
    componentLinearVelScale=<0.3 0.3 0.5>
    simSpaceSettings {
     maxAngularAcceleration=200
     maxAngularVelocity=200
     maxLinearAcceleration=50
     maxLinearVelocity=50
     velocityScaleZ=0.5
     externalLinearDragV=<0 0 0>
    }
   }
   riding_Standing {
    componentLinearAccScale=<0.3 0.3 0.5>
    componentLinearVelScale=<0.3 0.3 0.5>
    simSpaceSettings {
     maxAngularAcceleration=200
     maxAngularVelocity=200
     maxLinearAcceleration=50
     maxLinearVelocity=50
     velocityScaleZ=0.5
     externalLinearDragV=<0 0 0>
    }
   }
   skydive_DivingDown {
    componentLinearAccScale=<0.1 0.1 0.1>
    componentLinearVelScale=<0.1 0.1 0.1>
    simSpaceSettings {
     maxAngularAcceleration=200
     maxAngularVelocity=200
     maxLinearAcceleration=50
     maxLinearVelocity=50
     velocityScaleZ=0.5
     externalLinearDragV=<0 0 0>
    }
   }
   skydive_FloatingDown {
    componentLinearAccScale=<0.1 0.1 0.05>
    componentLinearVelScale=<0.1 0.1 0.05>
    simSpaceSettings {
     maxAngularAcceleration=100
     maxAngularVelocity=100
     maxLinearAcceleration=100
     maxLinearVelocity=100
     velocityScaleZ=0.5
     worldAlpha=0.5
     externalLinearDragV=<0.1 0.1 0.1>
     externalLinearVelocity=<0.1 0.1 0.1>
    }
   }
   skydive_GoingUp {
    componentLinearAccScale=<0.1 0.1 0.05>
    componentLinearVelScale=<0.1 0.1 0.05>
    simSpaceSettings {
     maxAngularAcceleration=50
     maxAngularVelocity=50
     maxLinearAcceleration=50
     maxLinearVelocity=50
     velocityScaleZ=0.5
     worldAlpha=0.25
     externalLinearDragV=<0 0 0>
    }
   }
   skydive_Parachuting {
    componentLinearAccScale=<0.1 0.1 0.1>
    componentLinearVelScale=<0.1 0.1 0.1>
    simSpaceSettings {
     maxAngularAcceleration=200
     maxAngularVelocity=200
     maxLinearAcceleration=50
     maxLinearVelocity=50
     velocityScaleZ=0.5
     externalLinearDragV=<0 0 0>
    }
   }
   skydive_Surfing {
    componentLinearAccScale=<0.3 0.3 0.5>
    componentLinearVelScale=<0.3 0.3 0.5>
    simSpaceSettings {
     maxAngularAcceleration=200
     maxAngularVelocity=200
     maxLinearAcceleration=50
     maxLinearVelocity=50
     velocityScaleZ=0.5
     externalLinearDragV=<0 0 0>
    }
   }
   swimming {
    componentLinearAccScale=<0.3 0.3 0.5>
    componentLinearVelScale=<0.3 0.3 0.5>
    simSpaceSettings {
     maxAngularAcceleration=200
     maxAngularVelocity=200
     maxLinearAcceleration=50
     maxLinearVelocity=50
     velocityScaleZ=0.5
     externalLinearDragV=<0 0 0>
    }
   }
   tactical_Sprint {
    componentLinearAccScale=<0.3 0.3 0.5>
    componentLinearVelScale=<0.3 0.3 0.5>
    simSpaceSettings {
     maxAngularAcceleration=200
     maxAngularVelocity=200
     maxLinearAcceleration=50
     maxLinearVelocity=50
     velocityScaleZ=0.5
     externalLinearDragV=<0 0 0>
    }
   }
   ziplining {
    componentLinearAccScale=<0.1 0.1 0.2>
    componentLinearVelScale=<0.1 0.1 0.2>
    simSpaceSettings {
     maxAngularAcceleration=200
     maxAngularVelocity=200
     maxLinearAcceleration=50
     maxLinearVelocity=50
     velocityScaleZ=0.5
     externalLinearDragV=<0 0 0>
    }
   }
  }
 }
 RBAN Settings Out {
  type=RBANSettingStruct
  default {
   componentLinearAccScale=<0 0 0>
   componentLinearVelScale=<0 0 0>
   simSpaceSettings {
    dampingAlpha=1
    maxAngularAcceleration=1000
    maxAngularVelocity=3
    maxLinearAcceleration=300
    maxLinearVelocity=700
    velocityScaleZ=0
    worldAlpha=1
    externalAngularVelocity=<0 0 0>
    externalLinearDragV=<0 0.5 0>
    externalLinearVelocity=<0 0 0>
   }
  }
 }
 Simulation Space {type=AnimPhysSimSpaceType, default=RootRelative}
 Skeletal Mesh {type=SkeletalMeshComponent}
 UseAnimDynamicsCloth {type=boolean}
 Velocity Interp Speed {type=number, default=1}
 Wind Gust Local Settings {type=WindGustLocalSettingStruct, default {elapsedTime=0, finalGustStrengthLocal=0, gustDuration=0, gustStrength=0, timeBetweenGusts=0}}
 Wind Gusts Params {
  type=WindGustParamsStruct
  default {
   dBNO {
    maxDownInterpSpeed=5
    maxGustDuration=2
    maxGustStrength=7
    maxTimeBetweenGusts=7
    maxUpInterpSpeed=5
    minDownInterpSpeed=2
    minGustDuration=0.5
    minGustStrength=3
    minTimeBetweenGusts=0
    minUpInterpSpeed=2
    windTurbulenceScalarA=0.1
    windTurbulenceScalarB=1
    windAmplitude=<0 0 0>
    windFrequency=<0 0 0>
    windOffset=<0 0 0>
   }
   dBNO_Carried {
    maxDownInterpSpeed=5
    maxGustDuration=2
    maxGustStrength=7
    maxTimeBetweenGusts=7
    maxUpInterpSpeed=5
    minDownInterpSpeed=2
    minGustDuration=0.5
    minGustStrength=3
    minTimeBetweenGusts=0
    minUpInterpSpeed=2
    windTurbulenceScalarA=0.1
    windTurbulenceScalarB=1
    windAmplitude=<0 0 0>
    windFrequency=<0 0 0>
    windOffset=<0 0 0>
   }
   dBNO_Impulse {
    maxDownInterpSpeed=5
    maxGustDuration=2
    maxGustStrength=7
    maxTimeBetweenGusts=7
    maxUpInterpSpeed=5
    minDownInterpSpeed=2
    minGustDuration=0.5
    minGustStrength=3
    minTimeBetweenGusts=0
    minUpInterpSpeed=2
    windTurbulenceScalarA=0.1
    windTurbulenceScalarB=1
    windAmplitude=<0 0 0>
    windFrequency=<0 0 0>
    windOffset=<0 0 0>
   }
   dBNO_RollBwd {
    maxDownInterpSpeed=5
    maxGustDuration=2
    maxGustStrength=7
    maxTimeBetweenGusts=7
    maxUpInterpSpeed=5
    minDownInterpSpeed=2
    minGustDuration=0.5
    minGustStrength=3
    minTimeBetweenGusts=0
    minUpInterpSpeed=2
    windTurbulenceScalarA=0.1
    windTurbulenceScalarB=1
    windAmplitude=<0 0 0>
    windFrequency=<0 0 0>
    windOffset=<0 0 0>
   }
   dBNO_RollFwd {
    maxDownInterpSpeed=5
    maxGustDuration=2
    maxGustStrength=7
    maxTimeBetweenGusts=7
    maxUpInterpSpeed=5
    minDownInterpSpeed=2
    minGustDuration=0.5
    minGustStrength=3
    minTimeBetweenGusts=0
    minUpInterpSpeed=2
    windTurbulenceScalarA=0.1
    windTurbulenceScalarB=1
    windAmplitude=<0 0 0>
    windFrequency=<0 0 0>
    windOffset=<0 0 0>
   }
   dBNO_Tumbling {
    maxDownInterpSpeed=5
    maxGustDuration=2
    maxGustStrength=7
    maxTimeBetweenGusts=7
    maxUpInterpSpeed=5
    minDownInterpSpeed=2
    minGustDuration=0.5
    minGustStrength=3
    minTimeBetweenGusts=0
    minUpInterpSpeed=2
    windTurbulenceScalarA=0.1
    windTurbulenceScalarB=1
    windAmplitude=<0 0 0>
    windFrequency=<0 0 0>
    windOffset=<0 0 0>
   }
   emoteOrMelee {
    maxDownInterpSpeed=5
    maxGustDuration=2
    maxGustStrength=7
    maxTimeBetweenGusts=7
    maxUpInterpSpeed=5
    minDownInterpSpeed=2
    minGustDuration=0.5
    minGustStrength=3
    minTimeBetweenGusts=0
    minUpInterpSpeed=2
    windTurbulenceScalarA=0.1
    windTurbulenceScalarB=1
    windAmplitude=<0 0 0>
    windFrequency=<0 0 0>
    windOffset=<0 0 0>
   }
   front_End {
    maxDownInterpSpeed=5
    maxGustDuration=2
    maxGustStrength=7
    maxTimeBetweenGusts=7
    maxUpInterpSpeed=5
    minDownInterpSpeed=2
    minGustDuration=0.5
    minGustStrength=3
    minTimeBetweenGusts=0
    minUpInterpSpeed=2
    windTurbulenceScalarA=0.1
    windTurbulenceScalarB=1
    windAmplitude=<0 0 0>
    windFrequency=<0 0 0>
    windOffset=<0 0 0>
   }
   onGround_Crouch_Moving {
    maxDownInterpSpeed=5
    maxGustDuration=2
    maxGustStrength=7
    maxTimeBetweenGusts=7
    maxUpInterpSpeed=5
    minDownInterpSpeed=2
    minGustDuration=0.5
    minGustStrength=3
    minTimeBetweenGusts=0
    minUpInterpSpeed=2
    windTurbulenceScalarA=0.1
    windTurbulenceScalarB=1
    windAmplitude=<0 0 0>
    windFrequency=<0 0 0>
    windOffset=<0 0 0>
   }
   onGround_Falling {
    maxDownInterpSpeed=5
    maxGustDuration=2
    maxGustStrength=7
    maxTimeBetweenGusts=7
    maxUpInterpSpeed=5
    minDownInterpSpeed=2
    minGustDuration=0.5
    minGustStrength=3
    minTimeBetweenGusts=0
    minUpInterpSpeed=2
    windTurbulenceScalarA=0.1
    windTurbulenceScalarB=1
    windAmplitude=<0 0 0>
    windFrequency=<0 0 0>
    windOffset=<0 0 0>
   }
   onGround_Moving {
    maxDownInterpSpeed=5
    maxGustDuration=2
    maxGustStrength=7
    maxTimeBetweenGusts=7
    maxUpInterpSpeed=5
    minDownInterpSpeed=2
    minGustDuration=0.5
    minGustStrength=3
    minTimeBetweenGusts=0
    minUpInterpSpeed=2
    windTurbulenceScalarA=0.1
    windTurbulenceScalarB=1
    windAmplitude=<0 0 0>
    windFrequency=<0 0 0>
    windOffset=<0 0 0>
   }
   onGround_Standing {
    maxDownInterpSpeed=5
    maxGustDuration=2
    maxGustStrength=7
    maxTimeBetweenGusts=7
    maxUpInterpSpeed=5
    minDownInterpSpeed=2
    minGustDuration=0.5
    minGustStrength=3
    minTimeBetweenGusts=0
    minUpInterpSpeed=2
    windTurbulenceScalarA=0.1
    windTurbulenceScalarB=1
    windAmplitude=<0 0 0>
    windFrequency=<0 0 0>
    windOffset=<0 0 0>
   }
   rail_Grinding {
    maxDownInterpSpeed=5
    maxGustDuration=8
    maxGustStrength=7
    maxTimeBetweenGusts=8
    maxUpInterpSpeed=5
    minDownInterpSpeed=2
    minGustDuration=0.2
    minGustStrength=3
    minTimeBetweenGusts=1
    minUpInterpSpeed=2
    windTurbulenceScalarA=0.1
    windTurbulenceScalarB=0.8
    windAmplitude=<20 65 42>
    windFrequency=<5 6 8.6>
    windOffset=<50 -500 -400>
   }
   riding_Moving {
    maxDownInterpSpeed=5
    maxGustDuration=2
    maxGustStrength=7
    maxTimeBetweenGusts=7
    maxUpInterpSpeed=5
    minDownInterpSpeed=2
    minGustDuration=0.5
    minGustStrength=3
    minTimeBetweenGusts=0
    minUpInterpSpeed=2
    windTurbulenceScalarA=0.1
    windTurbulenceScalarB=1
    windAmplitude=<0 0 0>
    windFrequency=<0 0 0>
    windOffset=<0 0 0>
   }
   riding_Moving_Targeting {
    maxDownInterpSpeed=5
    maxGustDuration=2
    maxGustStrength=7
    maxTimeBetweenGusts=7
    maxUpInterpSpeed=5
    minDownInterpSpeed=2
    minGustDuration=0.5
    minGustStrength=3
    minTimeBetweenGusts=0
    minUpInterpSpeed=2
    windTurbulenceScalarA=0.1
    windTurbulenceScalarB=1
    windAmplitude=<0 0 0>
    windFrequency=<0 0 0>
    windOffset=<0 0 0>
   }
   riding_Standing {
    maxDownInterpSpeed=5
    maxGustDuration=2
    maxGustStrength=7
    maxTimeBetweenGusts=7
    maxUpInterpSpeed=5
    minDownInterpSpeed=2
    minGustDuration=0.5
    minGustStrength=3
    minTimeBetweenGusts=0
    minUpInterpSpeed=2
    windTurbulenceScalarA=0.1
    windTurbulenceScalarB=1
    windAmplitude=<0 0 0>
    windFrequency=<0 0 0>
    windOffset=<0 0 0>
   }
   skydive_DivingDown {
    maxDownInterpSpeed=5
    maxGustDuration=2
    maxGustStrength=7
    maxTimeBetweenGusts=7
    maxUpInterpSpeed=5
    minDownInterpSpeed=2
    minGustDuration=0.5
    minGustStrength=3
    minTimeBetweenGusts=0
    minUpInterpSpeed=2
    windTurbulenceScalarA=0.1
    windTurbulenceScalarB=1
    windAmplitude=<0 0 0>
    windFrequency=<0 0 0>
    windOffset=<0 0 0>
   }
   skydive_FloatingDown {
    maxDownInterpSpeed=1.5
    maxGustDuration=2
    maxGustStrength=0.5
    maxTimeBetweenGusts=7
    maxUpInterpSpeed=1.5
    minDownInterpSpeed=0.5
    minGustDuration=0.5
    minGustStrength=0
    minTimeBetweenGusts=0
    minUpInterpSpeed=0.5
    windTurbulenceScalarA=0.1
    windTurbulenceScalarB=1
    windAmplitude=<0 0 0>
    windFrequency=<0 0 0>
    windOffset=<0 0 0>
   }
   skydive_GoingUp {
    maxDownInterpSpeed=1.5
    maxGustDuration=2
    maxGustStrength=0.5
    maxTimeBetweenGusts=7
    maxUpInterpSpeed=1.5
    minDownInterpSpeed=0.5
    minGustDuration=0.5
    minGustStrength=0
    minTimeBetweenGusts=0
    minUpInterpSpeed=0.5
    windTurbulenceScalarA=0.1
    windTurbulenceScalarB=1
    windAmplitude=<0 0 0>
    windFrequency=<0 0 0>
    windOffset=<0 0 0>
   }
   skydive_Parachuting {
    maxDownInterpSpeed=5
    maxGustDuration=2
    maxGustStrength=7
    maxTimeBetweenGusts=7
    maxUpInterpSpeed=5
    minDownInterpSpeed=2
    minGustDuration=0.5
    minGustStrength=3
    minTimeBetweenGusts=0
    minUpInterpSpeed=2
    windTurbulenceScalarA=0.1
    windTurbulenceScalarB=1
    windAmplitude=<0 0 0>
    windFrequency=<0 0 0>
    windOffset=<0 0 0>
   }
   skydive_Surfing {
    maxDownInterpSpeed=5
    maxGustDuration=2
    maxGustStrength=7
    maxTimeBetweenGusts=7
    maxUpInterpSpeed=5
    minDownInterpSpeed=2
    minGustDuration=0.5
    minGustStrength=3
    minTimeBetweenGusts=0
    minUpInterpSpeed=2
    windTurbulenceScalarA=0.1
    windTurbulenceScalarB=1
    windAmplitude=<0 0 0>
    windFrequency=<0 0 0>
    windOffset=<0 0 0>
   }
   swimming {
    maxDownInterpSpeed=5
    maxGustDuration=2
    maxGustStrength=7
    maxTimeBetweenGusts=7
    maxUpInterpSpeed=5
    minDownInterpSpeed=2
    minGustDuration=0.5
    minGustStrength=3
    minTimeBetweenGusts=0
    minUpInterpSpeed=2
    windTurbulenceScalarA=0.1
    windTurbulenceScalarB=1
    windAmplitude=<0 0 0>
    windFrequency=<0 0 0>
    windOffset=<0 0 0>
   }
   tactical_Sprint {
    maxDownInterpSpeed=5
    maxGustDuration=2
    maxGustStrength=7
    maxTimeBetweenGusts=7
    maxUpInterpSpeed=5
    minDownInterpSpeed=2
    minGustDuration=0.5
    minGustStrength=3
    minTimeBetweenGusts=0
    minUpInterpSpeed=2
    windTurbulenceScalarA=0.1
    windTurbulenceScalarB=1
    windAmplitude=<0 0 0>
    windFrequency=<0 0 0>
    windOffset=<0 0 0>
   }
   ziplining {
    maxDownInterpSpeed=5
    maxGustDuration=2
    maxGustStrength=7
    maxTimeBetweenGusts=7
    maxUpInterpSpeed=5
    minDownInterpSpeed=2
    minGustDuration=0.5
    minGustStrength=3
    minTimeBetweenGusts=0
    minUpInterpSpeed=2
    windTurbulenceScalarA=0.1
    windTurbulenceScalarB=1
    windAmplitude=<0 0 0>
    windFrequency=<0 0 0>
    windOffset=<0 0 0>
   }
  }
 }
 WPO Pho Wind Intensity {type=number}
]
