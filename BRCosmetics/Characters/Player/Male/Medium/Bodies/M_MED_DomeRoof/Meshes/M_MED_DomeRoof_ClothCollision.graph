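# Physics asset for the male medium DomeRoof player body cosmetic, named as the collision set for its cloth.
# Capsules cover the pelvis, left thigh, foot and hand, right thumb, and a chain of left front jacket bones.
# The left calf and right hand bodies have no shapes.
# Constraints link bodies back toward the pelvis, except the unconstrained thigh, with free angular motion on the jacket and thumb chains.
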PhysicsAsset /BRCosmetics/Characters/Player/Male/Medium/Bodies/M_MED_DomeRoof/Meshes/M_MED_DomeRoof_ClothCollision
bodies {
 calf_l []
 dyn_jacket_front_l_02 [{
   shapeType=Capsule
   length=10.1631
   radius=8.8032
   shapeName=dyn_jacket_front_l_02_capsule
   center=<5.8532 4.3338 8.1821>
   rotation=<rot 0 -179.439 -107.2612>
  }]
 dyn_jacket_front_l_03 [{
   shapeType=Capsule
   length=7.5312
   radius=10
   shapeName=dyn_jacket_front_l_03_capsule
   center=<1.2818 7.9901 6.7878>
   rotation=<rot 15.008 -89.5619 90.1133>
  }]
 dyn_jacket_front_l_04 [{
   shapeType=Capsule
   length=8.3125
   radius=9.4062
   shapeName=dyn_jacket_front_l_04_capsule
   center=<-0.1834 2.6446 3.8233>
   rotation=<rot 12.4672 -86.6972 90.7501>
  }]
 dyn_jacket_front_l_05 [{
   shapeType=Capsule
   length=13.7812
   radius=9.9688
   shapeName=dyn_jacket_front_l_05_capsule
   center=<7.0726 2.7442 4.0728>
   rotation=<rot -11.3594 91.8849 89.6288>
  }]
 foot_l [{shapeType=Capsule, length=14.296, radius=8.7007, center=<-2.958 7.342 0.0415>, rotation=<rot 0 8.9507 -89.8621>}]
 hand_l [{shapeType=Capsule, length=12.4252, radius=4.8152, center=<3.9402 -1.7616 0.0109>, rotation=<rot 0.0001 85.5197 -84.3836>}]
 hand_r []
 pelvis [{shapeType=Capsule, length=1.4608, radius=12.2707}]
 thigh_l [{
   shapeType=Capsule
   length=11.1875
   radius=11.25
   shapeName=thigh_l_capsule
   center=<-23.9514 -0.7594 -1.4254>
   rotation=<rot -77.0299 -0.0001 0>
  }]
 thumb_01_r [{shapeType=Capsule, length=6.0302, radius=2, shapeName=thumb_01_r_capsule, center=<-2 0 0>, rotation=<rot 90 0 0>}]
 thumb_02_r [{shapeType=Capsule, length=2.2439, radius=2, shapeName=thumb_02_r_capsule, center=<-2 0 0>, rotation=<rot 90 0 0>}]
 thumb_03_r [{shapeType=Capsule, length=1.608, radius=2, shapeName=thumb_03_r_capsule, center=<-2 0 0>, rotation=<rot 90 0 0>}]
}
constraints [
 {bone1Name=hand_l, bone2Name=pelvis, swing1LimitDegrees=45, swing2LimitDegrees=45, twistLimitDegrees=45}
 {bone1Name=calf_l, bone2Name=pelvis, swing1LimitDegrees=45, swing2LimitDegrees=45, twistLimitDegrees=45}
 {bone1Name=foot_l, bone2Name=calf_l, swing1LimitDegrees=45, swing2LimitDegrees=45, twistLimitDegrees=45}
 {bone1Name=hand_r, bone2Name=pelvis, swing1LimitDegrees=45, swing2LimitDegrees=45, twistLimitDegrees=45}
 {
  bone1Name=thumb_01_r
  bone2Name=hand_r
  swing1LimitDegrees=45
  swing1Motion=Free
  swing2LimitDegrees=45
  swing2Motion=Free
  twistLimitDegrees=45
  twistMotion=Free
 }
 {
  bone1Name=thumb_02_r
  bone2Name=thumb_01_r
  swing1LimitDegrees=45
  swing1Motion=Free
  swing2LimitDegrees=45
  swing2Motion=Free
  twistLimitDegrees=45
  twistMotion=Free
 }
 {
  bone1Name=thumb_03_r
  bone2Name=thumb_02_r
  swing1LimitDegrees=45
  swing1Motion=Free
  swing2LimitDegrees=45
  swing2Motion=Free
  twistLimitDegrees=45
  twistMotion=Free
 }
 {
  bone1Name=dyn_jacket_front_l_03
  bone2Name=dyn_jacket_front_l_02
  swing1LimitDegrees=45
  swing1Motion=Free
  swing2LimitDegrees=45
  swing2Motion=Free
  twistLimitDegrees=45
  twistMotion=Free
 }
 {
  bone1Name=dyn_jacket_front_l_04
  bone2Name=dyn_jacket_front_l_03
  swing1LimitDegrees=45
  swing1Motion=Free
  swing2LimitDegrees=45
  swing2Motion=Free
  twistLimitDegrees=45
  twistMotion=Free
 }
 {
  bone1Name=dyn_jacket_front_l_05
  bone2Name=dyn_jacket_front_l_04
  swing1LimitDegrees=45
  swing1Motion=Free
  swing2LimitDegrees=45
  swing2Motion=Free
  twistLimitDegrees=45
  twistMotion=Free
 }
 {
  bone1Name=dyn_jacket_front_l_02
  bone2Name=pelvis
  swing1LimitDegrees=45
  swing1Motion=Free
  swing2LimitDegrees=45
  swing2Motion=Free
  twistLimitDegrees=45
  twistMotion=Free
 }
]
