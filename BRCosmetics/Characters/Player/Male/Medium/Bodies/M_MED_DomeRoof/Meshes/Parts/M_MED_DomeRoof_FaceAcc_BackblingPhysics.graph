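# Physics asset for the male medium DomeRoof face accessory, defining secondary-motion physics for its ponytail, hair, head cloth and hat bones.
# Spine, clavicle and head capsules form the body volumes, linked by free-rotation constraints.
# Each dynamic chain hangs from a MSTR root capsule with free rotation on the head, followed by capsules with narrow swing and twist limits.
# Both "hair_a" chains fork at segment 02 into 03 and 04.
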PhysicsAsset /BRCosmetics/Characters/Player/Male/Medium/Bodies/M_MED_DomeRoof/Meshes/Parts/M_MED_DomeRoof_FaceAcc_BackblingPhysics
bodies {
 clavicle_l [{
   shapeType=Capsule
   length=11.3175
   radius=10.1925
   shapeName=clavicle_l_capsule
   center=<6.6996 0 -1.2467>
   rotation=<rot -0.0002 -90 101.6376>
  }]
 clavicle_r [{
   shapeType=Capsule
   length=11.3175
   radius=10.1925
   shapeName=clavicle_l_capsule
   center=<-6.6996 0 1.2467>
   rotation=<rot -0.0002 -90 -78.3624>
  }]
 dyn_hair_a_l_01 [{shapeType=Capsule, length=4.4121, radius=2.5739, shapeName=dyn_hair_a_l_01_capsule, center=<1.5 0 2>, rotation=<rot 0 0 90>}]
 dyn_hair_a_l_01_MSTR [{shapeType=Capsule, length=0.505, radius=0.505, shapeName=dyn_hair_a_l_01_MSTR_capsule, rotation=<rot 0 -90 0>}]
 dyn_hair_a_l_02 [{shapeType=Capsule, length=4.3332, radius=3.3363, shapeName=dyn_hair_a_l_02_capsule, center=<1.5 0 2>, rotation=<rot 0 0 90>}]
 dyn_hair_a_l_03 [{shapeType=Capsule, length=4.4622, radius=3.0752, shapeName=dyn_hair_a_l_03_capsule, center=<1.5 0 2>, rotation=<rot 0 0 90>}]
 dyn_hair_a_l_04 [{shapeType=Capsule, length=4.2982, radius=2.5932, shapeName=dyn_hair_a_l_04_capsule, center=<1.5 0 2>, rotation=<rot 0 0 90>}]
 dyn_hair_a_r_01 [{shapeType=Capsule, length=4.4121, radius=2.5739, shapeName=dyn_hair_a_l_01_capsule, center=<-1.5 0 -2>, rotation=<rot 0 0 -90>}]
 dyn_hair_a_r_01_MSTR [{shapeType=Capsule, length=0.505, radius=0.505, shapeName=dyn_hair_a_r_01_MSTR_capsule, rotation=<rot 0 -90 0>}]
 dyn_hair_a_r_02 [{shapeType=Capsule, length=4.3332, radius=3.3363, shapeName=dyn_hair_a_l_02_capsule, center=<-1.5 0 -2>, rotation=<rot 0 0 -90>}]
 dyn_hair_a_r_03 [{shapeType=Capsule, length=4.4622, radius=3.0752, shapeName=dyn_hair_a_l_03_capsule, center=<-1.5 0 -2>, rotation=<rot 0 0 -90>}]
 dyn_hair_a_r_04 [{shapeType=Capsule, length=4.2982, radius=2.5932, shapeName=dyn_hair_a_l_04_capsule, center=<-1.5 0 -2>, rotation=<rot 0 0 -90>}]
 dyn_hair_b_l_01 [{shapeType=Capsule, length=3.2015, radius=3.1998, shapeName=dyn_hair_b_l_01_capsule, center=<1.5 0 2>, rotation=<rot 0 0 90>}]
 dyn_hair_b_l_01_MSTR [{shapeType=Capsule, length=0.505, radius=0.505, shapeName=dyn_hair_b_l_01_MSTR_capsule, rotation=<rot 0 -90 0>}]
 dyn_hair_b_l_02 [{shapeType=Capsule, length=0.505, radius=0.505, shapeName=dyn_hair_b_l_02_capsule, center=<1.5 0 2>, rotation=<rot 0 0 90>}]
 dyn_hair_b_r_01 [{shapeType=Capsule, length=3.2015, radius=3.1998, shapeName=dyn_hair_b_l_01_capsule, center=<-1.5 0 -2>, rotation=<rot 0 0 -90>}]
 dyn_hair_b_r_01_MSTR [{shapeType=Capsule, length=0.505, radius=0.505, shapeName=dyn_hair_b_r_01_MSTR_capsule, rotation=<rot 0 -90 0>}]
 dyn_hair_b_r_02 [{shapeType=Capsule, length=0.505, radius=0.505, shapeName=dyn_hair_b_l_02_capsule, center=<-1.5 0 -2>, rotation=<rot 0 0 -90>}]
 dyn_hat_rear_01 [{shapeType=Capsule, length=8, radius=4, shapeName=dyn_hat_rear_01_capsule, center=<2 0 0>, rotation=<rot 0 0 90>}]
 dyn_hat_rear_01_MSTR [{shapeType=Capsule, length=0.505, radius=0.505, shapeName=dyn_hat_rear_01_MSTR_capsule, rotation=<rot 0 -90 0>}]
 dyn_hat_rear_02 [{shapeType=Capsule, length=8, radius=4, shapeName=dyn_hat_rear_02_capsule, center=<2 0 0>, rotation=<rot 0 0 90>}]
 dyn_hat_side_l_01 [{shapeType=Capsule, length=8, radius=4, shapeName=dyn_hat_side_l_01_capsule, center=<2 0 0>, rotation=<rot 0 0 90>}]
 dyn_hat_side_l_01_MSTR [{shapeType=Capsule, length=0.505, radius=0.505, shapeName=dyn_hat_side_l_01_MSTR_capsule, rotation=<rot 0 -90 0>}]
 dyn_hat_side_l_02 [{shapeType=Capsule, length=8, radius=4, shapeName=dyn_hat_side_l_02_capsule, center=<2 0 0>, rotation=<rot 0 0 90>}]
 dyn_hat_side_r_01 [{shapeType=Capsule, length=8, radius=4, shapeName=dyn_hat_side_r_01_capsule, center=<2 0 0>, rotation=<rot 0 0 90>}]
 dyn_hat_side_r_01_MSTR [{shapeType=Capsule, length=0.505, radius=0.505, shapeName=dyn_hat_side_r_01_MSTR_capsule, rotation=<rot 0 -90 0>}]
 dyn_hat_side_r_02 [{shapeType=Capsule, length=8, radius=4, shapeName=dyn_hat_side_r_02_capsule, center=<2 0 0>, rotation=<rot 0 0 90>}]
 dyn_head_cloth_01 [{shapeType=Capsule, length=2.625, radius=1.375, shapeName=dyn_head_cloth_01_capsule, center=<5 0 0>, rotation=<rot 0 0 -90>}]
 dyn_head_cloth_01_MSTR [{shapeType=Capsule, length=5.6925, radius=1, shapeName=dyn_head_cloth_01_MSTR_capsule, rotation=<rot 0 0 90>}]
 dyn_head_cloth_02 [{shapeType=Capsule, length=3.141, radius=1.7475, shapeName=dyn_head_cloth_02_capsule, center=<2.3 0 0.4>, rotation=<rot 0 0 -90>}]
 dyn_head_cloth_03 [{
   shapeType=Capsule
   length=2.9395
   radius=2.4812
   shapeName=dyn_head_cloth_03_capsule
   center=<1.7001 -0.4043 0.3086>
   rotation=<rot 0 0 -93.51>
  }]
 dyn_ponytail_01 [{shapeType=Capsule, length=8.3524, radius=2, shapeName=dyn_ponytail_01_capsule, center=<4 0 1>, rotation=<rot 0 90 90>}]
 dyn_ponytail_01_MSTR [{shapeType=Capsule, length=0.505, radius=0.505, shapeName=dyn_ponytail_01_MSTR_capsule, rotation=<rot 0 -90 0>}]
 dyn_ponytail_02 [{shapeType=Capsule, length=2.4098, radius=2, shapeName=dyn_ponytail_02_capsule, center=<4 0 1>, rotation=<rot 0 90 90>}]
 dyn_ponytail_03 [{shapeType=Capsule, length=2.7436, radius=2, shapeName=dyn_ponytail_03_capsule, center=<4 0 1>, rotation=<rot 0 90 90>}]
 dyn_ponytail_04 [{shapeType=Capsule, length=2.0106, radius=2, shapeName=dyn_ponytail_04_capsule, center=<4 0 1>, rotation=<rot 0 90 90>}]
 dyn_ponytail_05 [{shapeType=Capsule, length=3.9214, radius=2.875, shapeName=dyn_ponytail_05_capsule, center=<5.3531 0 -0.1212>, rotation=<rot 0 90 71.4475>}]
 head [{shapeType=Capsule, length=7.5675, radius=7.3605, shapeName=head_capsule, center=<0 4.4697 0>, rotation=<rot 90 0 0>}]
 spine_03 [{shapeType=Capsule, length=11.5675, radius=13.0675, shapeName=spine_03_capsule, center=<7.3788 0 0>, rotation=<rot 0 -90 0>}]
 spine_05 [{shapeType=Capsule, length=14.0675, radius=16.0675, shapeName=spine_05_capsule, center=<0.3891 4.486 0>, rotation=<rot 0 -62.8684 0>}]
}
constraints [
 {bone1Name=dyn_ponytail_01, bone2Name=dyn_ponytail_01_MSTR, swing1LimitDegrees=4, swing2LimitDegrees=6, twistLimitDegrees=4}
 {bone1Name=dyn_ponytail_02, bone2Name=dyn_ponytail_01, swing1LimitDegrees=4, swing2LimitDegrees=6, twistLimitDegrees=4}
 {bone1Name=dyn_ponytail_03, bone2Name=dyn_ponytail_02, swing1LimitDegrees=4, swing2LimitDegrees=6, twistLimitDegrees=4}
 {bone1Name=dyn_ponytail_04, bone2Name=dyn_ponytail_03, swing1LimitDegrees=4, swing2LimitDegrees=6, twistLimitDegrees=4}
 {bone1Name=dyn_ponytail_05, bone2Name=dyn_ponytail_04, swing1LimitDegrees=4, swing2LimitDegrees=6, twistLimitDegrees=4}
 {
  bone1Name=dyn_ponytail_01_MSTR
  bone2Name=head
  swing1LimitDegrees=45
  swing1Motion=Free
  swing2LimitDegrees=45
  swing2Motion=Free
  twistLimitDegrees=45
  twistMotion=Free
 }
 {
  bone1Name=dyn_head_cloth_01_MSTR
  bone2Name=head
  swing1LimitDegrees=45
  swing1Motion=Free
  swing2LimitDegrees=45
  swing2Motion=Free
  twistLimitDegrees=45
  twistMotion=Free
 }
 {
  bone1Name=head
  bone2Name=spine_05
  swing1LimitDegrees=45
  swing1Motion=Free
  swing2LimitDegrees=45
  swing2Motion=Free
  twistLimitDegrees=45
  twistMotion=Free
 }
 {
  bone1Name=dyn_hair_a_l_01_MSTR
  bone2Name=head
  swing1LimitDegrees=45
  swing1Motion=Free
  swing2LimitDegrees=45
  swing2Motion=Free
  twistLimitDegrees=45
  twistMotion=Free
 }
 {bone1Name=dyn_hair_a_l_02, bone2Name=dyn_hair_a_l_01, swing1LimitDegrees=10, swing2LimitDegrees=10, twistLimitDegrees=12.01}
 {bone1Name=dyn_hair_a_l_03, bone2Name=dyn_hair_a_l_02, swing1LimitDegrees=10, swing2LimitDegrees=10, twistLimitDegrees=12.01}
 {bone1Name=dyn_hair_a_l_04, bone2Name=dyn_hair_a_l_02, swing1LimitDegrees=10, swing2LimitDegrees=10, twistLimitDegrees=12.01}
 {
  bone1Name=dyn_hair_a_r_01_MSTR
  bone2Name=head
  swing1LimitDegrees=45
  swing1Motion=Free
  swing2LimitDegrees=45
  swing2Motion=Free
  twistLimitDegrees=45
  twistMotion=Free
 }
 {
  bone1Name=dyn_hair_b_l_01_MSTR
  bone2Name=head
  swing1LimitDegrees=45
  swing1Motion=Free
  swing2LimitDegrees=45
  swing2Motion=Free
  twistLimitDegrees=45
  twistMotion=Free
 }
 {bone1Name=dyn_hair_b_l_01, bone2Name=dyn_hair_b_l_01_MSTR, swing1LimitDegrees=10, swing2LimitDegrees=10, twistLimitDegrees=12.01}
 {bone1Name=dyn_hair_b_l_02, bone2Name=dyn_hair_b_l_01, swing1LimitDegrees=10, swing2LimitDegrees=10, twistLimitDegrees=12.01}
 {
  bone1Name=dyn_hair_b_r_01_MSTR
  bone2Name=head
  swing1LimitDegrees=45
  swing1Motion=Free
  swing2LimitDegrees=45
  swing2Motion=Free
  twistLimitDegrees=45
  twistMotion=Free
 }
 {bone1Name=dyn_hair_a_r_02, bone2Name=dyn_hair_a_r_01, swing1LimitDegrees=10, swing2LimitDegrees=10, twistLimitDegrees=12.01}
 {bone1Name=dyn_hair_a_r_03, bone2Name=dyn_hair_a_r_02, swing1LimitDegrees=10, swing2LimitDegrees=10, twistLimitDegrees=12.01}
 {bone1Name=dyn_hair_a_r_04, bone2Name=dyn_hair_a_r_02, swing1LimitDegrees=10, swing2LimitDegrees=10, twistLimitDegrees=12.01}
 {bone1Name=dyn_hair_a_l_01, bone2Name=dyn_hair_a_l_01_MSTR, swing1LimitDegrees=10, swing2LimitDegrees=10, twistLimitDegrees=12.01}
 {bone1Name=dyn_hair_a_r_01, bone2Name=dyn_hair_a_r_01_MSTR, swing1LimitDegrees=10, swing2LimitDegrees=10, twistLimitDegrees=12.01}
 {bone1Name=dyn_hair_b_r_01, bone2Name=dyn_hair_b_r_01_MSTR, swing1LimitDegrees=10, swing2LimitDegrees=10, twistLimitDegrees=12.01}
 {bone1Name=dyn_hair_b_r_02, bone2Name=dyn_hair_b_r_01, swing1LimitDegrees=10, swing2LimitDegrees=10, twistLimitDegrees=12.01}
 {
  bone1Name=spine_05
  bone2Name=spine_03
  swing1LimitDegrees=45
  swing1Motion=Free
  swing2LimitDegrees=45
  swing2Motion=Free
  twistLimitDegrees=45
  twistMotion=Free
 }
 {bone1Name=dyn_head_cloth_01, bone2Name=dyn_head_cloth_01_MSTR, swing1LimitDegrees=22, swing2LimitDegrees=22, twistLimitDegrees=22}
 {bone1Name=dyn_head_cloth_02, bone2Name=dyn_head_cloth_01, swing1LimitDegrees=8, swing2LimitDegrees=8, twistLimitDegrees=8}
 {bone1Name=dyn_head_cloth_03, bone2Name=dyn_head_cloth_02, swing1LimitDegrees=8, swing2LimitDegrees=8, twistLimitDegrees=8}
 {
  bone1Name=clavicle_l
  bone2Name=spine_05
  swing1LimitDegrees=45
  swing1Motion=Free
  swing2LimitDegrees=45
  swing2Motion=Free
  twistLimitDegrees=45
  twistMotion=Free
 }
 {
  bone1Name=clavicle_r
  bone2Name=spine_05
  swing1LimitDegrees=45
  swing1Motion=Free
  swing2LimitDegrees=45
  swing2Motion=Free
  twistLimitDegrees=45
  twistMotion=Free
 }
 {
  bone1Name=dyn_hat_rear_01_MSTR
  bone2Name=head
  swing1LimitDegrees=45
  swing1Motion=Free
  swing2LimitDegrees=45
  swing2Motion=Free
  twistLimitDegrees=45
  twistMotion=Free
 }
 {bone1Name=dyn_hat_rear_01, bone2Name=dyn_hat_rear_01_MSTR, swing1LimitDegrees=5, swing2LimitDegrees=5, twistLimitDegrees=1}
 {bone1Name=dyn_hat_rear_02, bone2Name=dyn_hat_rear_01, swing1LimitDegrees=5, swing2LimitDegrees=5, twistLimitDegrees=1}
 {
  bone1Name=dyn_hat_side_l_01_MSTR
  bone2Name=head
  swing1LimitDegrees=45
  swing1Motion=Free
  swing2LimitDegrees=45
  swing2Motion=Free
  twistLimitDegrees=45
  twistMotion=Free
 }
 {bone1Name=dyn_hat_side_l_01, bone2Name=dyn_hat_side_l_01_MSTR, swing1LimitDegrees=5, swing2LimitDegrees=5, twistLimitDegrees=1}
 {bone1Name=dyn_hat_side_l_02, bone2Name=dyn_hat_side_l_01, swing1LimitDegrees=5, swing2LimitDegrees=5, twistLimitDegrees=1}
 {
  bone1Name=dyn_hat_side_r_01_MSTR
  bone2Name=head
  swing1LimitDegrees=45
  swing1Motion=Free
  swing2LimitDegrees=45
  swing2Motion=Free
  twistLimitDegrees=45
  twistMotion=Free
 }
 {bone1Name=dyn_hat_side_r_01, bone2Name=dyn_hat_side_r_01_MSTR, swing1LimitDegrees=5, swing2LimitDegrees=5, twistLimitDegrees=1}
 {bone1Name=dyn_hat_side_r_02, bone2Name=dyn_hat_side_r_01, swing1LimitDegrees=5, swing2LimitDegrees=5, twistLimitDegrees=1}
]
