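# Control Rig for the DomeRoof male character's face accessory mesh, correcting hat, ponytail and head cloth bones.
# Three spherical pose readers measure head orientation relative to spine_05, with debug drawing enabled.
# Aim constraints then target the ponytail and head cloth master bones at spine_05.
# Finally, the reader outputs weight three transform modifiers: two adjust hat bones, while the third targets no bone and is effectively a placeholder.
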ControlRigBlueprint /BRCosmetics/Characters/Player/Male/Medium/Bodies/M_MED_DomeRoof/Meshes/Parts/M_MED_DomeRoof_FaceAcc_CtrlRig
graphs [
 RigVMFunctionLibrary {links [], nodes []}
 RigVMFunction_Sequence_1_ContainedGraph {nodes [
   n1 = RigVMFunctionEntryNode()
   RigVMFunction_Sequence = &RigVMFunction_Sequence(ExecuteContext=@n1.ExecuteContext)
   n2 = &RigVMFunction_Sequence(ExecuteContext=@RigVMFunction_Sequence.B)
   RigVMFunctionReturnNode(A=@RigVMFunction_Sequence.A, B=@n2.A, C=@n2.B)
  ]}
 RigVMModel {nodes [
   RigUnit_BeginExecution = &RigUnit_BeginExecution()
   RigVMFunction_Sequence = RigVMAggregateNode(ExecuteContext=@RigUnit_BeginExecution.ExecutePin)
   n1 = &RigUnit_SphericalPoseReader(
    ActiveRegionScaleFactors=(PositiveWidth=1,NegativeWidth=1,PositiveHeight=1,NegativeHeight=1)
    ActiveRegionSize=0.1
    Debug=(bDrawDebug=true,bDraw2D=False,bDrawLocalAxes=False,DebugScale=25,DebugSegments=20,DebugThickness=0.25)
    DriverAxis=(X=1,Y=0,Z=0)
    DriverItem=(Type=Bone,Name="head")
    FalloffRegionScaleFactors=(PositiveWidth=1,NegativeWidth=1,PositiveHeight=1,NegativeHeight=1)
    FalloffSize=0.3
    FlipHeightScaling=False
    FlipWidthScaling=False
    OptionalParentItem=(Type=Bone,Name="spine_05")
    RotationOffset=(X=-90,Y=0,Z=20)
    ExecutePin=@RigVMFunction_Sequence.A
   )
   > n2 = &RigUnit_SphericalPoseReader(
    ActiveRegionScaleFactors=(PositiveWidth=1,NegativeWidth=1,PositiveHeight=1,NegativeHeight=1)
    ActiveRegionSize=0.1
    Debug=(bDrawDebug=true,bDraw2D=False,bDrawLocalAxes=False,DebugScale=25,DebugSegments=20,DebugThickness=0.25)
    DriverAxis=(X=1,Y=0,Z=0)
    DriverItem=(Type=Bone,Name="head")
    FalloffRegionScaleFactors=(PositiveWidth=1,NegativeWidth=1,PositiveHeight=1,NegativeHeight=1)
    FalloffSize=0.35
    FlipHeightScaling=False
    FlipWidthScaling=False
    OptionalParentItem=(Type=Bone,Name="spine_05")
    RotationOffset=(X=-20,Y=-15,Z=0)
   )
   > n3 = &RigUnit_SphericalPoseReader(
    ActiveRegionScaleFactors=(PositiveWidth=1,NegativeWidth=1,PositiveHeight=1,NegativeHeight=1)
    ActiveRegionSize=0.1
    Debug=(bDrawDebug=true,bDraw2D=False,bDrawLocalAxes=False,DebugScale=25,DebugSegments=20,DebugThickness=0.25)
    DriverAxis=(X=1,Y=0,Z=0)
    DriverItem=(Type=Bone,Name="head")
    FalloffRegionScaleFactors=(PositiveWidth=1,NegativeWidth=1,PositiveHeight=1,NegativeHeight=1)
    FalloffSize=0.3
    FlipHeightScaling=False
    FlipWidthScaling=False
    OptionalParentItem=(Type=Bone,Name="spine_05")
    RotationOffset=(X=180,Y=0,Z=20)
   )
   &RigUnit_AimConstraintLocalSpaceOffset(
    Child=(Type=Bone,Name="dyn_ponytail_01_MSTR")
    Filter=(bX=True,bY=True,bZ=True)
    Parents=((Item=(Type=Bone,Name="spine_05"),Weight=1))
    Weight=1
    WorldUp=(Target=(X=0,Y=0,Z=1),Kind=Location,Space=(Type=Bone,Name="spine_05"))
    ExecutePin=@RigVMFunction_Sequence.B
   )
   > &RigUnit_AimConstraintLocalSpaceOffset(
    Child=(Type=Bone,Name="dyn_head_cloth_01_MSTR")
    Filter=(bX=True,bY=True,bZ=True)
    Parents=((Item=(Type=Bone,Name="spine_05"),Weight=1))
    Weight=1
    WorldUp=(Target=(X=0,Y=0,Z=1),Kind=Location,Space=(Type=Bone,Name="spine_05"))
   )
   &RigUnit_ModifyTransforms(
    ItemToModify="((Item=(Type=Bone,Name=\"deform_hat_overall\"),Transform=(Rotation=(X=0,Y=0,Z=0.0533,W=0.9986),Translation=(X=0,Y=0,Z=0),Scale3D=(X=1,Y=1,Z=1))),(Item=(Type=Bone,Name=\"dyn_hat_MSTR\"),Transform=(Rotation=(X=0,Y=0,Z=0.1343,W=0.9909),Translation=(X=0,Y=0,Z=0),S ...(truncated)"
    WeightMinimum=0
    ExecutePin=@RigVMFunction_Sequence.C
    Weight=@n1.OutputParam
   )
   > &RigUnit_ModifyTransforms(
    ItemToModify="((Item=(Type=Bone,Name=\"dyn_hat_MSTR\"),Transform=(Rotation=(X=-0.0118,Y=0.2168,Z=0.2891,W=0.9323),Translation=(X=0,Y=0,Z=0),Scale3D=(X=1,Y=1,Z=1))),(Item=(Type=Bone,Name=\"deform_hat_overall\"),Transform=(Rotation=(X=0,Y=0,Z=0.0903,W=0.9959),Translation=(X=0 ...(truncated)"
    WeightMinimum=0
    Weight=@n2.OutputParam
   )
   > &RigUnit_ModifyTransforms(
    ItemToModify=((Item=(Type=Bone,Name="None"),Transform=(Rotation=(X=0,Y=0,Z=0,W=1),Translation=(X=0,Y=0,Z=0),Scale3D=(X=1,Y=1,Z=1))))
    WeightMinimum=0
    Weight=@n3.OutputParam
   )
  ]}
]
hierarchy [
 root {loc=<0 0 0>, rot=<0 0 0 1>}
 pelvis {parent=root, loc=<0 -0.7071 93.1191>, rot=<0.0127 0.707 -0.0127 -0.707>}
 spine_01 {parent=pelvis, loc=<6.7494 -0.6388 0>, rot=<0 0 0.034 -0.9994>}
 spine_02 {parent=spine_01, loc=<7.2518 -0.035 0>, rot=<0 0 0.0238 -0.9997>}
 spine_03 {parent=spine_02, loc=<6.0104 0 0>, rot=<0 0 0.0149 -0.9999>}
 spine_04 {parent=spine_03, loc=<8.1024 0 0>, rot=<0 0 -0.0132 -0.9999>}
 spine_05 {parent=spine_04, loc=<10.0146 0 0>, rot=<0 0 -0.0015 -1>}
 neck_01 {parent=spine_05, loc=<11.3138 0.1529 0>, rot=<0 0 -0.2003 -0.9797>}
 neck_02 {parent=neck_01, loc=<4.5627 0 0>, rot=<0 0 0.0504 -0.9987>}
 head {parent=neck_02, loc=<5.3914 0 0>, rot=<0 0 0.115 -0.9934>}
 dyn_ponytail_01_MSTR {parent=head, loc=<0.9427 -4.2851 -1.982>, rot=<0.2057 -0.6434 0.6919 -0.2551>}
 dyn_ponytail_01 {parent=dyn_ponytail_01_MSTR, loc=<0 0 0>, rot=<0 0 0 1>}
 dyn_ponytail_02 {parent=dyn_ponytail_01, loc=<5.2 0 0>, rot=<-0.0001 -0.0266 0.0055 -0.9996>}
 dyn_ponytail_03 {parent=dyn_ponytail_02, loc=<5.2 0 0>, rot=<0.0021 -0.0651 0.0216 -0.9976>}
 dyn_ponytail_04 {parent=dyn_ponytail_03, loc=<5.2 0 0>, rot=<0.0018 -0.0498 0.0174 -0.9986>}
 dyn_ponytail_05 {parent=dyn_ponytail_04, loc=<4.6545 0 0>, rot=<0 0 0 1>}
 dyn_head_cloth_01_MSTR {parent=head, loc=<0.7603 -4.5707 2.924>, rot=<0.1964 -0.6052 0.7424 -0.2097>}
 dyn_head_cloth_01 {parent=dyn_head_cloth_01_MSTR, loc=<0 0 0>, rot=<0 0 0 1>}
 dyn_head_cloth_02 {parent=dyn_head_cloth_01, loc=<7 0 0>, rot=<0 0 0 1>}
 dyn_head_cloth_03 {parent=dyn_head_cloth_02, loc=<7 0 0>, rot=<0 0 0 1>}
 dyn_hair_a_l_01_MSTR {parent=head, loc=<3.2193 3.327 -7.5241>, rot=<0.0352 -0.9928 0.0416 -0.1062>}
 dyn_hair_a_l_01 {parent=dyn_hair_a_l_01_MSTR, loc=<0 0 0>, rot=<0 0 0 1>}
 dyn_hair_a_l_02 {parent=dyn_hair_a_l_01, loc=<5.2 0 0>, rot=<-0.0001 -0.0266 0.0055 -0.9996>}
 dyn_hair_a_l_03 {parent=dyn_hair_a_l_02, loc=<3.6349 0 0>, rot=<-0.0014 0.029 -0.0048 -0.9996>}
 dyn_hair_a_l_04 {parent=dyn_hair_a_l_02, loc=<2.3452 -2.1083 0>, rot=<0 0.2192 0 -0.9757>}
 dyn_hair_b_l_01_MSTR {parent=head, loc=<0.1027 -2.0922 -6.1618>, rot=<0.0669 -0.9248 0.2984 -0.2261>}
 dyn_hair_b_l_01 {parent=dyn_hair_b_l_01_MSTR, loc=<0 0 0>, rot=<0 0 0 1>}
 dyn_hair_b_l_02 {parent=dyn_hair_b_l_01, loc=<2.7822 0 0>, rot=<-0.0001 -0.0266 0.0055 -0.9996>}
 dyn_hair_a_r_01_MSTR {parent=head, loc=<3.2194 3.327 7.5241>, rot=<0.9928 0.0352 -0.1062 -0.0416>}
 dyn_hair_a_r_01 {parent=dyn_hair_a_r_01_MSTR, loc=<0 0 0>, rot=<0 0 0 1>}
 dyn_hair_a_r_02 {parent=dyn_hair_a_r_01, loc=<-5.2 0 0>, rot=<-0.0001 -0.0266 0.0055 -0.9996>}
 dyn_hair_a_r_03 {parent=dyn_hair_a_r_02, loc=<-3.6349 0 0>, rot=<-0.0014 0.029 -0.0048 -0.9996>}
 dyn_hair_a_r_04 {parent=dyn_hair_a_r_02, loc=<-2.3448 2.1083 -0.0001>, rot=<0 0.2192 0 -0.9757>}
 dyn_hair_b_r_01_MSTR {parent=head, loc=<0.1031 -2.0922 6.1618>, rot=<0.9248 0.0669 -0.2261 -0.2984>}
 dyn_hair_b_r_01 {parent=dyn_hair_b_r_01_MSTR, loc=<0 0 0>, rot=<0 0 0 1>}
 dyn_hair_b_r_02 {parent=dyn_hair_b_r_01, loc=<-2.7826 0 0.0002>, rot=<-0.0001 -0.0266 0.0055 -0.9996>}
 deform_hat_overall {parent=head, loc=<4.9016 2.3986 -0.1924>, rot=<0 0 0 1>}
 dyn_hat_MSTR {parent=deform_hat_overall, loc=<0 0 0>, rot=<0 0 0 1>}
 dyn_hat_rear_01_MSTR {parent=dyn_hat_MSTR, loc=<-5.1766 -9.9545 0.1924>, rot=<0.3786 -0.5397 0.6139 -0.4342>}
 dyn_hat_rear_01 {parent=dyn_hat_rear_01_MSTR, loc=<0 0 0>, rot=<0 0 0 1>}
 dyn_hat_rear_02 {parent=dyn_hat_rear_01, loc=<6.6756 0 0>, rot=<0 0 0 1>}
 dyn_hat_side_l_01_MSTR {parent=dyn_hat_MSTR, loc=<-2.9284 -6.5015 -8.6385>, rot=<0.1271 -0.6444 0.3558 -0.6649>}
 dyn_hat_side_l_01 {parent=dyn_hat_side_l_01_MSTR, loc=<0 0 0>, rot=<0 0 0 1>}
 dyn_hat_side_l_02 {parent=dyn_hat_side_l_01, loc=<5.256 0 0>, rot=<-0.0109 0.1792 0.0403 -0.9829>}
 dyn_hat_side_r_01_MSTR {parent=dyn_hat_MSTR, loc=<-5.1402 -6.0911 7.8805>, rot=<0.5636 -0.3419 0.7367 -0.1507>}
 dyn_hat_side_r_01 {parent=dyn_hat_side_r_01_MSTR, loc=<0 0 0>, rot=<0 0 0 1>}
 dyn_hat_side_r_02 {parent=dyn_hat_side_r_01, loc=<7 0 0>, rot=<0 0 0 1>}
 dyn_hat_feather_root {parent=deform_hat_overall, loc=<1.8762 2.0925 8.5699>, rot=<0.0029 -0.7071 -0.0029 -0.7071>}
 dyn_hat_feather_a_01 {parent=dyn_hat_feather_root, loc=<5.4597 10.3913 12.9651>, rot=<0.2052 0.16 -0.3611 -0.8955>}
 dyn_hat_feather_a_02 {parent=dyn_hat_feather_a_01, loc=<6.0462 0 0>, rot=<0 -0.1736 0 -0.9848>}
 dyn_hat_feather_b_01 {parent=dyn_hat_feather_root, loc=<3.8266 3.3963 11.7126>, rot=<-0.0499 0.5054 0.0846 -0.8573>}
 dyn_hat_feather_b_02 {parent=dyn_hat_feather_b_01, loc=<4.1708 0 0>, rot=<0 0 0.1942 -0.981>}
 dyn_hat_feather_f_01 {parent=dyn_hat_feather_b_01, loc=<2.6158 -1.9558 0.0499>, rot=<0.1985 -0.2254 0.2238 -0.9272>}
 dyn_hat_feather_c_01 {parent=dyn_hat_feather_root, loc=<-0.3867 3.6264 4.8328>, rot=<-0.1454 -0.2681 0.813 0.4961>}
 dyn_hat_feather_d_01 {parent=dyn_hat_feather_root, loc=<2.0482 0.2913 8.7679>, rot=<-0.0875 0.4128 0.3122 -0.8512>}
 dyn_hat_feather_e_01 {parent=dyn_hat_feather_root, loc=<0.7459 6.8701 8.743>, rot=<-0.0932 -0.0192 -0.7846 -0.6126>}
 dyn_hat_feather_e_02 {parent=dyn_hat_feather_e_01, loc=<5.7715 -0.6046 0.4988>, rot=<-0.0837 -0.3537 -0.0241 -0.9313>}
 dyn_hat_feather_e_03 {parent=dyn_hat_feather_e_01, loc=<2.6078 1.626 -0.4882>, rot=<0.2027 -0.2339 -0.383 -0.8704>}
 dyn_hat_feather_h_01 {parent=dyn_hat_feather_root, loc=<6.6513 12.5819 10.4574>, rot=<0.1752 0.0302 -0.4264 -0.8869>}
 dyn_hat_feather_front_01 {parent=dyn_hat_feather_root, loc=<6.5178 6.7893 12.7207>, rot=<0.1161 0.3305 -0.1909 -0.917>}
 dyn_hat_feather_front_a_01 {parent=dyn_hat_feather_front_01, loc=<5.1322 2.4975 -1.8963>, rot=<0.0945 -0.3839 -0.304 -0.8668>}
 dyn_hat_feather_front_a_02 {parent=dyn_hat_feather_front_a_01, loc=<3.7187 0 0>, rot=<-0.0034 -0.173 -0.071 -0.9823>}
 dyn_hat_feather_g_01 {parent=dyn_hat_feather_front_a_01, loc=<3.4598 2.1293 -0.3561>, rot=<-0.0051 0.1438 0.0184 -0.9894>}
 dyn_hat_feather_front_02 {parent=dyn_hat_feather_front_01, loc=<3.3739 -1.2414 0.5886>, rot=<-0.1358 -0.2284 0.0507 -0.9627>}
 dyn_hat_feather_front_b_01 {parent=dyn_hat_feather_front_02, loc=<7.659 -0.5116 -0.6955>, rot=<0.2931 -0.173 -0.0424 -0.9394>}
 dyn_hat_feather_front_b_02 {parent=dyn_hat_feather_front_b_01, loc=<7.8906 0 -1.05>, rot=<0.0148 -0.2849 -0.1726 -0.9428>}
 dyn_hat_feather_front_b_03 {parent=dyn_hat_feather_front_b_01, loc=<7.2826 0.2833 -2.6416>, rot=<-0.0328 -0.4509 -0.1436 -0.8803>}
 dyn_hat_feather_front_b_04 {parent=dyn_hat_feather_front_b_01, loc=<4.9676 5.0354 -1.3318>, rot=<-0.1512 -0.4655 -0.3334 -0.8058>}
 dyn_hat_feather_front_c_01 {parent=dyn_hat_feather_front_02, loc=<6.1016 -2.6624 0.6256>, rot=<0.1094 -0.0702 0.2135 -0.9682>}
 dyn_hat_feather_front_c_02 {parent=dyn_hat_feather_front_c_01, loc=<9.1507 1.0892 -3.3276>, rot=<0.0891 -0.4949 -0.0742 -0.8612>}
 dyn_hat_feather_front_c_03 {parent=dyn_hat_feather_front_c_02, loc=<3.9683 0 0>, rot=<0 -0.1508 0 -0.9886>}
 dyn_hat_feather_front_c_04 {parent=dyn_hat_feather_front_c_01, loc=<6.8936 -1.3351 -0.6414>, rot=<-0.1106 -0.0442 -0.0054 -0.9929>}
 dyn_hat_feather_front_d_01 {parent=dyn_hat_feather_front_01, loc=<3.8348 -3.5832 1.0289>, rot=<-0.1307 -0.1699 0.4477 -0.8681>}
 dyn_hat_feather_front_d_02 {parent=dyn_hat_feather_front_d_01, loc=<6.6679 0 0>, rot=<0.0496 -0.2293 0.14 -0.962>}
 dyn_hat_feather_front_d_03 {parent=dyn_hat_feather_front_d_01, loc=<2.3328 -0.7813 0.5895>, rot=<0.0888 0.0652 0.1191 -0.9868>}
 dyn_hat_feather_i_01 {parent=dyn_hat_feather_front_01, loc=<5.4421 -2.729 1.0468>, rot=<0.3246 -0.1475 0.231 -0.9053>}
 dyn_hat_feather_i_02 {parent=dyn_hat_feather_i_01, loc=<3.6032 0 0>, rot=<0.0002 -0.1711 -0.0062 -0.9852>}
 dyn_hat_front_r_01_MSTR {parent=deform_hat_overall, loc=<-1.2216 4.1813 8.0487>, rot=<0.6181 0.1067 0.6299 0.4579>}
 dyn_hat_front_r_01 {parent=dyn_hat_front_r_01_MSTR, loc=<0 0 0>, rot=<0 0 0 1>}
 FACIAL_C_FacialRoot {parent=head, loc=<2.4232 0.8474 0.0026>, rot=<0.48 -0.5192 -0.48 -0.5192>}
 FACIAL_L_ForeheadOut {parent=FACIAL_C_FacialRoot, loc=<3.8492 -3.3077 6.9695>, rot=<-0.1145 -0.2562 -0.0774 -0.9567>}
 FACIAL_R_ForeheadOut {parent=FACIAL_C_FacialRoot, loc=<-3.8721 -3.327 6.9586>, rot=<-0.1132 0.2561 0.0769 -0.9569>}
 FACIAL_L_CheekInner {parent=FACIAL_C_FacialRoot, loc=<2.059 0.6315 7.0716>, rot=<-0.1097 -0.0909 -0.1959 -0.9702>}
 FACIAL_R_CheekInner {parent=FACIAL_C_FacialRoot, loc=<-2.0603 0.6231 7.0898>, rot=<-0.11 0.0905 0.1957 -0.9703>}
 FACIAL_L_CheekOuter {parent=FACIAL_C_FacialRoot, loc=<3.6595 0.9366 5.431>, rot=<0.068 -0.3756 0.1054 -0.9183>}
 FACIAL_R_CheekOuter {parent=FACIAL_C_FacialRoot, loc=<-3.6692 0.9599 5.4732>, rot=<0.0675 0.3754 -0.1055 -0.9184>}
 FACIAL_L_NasolabialBulge {parent=FACIAL_C_FacialRoot, loc=<3.4546 4.2368 5.2921>, rot=<0.0739 -0.3768 0.0304 -0.9228>}
 FACIAL_R_NasolabialBulge {parent=FACIAL_C_FacialRoot, loc=<-3.4494 4.2309 5.2864>, rot=<0.0739 0.3768 -0.0303 -0.9228>}
 FACIAL_L_NasolabialFurrow {parent=FACIAL_C_FacialRoot, loc=<2.8881 3.0996 7.7113>, rot=<-0.0198 -0.1312 -0.3617 -0.9228>}
 FACIAL_R_NasolabialFurrow {parent=FACIAL_C_FacialRoot, loc=<-2.905 3.1127 7.7227>, rot=<-0.0201 0.1309 0.3616 -0.9229>}
 FACIAL_L_CheekLower {parent=FACIAL_C_FacialRoot, loc=<3.7657 3.9914 4.6147>, rot=<0.0502 -0.4258 0.0651 -0.9011>}
 FACIAL_R_CheekLower {parent=FACIAL_C_FacialRoot, loc=<-3.7585 3.9862 4.6069>, rot=<0.0502 0.4258 -0.0651 -0.9011>}
 FACIAL_C_Nose {parent=FACIAL_C_FacialRoot, loc=<0.0247 0.3319 9.0775>, rot=<-0.1185 -0.0003 0 -0.993>}
 FACIAL_C_NoseLower {parent=FACIAL_C_Nose, loc=<-0.0186 2.802 -0.5793>, rot=<0.1312 0.0001 0 -0.9914>}
 FACIAL_L_Nostril {parent=FACIAL_C_Nose, loc=<1.1419 2.104 -0.3523>, rot=<0.0173 0.1304 0.0023 -0.9913>}
 FACIAL_R_Nostril {parent=FACIAL_C_Nose, loc=<-1.1908 2.1009 -0.3845>, rot=<0.0173 -0.1304 -0.0023 -0.9913>}
 FACIAL_C_NoseTip {parent=FACIAL_C_Nose, loc=<-0.0312 2.1113 0.9003>, rot=<0.0872 0 0 -0.9962>}
 FACIAL_C_MouthUpper {parent=FACIAL_C_FacialRoot, loc=<0.0037 4.543 6.2893>, rot=<0.0174 0 0 -0.9998>}
 FACIAL_C_LipUpper {parent=FACIAL_C_MouthUpper, loc=<-0.0014 -0.0715 2.3838>, rot=<-0.0872 0 0 -0.9962>}
 FACIAL_L_LipUpper {parent=FACIAL_C_MouthUpper, loc=<1.1911 -0.0444 2.1549>, rot=<-0.0778 -0.1601 -0.0379 -0.9833>}
 FACIAL_R_LipUpper {parent=FACIAL_C_MouthUpper, loc=<-1.1975 -0.0359 2.161>, rot=<-0.0778 0.1601 0.0379 -0.9833>}
 FACIAL_L_LipUpperOuter {parent=FACIAL_C_MouthUpper, loc=<1.9851 0.0886 1.5744>, rot=<-0.0984 -0.2293 -0.0983 -0.9634>}
 FACIAL_R_LipUpperOuter {parent=FACIAL_C_MouthUpper, loc=<-1.993 0.1037 1.593>, rot=<-0.0984 0.2293 0.0983 -0.9634>}
 FACIAL_L_LipCorner {parent=FACIAL_C_MouthUpper, loc=<2.56 0.3853 0.8109>, rot=<0 -0.2078 0 -0.9782>}
 FACIAL_R_LipCorner {parent=FACIAL_C_MouthUpper, loc=<-2.5727 0.388 0.8115>, rot=<0 0.2078 0 -0.9782>}
 FACIAL_L_Masseter {parent=FACIAL_C_FacialRoot, loc=<6.0367 1.409 3.0759>, rot=<0.0835 0.1128 0.0062 -0.9901>}
 FACIAL_R_Masseter {parent=FACIAL_C_FacialRoot, loc=<-6.045 1.4225 3.0359>, rot=<0.0812 -0.1129 -0.0066 -0.9903>}
 FACIAL_C_UnderChin {parent=FACIAL_C_FacialRoot, loc=<0.0055 9.2154 3.6354>, rot=<0.5118 0 0 -0.8591>}
 FACIAL_L_UnderChin {parent=FACIAL_C_FacialRoot, loc=<3.4824 7.6772 1.9019>, rot=<0.4155 -0.3535 0.0876 -0.8335>}
 FACIAL_R_UnderChin {parent=FACIAL_C_FacialRoot, loc=<-3.4611 7.6745 1.8406>, rot=<0.4129 0.3536 -0.0867 -0.8348>}
 FACIAL_C_LowerLipRotation {parent=FACIAL_C_FacialRoot, loc=<0.0024 2.1562 1.3875>, rot=<0.3583 0 0 -0.9336>}
 FACIAL_C_MouthLower {parent=FACIAL_C_LowerLipRotation, loc=<0.0014 -1.5054 5.24>, rot=<-0.342 0 0 -0.9397>}
 FACIAL_C_LipLower {parent=FACIAL_C_MouthLower, loc=<-0.002 1.1688 2.1202>, rot=<0.0872 0 0 -0.9962>}
 FACIAL_L_LipLower {parent=FACIAL_C_MouthLower, loc=<1.09 1.1578 1.9436>, rot=<0.0927 -0.1294 0.0272 -0.9869>}
 FACIAL_R_LipLower {parent=FACIAL_C_MouthLower, loc=<-1.0973 1.1611 1.9395>, rot=<0.0927 0.1294 -0.0272 -0.9869>}
 FACIAL_L_LipLowerOuter {parent=FACIAL_C_MouthLower, loc=<1.9471 0.8759 1.3453>, rot=<0.0913 -0.2321 0.0318 -0.9679>}
 FACIAL_R_LipLowerOuter {parent=FACIAL_C_MouthLower, loc=<-1.9557 0.8801 1.341>, rot=<0.0913 0.2321 -0.0318 -0.9679>}
 FACIAL_C_Jaw {parent=FACIAL_C_FacialRoot, loc=<0.0024 2.1562 1.3875>, rot=<0.3583 0 0 -0.9336>}
 FACIAL_C_Jawline {parent=FACIAL_C_Jaw, loc=<0.0001 2.7099 8.1055>, rot=<0.2597 0 0 -0.9657>}
 FACIAL_L_ChinSide {parent=FACIAL_C_Jaw, loc=<3.1373 0.4729 7.143>, rot=<-0.2199 -0.3738 0.0502 -0.8997>}
 FACIAL_R_ChinSide {parent=FACIAL_C_Jaw, loc=<-3.1383 0.4872 7.1329>, rot=<-0.2205 0.3737 -0.05 -0.8995>}
 FACIAL_C_Chin {parent=FACIAL_C_Jaw, loc=<-0.0008 1.164 7.7817>, rot=<-0.2558 0 0 -0.9667>}
 FACIAL_L_Ear {parent=FACIAL_C_FacialRoot, loc=<6.8332 -0.6605 0.1934>, rot=<-0.0067 -0.0012 -0.0261 -0.9996>}
 FACIAL_R_Ear {parent=FACIAL_C_FacialRoot, loc=<-6.8701 -0.695 0.062>, rot=<-0.0125 0.001 0.0262 -0.9996>}
 FACIAL_L_Jawline {parent=FACIAL_C_FacialRoot, loc=<4.3004 5.6637 3.2016>, rot=<0.2259 0.1566 -0.0575 -0.9597>}
 FACIAL_R_Jawline {parent=FACIAL_C_FacialRoot, loc=<-4.3172 5.6522 3.1719>, rot=<0.2228 -0.157 0.0567 -0.9605>}
 clavicle_l {parent=spine_05, loc=<4.3575 -0.1466 -2.3223>, rot=<0.0465 -0.7289 -0.009 -0.6829>}
 clavicle_r {parent=spine_05, loc=<4.3575 -0.1466 2.3223>, rot=<0.7289 0.0465 -0.6829 0.009>}
]
