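# Anim Blueprint for the DomeRoof male face accessory, adding secondary motion including hat feathers.
# Copies the attached parent pose, layers in the head mesh from facial_root, then runs a Control Rig.
# WearingBackbling blends between two RigidBody simulations with separate physics assets, followed by serial AnimDynamics springs on feather bones.
# The event graph detects backblings, adjusts motion scales, computes gravity/RBAN settings and sets simulation space by front-end state.
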AnimBlueprint /BRCosmetics/Characters/Player/Male/Medium/Bodies/M_MED_DomeRoof/Meshes/Parts/M_MED_DomeRoof_FaceAcc_AnimBP
target_skeleton=&/M_MED_DomeRoof_FaceAcc_Skeleton
graphs [
 AnimGraph {nodes [
   n1 = RigidBody!(
    ComponentPose=Local To Component(LocalPose=Use cached pose 'PostDeformationRig'())
    Alpha
    OverrideWorldGravity:Vector="0,0,-980"
    ExternalForce
    SimSpaceSettings:Sim Space Settings Structure=(WorldAlpha=1,VelocityScaleZ=1,DampingAlpha=1,MaxLinearVelocity=20,MaxAngularVelocity=40,MaxLinearAcceleration=20,MaxAngularAcceleration=40,ExternalLinearDragV=(X=0,Y=0,Z=0),ExternalLinearVelocity=(X=0,Y=0,Z=0),ExternalAngularVelocity=(X=0,Y=0,Z=0))
   ) -> (Pose) {
    bClampLinearTranslationLimitToRefPose=true
    bOverrideWorldGravity=true
    overridePhysicsAsset=&/M_MED_DomeRoof_FaceAcc_BackblingPhysics
    simulationSpace=BaseBoneSpace
    alphaBoolBlend {blendOption=Linear}
    baseBoneRef {boneName=spine_04}
    componentAppliedLinearAccClamp=<1000 600 600>
    componentLinearAccScale {z=0.2}
    componentLinearVelScale {z=0.2}
    overrideWorldGravity {z=-980}
    simSpaceSettings {maxAngularAcceleration=40, maxAngularVelocity=40, maxLinearAcceleration=20, maxLinearVelocity=20, worldAlpha=1}
   }
   n2 = Component To Local(ComponentPose=@n1)
   n3 = RigidBody!(
    ComponentPose=Local To Component(LocalPose=Use cached pose 'PostDeformationRig'())
    Alpha
    OverrideWorldGravity="0,0,-980"
    ExternalForce
    ComponentLinearAccScale:Vector=Get Component Linear Acc Scale() -> (Component Linear Acc Scale:Vector)
    ComponentLinearVelScale:Vector=Get Component Linear Vel Scale() -> (Component Linear Vel Scale:Vector)
    SimSpaceSettings=(WorldAlpha=1,VelocityScaleZ=1,DampingAlpha=1,MaxLinearVelocity=20,MaxAngularVelocity=40,MaxLinearAcceleration=20,MaxAngularAcceleration=40,ExternalLinearDragV=(X=0,Y=0,Z=0),ExternalLinearVelocity=(X=0,Y=0,Z=0),ExternalAngularVelocity=(X=0,Y=0,Z=0))
   ) -> (Pose) {
    bOverrideWorldGravity=true
    overridePhysicsAsset=&/M_MED_DomeRoof_FaceAcc_Physics
    simulationSpace=BaseBoneSpace
    alphaBoolBlend {blendOption=Linear}
    baseBoneRef {boneName=spine_04}
    componentAppliedLinearAccClamp=<2000 250 2000>
    componentLinearAccScale=<0.4 0.2 0.2>
    componentLinearVelScale=<0.4 0.2 0.2>
    overrideWorldGravity {z=-980}
    simSpaceSettings {maxAngularAcceleration=40, maxAngularVelocity=40, maxLinearAcceleration=20, maxLinearVelocity=20, worldAlpha=1}
   }
   n4 = Component To Local(ComponentPose=@n3)
   n5 = Get WearingBackbling() -> (WearingBackbling:bool)
   n6 = Anim Dynamics - Bone: dyn_hat_feather_root!(
    SimulationSpace:AnimPhysSimSpaceType Enum=Component
    ComponentPose=Local To Component(LocalPose=Blend Poses by bool(BlendPose_0=@n2, BlendPose_1=@n4, bActiveValue=@n5) {blendType=Linear})
    Alpha=1
    ExternalForce="0,0,0"
   ) -> (Pose) {
    angularDampingOverride=1
    angularSpringConstant=20000
    bAngularSpring=true
    bOverrideAngularDamping=true
    bUseGravityOverride=true
    gravityScale=0.35
    numSolverIterationsPreUpdate=1
    simulationSpace=RootRelative
    alphaBoolBlend {blendOption=Linear}
    boundBone {boneName=dyn_hat_feather_root}
    componentAppliedLinearAccClamp=<25000 25000 25000>
    componentLinearAccScale=<1 1 1>
    componentLinearVelScale=<1 1 1>
    gravityOverride {z=-980}
    physicsBodyDefinitions [{
      boundBone {boneName=dyn_hat_feather_root}
      boxExtents=<15 20 20>
      constraintSetup {angularLimitsMax=<0 0 0>, angularLimitsMin=<-5 -5 -5>, angularTarget=<1 0 0>}
      localJointOffset=<2 0 0>
     }]
    relativeSpaceBone {boneName=head}
   }
   n7 = Anim Dynamics - Bone: dyn_hat_feather_a_01!(SimulationSpace:AnimPhysSimSpaceType Enum, ComponentPose=@n6, Alpha=1, ExternalForce="0,0,0") -> (Pose) {
    angularDampingOverride=1
    angularSpringConstant=20000
    bAngularSpring=true
    bOverrideAngularDamping=true
    bUseGravityOverride=true
    gravityScale=0.35
    numSolverIterationsPreUpdate=1
    simulationSpace=RootRelative
    alphaBoolBlend {blendOption=Linear}
    boundBone {boneName=dyn_hat_feather_a_01}
    componentAppliedLinearAccClamp=<25000 25000 25000>
    componentLinearAccScale=<1 1 1>
    componentLinearVelScale=<1 1 1>
    gravityOverride {z=-980}
    physicsBodyDefinitions [{
      boundBone {boneName=dyn_hat_feather_a_01}
      boxExtents=<15 20 20>
      constraintSetup {angularLimitsMax=<0 0 0>, angularLimitsMin=<-5 -5 -5>, angularTarget=<1 0 0>}
      localJointOffset=<2 0.5 0>
     }]
    relativeSpaceBone {boneName=head}
   }
   n8 = Anim Dynamics - Bone: dyn_hat_feather_a_02!(SimulationSpace:AnimPhysSimSpaceType Enum, ComponentPose=@n7, Alpha=1, ExternalForce="0,0,0") -> (Pose) {
    angularDampingOverride=1
    angularSpringConstant=20000
    bAngularSpring=true
    bOverrideAngularDamping=true
    bUseGravityOverride=true
    gravityScale=0.35
    numSolverIterationsPreUpdate=1
    simulationSpace=RootRelative
    alphaBoolBlend {blendOption=Linear}
    boundBone {boneName=dyn_hat_feather_a_02}
    componentAppliedLinearAccClamp=<25000 25000 25000>
    componentLinearAccScale=<1 1 1>
    componentLinearVelScale=<1 1 1>
    gravityOverride {z=-980}
    physicsBodyDefinitions [{
      boundBone {boneName=dyn_hat_feather_a_02}
      boxExtents=<15 20 20>
      constraintSetup {angularLimitsMax=<10 10 10>, angularLimitsMin=<-10 -10 -10>, angularTarget=<1 0 0>}
      localJointOffset=<2 -0.5 0>
     }]
    relativeSpaceBone {boneName=head}
   }
   n9 = Anim Dynamics - Bone: dyn_hat_feather_b_01!(SimulationSpace:AnimPhysSimSpaceType Enum, ComponentPose=@n8, Alpha=1, ExternalForce="0,0,0") -> (Pose) {
    angularDampingOverride=1
    angularSpringConstant=20000
    bAngularSpring=true
    bOverrideAngularDamping=true
    bUseGravityOverride=true
    gravityScale=0.35
    numSolverIterationsPreUpdate=1
    simulationSpace=RootRelative
    alphaBoolBlend {blendOption=Linear}
    boundBone {boneName=dyn_hat_feather_b_01}
    componentAppliedLinearAccClamp=<25000 25000 25000>
    componentLinearAccScale=<1 1 1>
    componentLinearVelScale=<1 1 1>
    gravityOverride {z=-980}
    physicsBodyDefinitions [{
      boundBone {boneName=dyn_hat_feather_b_01}
      boxExtents=<15 20 20>
      constraintSetup {angularLimitsMax=<0 0 0>, angularLimitsMin=<-5 -5 -5>, angularTarget=<1 0 0>}
      localJointOffset=<2 0 0.5>
     }]
    relativeSpaceBone {boneName=head}
   }
   n10 = Anim Dynamics - Bone: dyn_hat_feather_b_02!(SimulationSpace:AnimPhysSimSpaceType Enum, ComponentPose=@n9, Alpha=1, ExternalForce="0,0,0") -> (Pose) {
    angularDampingOverride=1
    angularSpringConstant=20000
    bAngularSpring=true
    bOverrideAngularDamping=true
    bUseGravityOverride=true
    gravityScale=0.35
    numSolverIterationsPreUpdate=1
    simulationSpace=RootRelative
    alphaBoolBlend {blendOption=Linear}
    boundBone {boneName=dyn_hat_feather_b_02}
    componentAppliedLinearAccClamp=<25000 25000 25000>
    componentLinearAccScale=<0.5 0.5 0.5>
    componentLinearVelScale=<0.5 0.5 0.5>
    gravityOverride {z=-980}
    physicsBodyDefinitions [{
      boundBone {boneName=dyn_hat_feather_b_02}
      boxExtents=<15 20 20>
      constraintSetup {angularLimitsMax=<10 10 10>, angularLimitsMin=<-10 -10 -10>, angularTarget=<1 0 0>}
      localJointOffset=<2 0 -0.5>
     }]
    relativeSpaceBone {boneName=head}
   }
   n11 = Anim Dynamics - Bone: dyn_hat_feather_f_01!(SimulationSpace:AnimPhysSimSpaceType Enum, ComponentPose=@n10, Alpha=1, ExternalForce="0,0,0") -> (Pose) {
    angularDampingOverride=1
    angularSpringConstant=20000
    bAngularSpring=true
    bOverrideAngularDamping=true
    bUseGravityOverride=true
    gravityScale=0.35
    numSolverIterationsPreUpdate=1
    simulationSpace=RootRelative
    alphaBoolBlend {blendOption=Linear}
    boundBone {boneName=dyn_hat_feather_f_01}
    componentAppliedLinearAccClamp=<25000 25000 25000>
    componentLinearAccScale=<1 1 1>
    componentLinearVelScale=<1 1 1>
    gravityOverride {z=-980}
    physicsBodyDefinitions [{
      boundBone {boneName=dyn_hat_feather_f_01}
      boxExtents=<15 20 20>
      constraintSetup {angularLimitsMax=<0 0 0>, angularLimitsMin=<-5 -5 -5>, angularTarget=<1 0 0>}
      localJointOffset=<2 0.5 0.5>
     }]
    relativeSpaceBone {boneName=head}
   }
   n12 = Anim Dynamics - Bone: dyn_hat_feather_c_01!(SimulationSpace:AnimPhysSimSpaceType Enum, ComponentPose=@n11, Alpha=1, ExternalForce="0,0,0") -> (Pose) {
    angularDampingOverride=1
    angularSpringConstant=20000
    bAngularSpring=true
    bOverrideAngularDamping=true
    bUseGravityOverride=true
    gravityScale=0.35
    numSolverIterationsPreUpdate=1
    simulationSpace=RootRelative
    alphaBoolBlend {blendOption=Linear}
    boundBone {boneName=dyn_hat_feather_c_01}
    componentAppliedLinearAccClamp=<25000 25000 25000>
    componentLinearAccScale=<1 1 1>
    componentLinearVelScale=<1 1 1>
    gravityOverride {z=-980}
    physicsBodyDefinitions [{
      boundBone {boneName=dyn_hat_feather_c_01}
      boxExtents=<15 20 20>
      constraintSetup {angularLimitsMax=<10 10 10>, angularLimitsMin=<-10 -10 -10>, angularTarget=<1 0 0>}
      localJointOffset=<2 -0.5 -0.5>
     }]
    relativeSpaceBone {boneName=head}
   }
   n13 = Anim Dynamics - Bone: dyn_hat_feather_d_01!(SimulationSpace:AnimPhysSimSpaceType Enum, ComponentPose=@n12, Alpha=1, ExternalForce="0,0,0") -> (Pose) {
    angularDampingOverride=1
    angularSpringConstant=20000
    bAngularSpring=true
    bOverrideAngularDamping=true
    bUseGravityOverride=true
    gravityScale=0.35
    numSolverIterationsPreUpdate=1
    simulationSpace=RootRelative
    alphaBoolBlend {blendOption=Linear}
    boundBone {boneName=dyn_hat_feather_d_01}
    componentAppliedLinearAccClamp=<25000 25000 25000>
    componentLinearAccScale=<1 1 1>
    componentLinearVelScale=<1 1 1>
    gravityOverride {z=-980}
    physicsBodyDefinitions [{
      boundBone {boneName=dyn_hat_feather_d_01}
      boxExtents=<15 20 20>
      constraintSetup {angularLimitsMax=<0 0 0>, angularLimitsMin=<-5 -5 -5>, angularTarget=<1 0 0>}
      localJointOffset=<2 0 0>
     }]
    relativeSpaceBone {boneName=head}
   }
   n14 = Anim Dynamics - Bone: dyn_hat_feather_e_01!(SimulationSpace:AnimPhysSimSpaceType Enum, ComponentPose=@n13, Alpha=1, ExternalForce="0,0,0") -> (Pose) {
    angularDampingOverride=1
    angularSpringConstant=20000
    bAngularSpring=true
    bOverrideAngularDamping=true
    bUseGravityOverride=true
    gravityScale=0.35
    numSolverIterationsPreUpdate=1
    simulationSpace=RootRelative
    alphaBoolBlend {blendOption=Linear}
    boundBone {boneName=dyn_hat_feather_e_01}
    componentAppliedLinearAccClamp=<25000 25000 25000>
    componentLinearAccScale=<0.5 0.5 0.5>
    componentLinearVelScale=<0.5 0.5 0.5>
    gravityOverride {z=-980}
    physicsBodyDefinitions [{
      boundBone {boneName=dyn_hat_feather_e_01}
      boxExtents=<15 25 20>
      constraintSetup {angularLimitsMax=<0 0 0>, angularLimitsMin=<-5 -5 -5>, angularTarget=<1 0 0>}
      localJointOffset=<2 0 0>
     }]
    relativeSpaceBone {boneName=head}
   }
   n15 = Anim Dynamics - Bone: dyn_hat_feather_e_02!(SimulationSpace:AnimPhysSimSpaceType Enum, ComponentPose=@n14, Alpha=1, ExternalForce="0,0,0") -> (Pose) {
    angularDampingOverride=1
    angularSpringConstant=20000
    bAngularSpring=true
    bOverrideAngularDamping=true
    bUseGravityOverride=true
    gravityScale=0.35
    numSolverIterationsPreUpdate=1
    simulationSpace=RootRelative
    alphaBoolBlend {blendOption=Linear}
    boundBone {boneName=dyn_hat_feather_e_02}
    componentAppliedLinearAccClamp=<25000 25000 25000>
    componentLinearAccScale=<1 1 1>
    componentLinearVelScale=<1 1 1>
    gravityOverride {z=-980}
    physicsBodyDefinitions [{
      boundBone {boneName=dyn_hat_feather_e_02}
      boxExtents=<15 20 15>
      constraintSetup {angularLimitsMax=<0 0 0>, angularLimitsMin=<-5 -5 -5>, angularTarget=<1 0 0>}
      localJointOffset=<2 0 0>
     }]
    relativeSpaceBone {boneName=head}
   }
   n16 = Anim Dynamics - Bone: dyn_hat_feather_e_03!(SimulationSpace:AnimPhysSimSpaceType Enum, ComponentPose=@n15, Alpha=1, ExternalForce="0,0,0") -> (Pose) {
    angularDampingOverride=1
    angularSpringConstant=20000
    bAngularSpring=true
    bOverrideAngularDamping=true
    bUseGravityOverride=true
    gravityScale=0.35
    numSolverIterationsPreUpdate=1
    simulationSpace=RootRelative
    alphaBoolBlend {blendOption=Linear}
    boundBone {boneName=dyn_hat_feather_e_03}
    componentAppliedLinearAccClamp=<25000 25000 25000>
    componentLinearAccScale=<1 1 1>
    componentLinearVelScale=<1 1 1>
    gravityOverride {z=-980}
    physicsBodyDefinitions [{
      boundBone {boneName=dyn_hat_feather_e_03}
      boxExtents=<15 20 20>
      constraintSetup {angularLimitsMax=<2.5 2.5 2.5>, angularLimitsMin=<-10 -10 -10>, angularTarget=<1 0 0>}
      localJointOffset=<2 0 0>
     }]
    relativeSpaceBone {boneName=head}
   }
   n17 = Anim Dynamics - Bone: dyn_hat_feather_h_01!(SimulationSpace:AnimPhysSimSpaceType Enum, ComponentPose=@n16, Alpha=1, ExternalForce="0,0,0") -> (Pose) {
    angularDampingOverride=1
    angularSpringConstant=20000
    bAngularSpring=true
    bOverrideAngularDamping=true
    bUseGravityOverride=true
    gravityScale=0.35
    numSolverIterationsPreUpdate=1
    simulationSpace=RootRelative
    alphaBoolBlend {blendOption=Linear}
    boundBone {boneName=dyn_hat_feather_h_01}
    componentAppliedLinearAccClamp=<25000 25000 25000>
    componentLinearAccScale=<1 1 1>
    componentLinearVelScale=<1 1 1>
    gravityOverride {z=-980}
    physicsBodyDefinitions [{
      boundBone {boneName=dyn_hat_feather_h_01}
      boxExtents=<15 20 20>
      constraintSetup {angularLimitsMax=<10 10 10>, angularLimitsMin=<-10 -10 -10>, angularTarget=<1 0 0>}
      localJointOffset=<2 0.5 0>
     }]
    relativeSpaceBone {boneName=head}
   }
   n18 = Anim Dynamics - Bone: dyn_hat_feather_front_01!(SimulationSpace:AnimPhysSimSpaceType Enum, ComponentPose=@n17, Alpha=1, ExternalForce="0,0,0") -> (Pose) {
    angularDampingOverride=1
    angularSpringConstant=20000
    bAngularSpring=true
    bOverrideAngularDamping=true
    bUseGravityOverride=true
    gravityScale=0.35
    numSolverIterationsPreUpdate=1
    simulationSpace=RootRelative
    alphaBoolBlend {blendOption=Linear}
    boundBone {boneName=dyn_hat_feather_front_01}
    componentAppliedLinearAccClamp=<25000 25000 25000>
    componentLinearAccScale=<1 1 1>
    componentLinearVelScale=<1 1 1>
    gravityOverride {z=-980}
    physicsBodyDefinitions [{
      boundBone {boneName=dyn_hat_feather_front_01}
      boxExtents=<15 20 20>
      constraintSetup {angularLimitsMax=<0 0 0>, angularLimitsMin=<-5 -5 -5>, angularTarget=<1 0 0>}
      localJointOffset=<2 -0.5 0>
     }]
    relativeSpaceBone {boneName=head}
   }
   n19 = Anim Dynamics - Bone: dyn_hat_feather_front_a_01!(SimulationSpace:AnimPhysSimSpaceType Enum, ComponentPose=@n18, Alpha=1, ExternalForce="0,0,0") -> (Pose) {
    angularDampingOverride=1
    angularSpringConstant=20000
    bAngularSpring=true
    bOverrideAngularDamping=true
    bUseGravityOverride=true
    gravityScale=0.35
    numSolverIterationsPreUpdate=1
    simulationSpace=RootRelative
    alphaBoolBlend {blendOption=Linear}
    boundBone {boneName=dyn_hat_feather_front_a_01}
    componentAppliedLinearAccClamp=<25000 25000 25000>
    componentLinearAccScale=<0.5 0.5 0.5>
    componentLinearVelScale=<0.5 0.5 0.5>
    gravityOverride {z=-980}
    physicsBodyDefinitions [{
      boundBone {boneName=dyn_hat_feather_front_a_01}
      boxExtents=<15 20 20>
      constraintSetup {angularLimitsMax=<0 0 0>, angularLimitsMin=<-5 -5 -5>, angularTarget=<1 0 0>}
      localJointOffset=<2 0 0.5>
     }]
    relativeSpaceBone {boneName=head}
   }
   n20 = Anim Dynamics - Bone: dyn_hat_feather_front_a_02!(SimulationSpace:AnimPhysSimSpaceType Enum, ComponentPose=@n19, Alpha=1, ExternalForce="0,0,0") -> (Pose) {
    angularDampingOverride=1
    angularSpringConstant=20000
    bAngularSpring=true
    bOverrideAngularDamping=true
    bUseGravityOverride=true
    gravityScale=0.35
    numSolverIterationsPreUpdate=1
    simulationSpace=RootRelative
    alphaBoolBlend {blendOption=Linear}
    boundBone {boneName=dyn_hat_feather_front_a_02}
    componentAppliedLinearAccClamp=<25000 25000 25000>
    componentLinearAccScale=<1 1 1>
    componentLinearVelScale=<1 1 1>
    gravityOverride {z=-980}
    physicsBodyDefinitions [{
      boundBone {boneName=dyn_hat_feather_front_a_02}
      boxExtents=<15 20 20>
      constraintSetup {angularLimitsMax=<2 2 2>, angularLimitsMin=<-7.5 -7.5 -7.5>, angularTarget=<1 0 0>}
      localJointOffset=<2 0 -0.5>
     }]
    relativeSpaceBone {boneName=head}
   }
   n21 = Anim Dynamics - Bone: dyn_hat_feather_g_01!(SimulationSpace:AnimPhysSimSpaceType Enum, ComponentPose=@n20, Alpha=1, ExternalForce="0,0,0") -> (Pose) {
    angularDampingOverride=1
    angularSpringConstant=20000
    bAngularSpring=true
    bOverrideAngularDamping=true
    bUseGravityOverride=true
    gravityScale=0.35
    numSolverIterationsPreUpdate=1
    simulationSpace=RootRelative
    alphaBoolBlend {blendOption=Linear}
    boundBone {boneName=dyn_hat_feather_g_01}
    componentAppliedLinearAccClamp=<25000 25000 25000>
    componentLinearAccScale=<0.5 0.5 0.5>
    componentLinearVelScale=<0.5 0.5 0.5>
    gravityOverride {z=-980}
    physicsBodyDefinitions [{
      boundBone {boneName=dyn_hat_feather_g_01}
      boxExtents=<15 20 20>
      constraintSetup {angularLimitsMax=<2.5 2.5 2.5>, angularLimitsMin=<-10 -10 -10>, angularTarget=<1 0 0>}
      localJointOffset=<2 0.5 0.5>
     }]
    relativeSpaceBone {boneName=head}
   }
   n22 = Anim Dynamics - Bone: dyn_hat_feather_front_02!(SimulationSpace:AnimPhysSimSpaceType Enum, ComponentPose=@n21, Alpha=1, ExternalForce="0,0,0") -> (Pose) {
    angularDampingOverride=1
    angularSpringConstant=20000
    bAngularSpring=true
    bOverrideAngularDamping=true
    bUseGravityOverride=true
    gravityScale=0.35
    numSolverIterationsPreUpdate=1
    simulationSpace=RootRelative
    alphaBoolBlend {blendOption=Linear}
    boundBone {boneName=dyn_hat_feather_front_02}
    componentAppliedLinearAccClamp=<25000 25000 25000>
    componentLinearAccScale=<1 1 1>
    componentLinearVelScale=<1 1 1>
    gravityOverride {z=-980}
    physicsBodyDefinitions [{
      boundBone {boneName=dyn_hat_feather_front_02}
      boxExtents=<15 20 20>
      constraintSetup {angularLimitsMax=<2.5 2.5 2.5>, angularLimitsMin=<-10 -10 -10>, angularTarget=<1 0 0>}
      localJointOffset=<2 -0.5 -0.5>
     }]
    relativeSpaceBone {boneName=head}
   }
   n23 = Anim Dynamics - Bone: dyn_hat_feather_front_b_01!(SimulationSpace:AnimPhysSimSpaceType Enum, ComponentPose=@n22, Alpha=1, ExternalForce="0,0,0") -> (Pose) {
    angularDampingOverride=1
    angularSpringConstant=20000
    bAngularSpring=true
    bOverrideAngularDamping=true
    bUseGravityOverride=true
    gravityScale=0.35
    numSolverIterationsPreUpdate=1
    simulationSpace=RootRelative
    alphaBoolBlend {blendOption=Linear}
    boundBone {boneName=dyn_hat_feather_front_b_01}
    componentAppliedLinearAccClamp=<25000 25000 25000>
    componentLinearAccScale=<0.5 0.5 0.5>
    componentLinearVelScale=<0.5 0.5 0.5>
    gravityOverride {z=-980}
    physicsBodyDefinitions [{
      boundBone {boneName=dyn_hat_feather_front_b_01}
      boxExtents=<15 20 20>
      constraintSetup {angularLimitsMax=<0 0 0>, angularLimitsMin=<-5 -5 -5>, angularTarget=<1 0 0>}
      localJointOffset=<2 0 0>
     }]
    relativeSpaceBone {boneName=head}
   }
   n24 = Anim Dynamics - Bone: dyn_hat_feather_front_b_02!(SimulationSpace:AnimPhysSimSpaceType Enum, ComponentPose=@n23, Alpha=1, ExternalForce="0,0,0") -> (Pose) {
    angularDampingOverride=1
    angularSpringConstant=20000
    bAngularSpring=true
    bOverrideAngularDamping=true
    bUseGravityOverride=true
    gravityScale=0.35
    numSolverIterationsPreUpdate=1
    simulationSpace=RootRelative
    alphaBoolBlend {blendOption=Linear}
    boundBone {boneName=dyn_hat_feather_front_b_02}
    componentAppliedLinearAccClamp=<25000 25000 25000>
    componentLinearAccScale=<1 1 1>
    componentLinearVelScale=<1 1 1>
    gravityOverride {z=-980}
    physicsBodyDefinitions [{
      boundBone {boneName=dyn_hat_feather_front_b_02}
      boxExtents=<15 25 20>
      constraintSetup {angularLimitsMax=<0 0 0>, angularLimitsMin=<-5 -5 -5>, angularTarget=<1 0 0>}
      localJointOffset=<2 0 0>
     }]
    relativeSpaceBone {boneName=head}
   }
   n25 = Anim Dynamics - Bone: dyn_hat_feather_front_b_03!(SimulationSpace:AnimPhysSimSpaceType Enum, ComponentPose=@n24, Alpha=1, ExternalForce="0,0,0") -> (Pose) {
    angularDampingOverride=1
    angularSpringConstant=20000
    bAngularSpring=true
    bOverrideAngularDamping=true
    bUseGravityOverride=true
    gravityScale=0.35
    numSolverIterationsPreUpdate=1
    simulationSpace=RootRelative
    alphaBoolBlend {blendOption=Linear}
    boundBone {boneName=dyn_hat_feather_front_b_03}
    componentAppliedLinearAccClamp=<25000 25000 25000>
    componentLinearAccScale=<0.5 0.5 0.5>
    componentLinearVelScale=<0.5 0.5 0.5>
    gravityOverride {z=-980}
    physicsBodyDefinitions [{
      boundBone {boneName=dyn_hat_feather_front_b_03}
      boxExtents=<15 20 15>
      constraintSetup {angularLimitsMax=<0 0 0>, angularLimitsMin=<-5 -5 -5>, angularTarget=<1 0 0>}
      localJointOffset=<2 0 0>
     }]
    relativeSpaceBone {boneName=head}
   }
   n26 = Anim Dynamics - Bone: dyn_hat_feather_front_b_04!(SimulationSpace:AnimPhysSimSpaceType Enum, ComponentPose=@n25, Alpha=1, ExternalForce="0,0,0") -> (Pose) {
    angularDampingOverride=1
    angularSpringConstant=20000
    bAngularSpring=true
    bOverrideAngularDamping=true
    bUseGravityOverride=true
    gravityScale=0.35
    numSolverIterationsPreUpdate=1
    simulationSpace=RootRelative
    alphaBoolBlend {blendOption=Linear}
    boundBone {boneName=dyn_hat_feather_front_b_04}
    componentAppliedLinearAccClamp=<25000 25000 25000>
    componentLinearAccScale=<1 1 1>
    componentLinearVelScale=<1 1 1>
    gravityOverride {z=-980}
    physicsBodyDefinitions [{
      boundBone {boneName=dyn_hat_feather_front_b_04}
      boxExtents=<15 20 20>
      constraintSetup {angularLimitsMax=<2.5 2.5 2.5>, angularLimitsMin=<-10 -10 -10>, angularTarget=<1 0 0>}
      localJointOffset=<2 0 0>
     }]
    relativeSpaceBone {boneName=head}
   }
   n27 = Anim Dynamics - Bone: dyn_hat_feather_front_c_01!(SimulationSpace:AnimPhysSimSpaceType Enum, ComponentPose=@n26, Alpha=1, ExternalForce="0,0,0") -> (Pose) {
    angularDampingOverride=1
    angularSpringConstant=20000
    bAngularSpring=true
    bOverrideAngularDamping=true
    bUseGravityOverride=true
    gravityScale=0.35
    numSolverIterationsPreUpdate=1
    simulationSpace=RootRelative
    alphaBoolBlend {blendOption=Linear}
    boundBone {boneName=dyn_hat_feather_front_c_01}
    componentAppliedLinearAccClamp=<25000 25000 25000>
    componentLinearAccScale=<1 1 1>
    componentLinearVelScale=<1 1 1>
    gravityOverride {z=-980}
    physicsBodyDefinitions [{
      boundBone {boneName=dyn_hat_feather_front_c_01}
      boxExtents=<15 20 20>
      constraintSetup {angularLimitsMax=<0 0 0>, angularLimitsMin=<-5 -5 -5>, angularTarget=<1 0 0>}
      localJointOffset=<2 0.5 0>
     }]
    relativeSpaceBone {boneName=head}
   }
   n28 = Anim Dynamics - Bone: dyn_hat_feather_front_c_02!(SimulationSpace:AnimPhysSimSpaceType Enum, ComponentPose=@n27, Alpha=1, ExternalForce="0,0,0") -> (Pose) {
    angularDampingOverride=1
    angularSpringConstant=20000
    bAngularSpring=true
    bOverrideAngularDamping=true
    bUseGravityOverride=true
    gravityScale=0.35
    numSolverIterationsPreUpdate=1
    simulationSpace=RootRelative
    alphaBoolBlend {blendOption=Linear}
    boundBone {boneName=dyn_hat_feather_front_c_02}
    componentAppliedLinearAccClamp=<25000 25000 25000>
    componentLinearAccScale=<0.5 0.5 0.5>
    componentLinearVelScale=<0.5 0.5 0.5>
    gravityOverride {z=-980}
    physicsBodyDefinitions [{
      boundBone {boneName=dyn_hat_feather_front_c_02}
      boxExtents=<15 20 20>
      constraintSetup {angularLimitsMax=<0 0 0>, angularLimitsMin=<-5 -5 -5>, angularTarget=<1 0 0>}
      localJointOffset=<2 -0.5 0>
     }]
    relativeSpaceBone {boneName=head}
   }
   n29 = Anim Dynamics - Bone: dyn_hat_feather_front_c_03!(SimulationSpace:AnimPhysSimSpaceType Enum, ComponentPose=@n28, Alpha=1, ExternalForce="0,0,0") -> (Pose) {
    angularDampingOverride=1
    angularSpringConstant=20000
    bAngularSpring=true
    bOverrideAngularDamping=true
    bUseGravityOverride=true
    gravityScale=0.35
    numSolverIterationsPreUpdate=1
    simulationSpace=RootRelative
    alphaBoolBlend {blendOption=Linear}
    boundBone {boneName=dyn_hat_feather_front_c_03}
    componentAppliedLinearAccClamp=<25000 25000 25000>
    componentLinearAccScale=<1 1 1>
    componentLinearVelScale=<1 1 1>
    gravityOverride {z=-980}
    physicsBodyDefinitions [{
      boundBone {boneName=dyn_hat_feather_front_c_03}
      boxExtents=<15 20 20>
      constraintSetup {angularLimitsMax=<0 0 0>, angularLimitsMin=<-5 -5 -5>, angularTarget=<1 0 0>}
      localJointOffset=<2 0 0.5>
     }]
    relativeSpaceBone {boneName=head}
   }
   n30 = Anim Dynamics - Bone: dyn_hat_feather_front_c_04!(SimulationSpace:AnimPhysSimSpaceType Enum, ComponentPose=@n29, Alpha=1, ExternalForce="0,0,0") -> (Pose) {
    angularDampingOverride=1
    angularSpringConstant=20000
    bAngularSpring=true
    bOverrideAngularDamping=true
    bUseGravityOverride=true
    gravityScale=0.35
    numSolverIterationsPreUpdate=1
    simulationSpace=RootRelative
    alphaBoolBlend {blendOption=Linear}
    boundBone {boneName=dyn_hat_feather_front_c_04}
    componentAppliedLinearAccClamp=<25000 25000 25000>
    componentLinearAccScale=<1 1 1>
    componentLinearVelScale=<1 1 1>
    gravityOverride {z=-980}
    physicsBodyDefinitions [{
      boundBone {boneName=dyn_hat_feather_front_c_04}
      boxExtents=<15 20 20>
      constraintSetup {angularLimitsMax=<2.5 2.5 2.5>, angularLimitsMin=<-10 -10 -10>, angularTarget=<1 0 0>}
      localJointOffset=<2 0 -0.5>
     }]
    relativeSpaceBone {boneName=head}
   }
   n31 = Anim Dynamics - Bone: dyn_hat_feather_front_d_01!(SimulationSpace:AnimPhysSimSpaceType Enum, ComponentPose=@n30, Alpha=1, ExternalForce="0,0,0") -> (Pose) {
    angularDampingOverride=1
    angularSpringConstant=20000
    bAngularSpring=true
    bOverrideAngularDamping=true
    bUseGravityOverride=true
    gravityScale=0.35
    numSolverIterationsPreUpdate=1
    simulationSpace=RootRelative
    alphaBoolBlend {blendOption=Linear}
    boundBone {boneName=dyn_hat_feather_front_d_01}
    componentAppliedLinearAccClamp=<25000 25000 25000>
    componentLinearAccScale=<1 1 1>
    componentLinearVelScale=<1 1 1>
    gravityOverride {z=-980}
    physicsBodyDefinitions [{
      boundBone {boneName=dyn_hat_feather_front_d_01}
      boxExtents=<15 20 20>
      constraintSetup {angularLimitsMax=<0 0 0>, angularLimitsMin=<-5 -5 -5>, angularTarget=<1 0 0>}
      localJointOffset=<2 0.5 0.5>
     }]
    relativeSpaceBone {boneName=head}
   }
   n32 = Anim Dynamics - Bone: dyn_hat_feather_front_d_02!(SimulationSpace:AnimPhysSimSpaceType Enum, ComponentPose=@n31, Alpha=1, ExternalForce="0,0,0") -> (Pose) {
    angularDampingOverride=1
    angularSpringConstant=20000
    bAngularSpring=true
    bOverrideAngularDamping=true
    bUseGravityOverride=true
    gravityScale=0.35
    numSolverIterationsPreUpdate=1
    simulationSpace=RootRelative
    alphaBoolBlend {blendOption=Linear}
    boundBone {boneName=dyn_hat_feather_front_d_02}
    componentAppliedLinearAccClamp=<25000 25000 25000>
    componentLinearAccScale=<0.5 0.5 0.5>
    componentLinearVelScale=<0.5 0.5 0.5>
    gravityOverride {z=-980}
    physicsBodyDefinitions [{
      boundBone {boneName=dyn_hat_feather_front_d_02}
      boxExtents=<15 20 20>
      constraintSetup {angularLimitsMax=<0 0 0>, angularLimitsMin=<-5 -5 -5>, angularTarget=<1 0 0>}
      localJointOffset=<2 -0.5 -0.5>
     }]
    relativeSpaceBone {boneName=head}
   }
   n33 = Anim Dynamics - Bone: dyn_hat_feather_front_d_03!(SimulationSpace:AnimPhysSimSpaceType Enum, ComponentPose=@n32, Alpha=1, ExternalForce="0,0,0") -> (Pose) {
    angularDampingOverride=1
    angularSpringConstant=20000
    bAngularSpring=true
    bOverrideAngularDamping=true
    bUseGravityOverride=true
    gravityScale=0.35
    numSolverIterationsPreUpdate=1
    simulationSpace=RootRelative
    alphaBoolBlend {blendOption=Linear}
    boundBone {boneName=dyn_hat_feather_front_d_03}
    componentAppliedLinearAccClamp=<25000 25000 25000>
    componentLinearAccScale=<1 1 1>
    componentLinearVelScale=<1 1 1>
    gravityOverride {z=-980}
    physicsBodyDefinitions [{
      boundBone {boneName=dyn_hat_feather_front_d_03}
      boxExtents=<15 20 20>
      constraintSetup {angularLimitsMax=<2.5 2.5 2.5>, angularLimitsMin=<-10 -10 -10>, angularTarget=<1 0 0>}
      localJointOffset=<2 0 0>
     }]
    relativeSpaceBone {boneName=head}
   }
   n34 = Anim Dynamics - Bone: dyn_hat_feather_i_01!(SimulationSpace:AnimPhysSimSpaceType Enum, ComponentPose=@n33, Alpha=1, ExternalForce="0,0,0") -> (Pose) {
    angularDampingOverride=1
    angularSpringConstant=20000
    bAngularSpring=true
    bOverrideAngularDamping=true
    bUseGravityOverride=true
    gravityScale=0.35
    numSolverIterationsPreUpdate=1
    simulationSpace=RootRelative
    alphaBoolBlend {blendOption=Linear}
    boundBone {boneName=dyn_hat_feather_i_01}
    componentAppliedLinearAccClamp=<25000 25000 25000>
    componentLinearAccScale=<1 1 1>
    componentLinearVelScale=<1 1 1>
    gravityOverride {z=-980}
    physicsBodyDefinitions [{
      boundBone {boneName=dyn_hat_feather_i_01}
      boxExtents=<15 25 20>
      constraintSetup {angularLimitsMax=<0 0 0>, angularLimitsMin=<-5 -5 -5>, angularTarget=<1 0 0>}
      localJointOffset=<2 0 0>
     }]
    relativeSpaceBone {boneName=head}
   }
   n35 = Anim Dynamics - Bone: dyn_hat_feather_i_02!(SimulationSpace:AnimPhysSimSpaceType Enum, ComponentPose=@n34, Alpha=1, ExternalForce="0,0,0") -> (Pose) {
    angularDampingOverride=1
    angularSpringConstant=20000
    bAngularSpring=true
    bOverrideAngularDamping=true
    bUseGravityOverride=true
    gravityScale=0.35
    numSolverIterationsPreUpdate=1
    simulationSpace=RootRelative
    alphaBoolBlend {blendOption=Linear}
    boundBone {boneName=dyn_hat_feather_i_02}
    componentAppliedLinearAccClamp=<25000 25000 25000>
    componentLinearAccScale=<1 1 1>
    componentLinearVelScale=<1 1 1>
    gravityOverride {z=-980}
    physicsBodyDefinitions [{
      boundBone {boneName=dyn_hat_feather_i_02}
      boxExtents=<15 20 15>
      constraintSetup {angularLimitsMax=<2.5 2.5 2.5>, angularLimitsMin=<-10 -10 -10>, angularTarget=<1 0 0>}
      localJointOffset=<2 0 0>
     }]
    relativeSpaceBone {boneName=head}
   }
   Output Pose(Result=Component To Local(ComponentPose=@n35)) {layerGroup=None}
   n36 = Get HeadComponent() -> (HeadComponent:Skeletal Mesh Component Budgeted Object Reference)
   n37 = Control Rig(Source=Layered blend per bone(
     BasePose=Copy Pose From Mesh() {bUseAttachedParent=true}
     BlendPoses_0=Copy Pose From Mesh(SourceMeshComponent=@n36) {bUseAttachedParent=true}
    ) {layerSetup [{branchFilters [{boneName=facial_root}]}]}) {
    bTransferInputCurves=false
    alphaBoolBlend {blendOption=Linear}
    controlRigAssetReference {blueprintRigClass=&/M_MED_DomeRoof_FaceAcc_CtrlRig.M_MED_DomeRoof_FaceAcc_CtrlRig_C}
   }
   Save cached pose 'PostDeformationRig'(Pose=@n37) {cacheName=PostDeformationRig}
  ]}
 EventGraph {nodes [
   Event BlueprintUpdateAnimation()
   > n1 = Sequence()
   > n2 = Branch(Condition=Get WearingBackbling())
   > n3 = Gravity Override State Function DEPRECATED(
    CharacterPartAnimInstance=Self-Reference()
    GameTimeInSeconds=GetGameTimeInSeconds()
    GravityOverrideParams=Get Gravity Override Params_Backbling() -> (Gravity Override Params_Backbling:Gravity Override Params Struct Structure)
    RBAN Params=Get RBAN Params_Backbling() -> (RBAN Params_Backbling:RBANParams Struct Structure)
   )
   > Set Gravity Override(Gravity Override=@n3)
   > Set RBAN Settings Out(RBAN Settings Out:RBANSetting Struct Structure=@n3.RBAN SettingsOut) -> (Output_Get:RBANSetting Struct Structure)
   n4 = Gravity Override State Function DEPRECATED(
    execute=@n2.else
    CharacterPartAnimInstance=Self-Reference()
    GameTimeInSeconds=GetGameTimeInSeconds()
    GravityOverrideParams=Get Gravity Override Params()
    RBAN Params=Get RBAN Params()
   )
   > Set Gravity Override(Gravity Override=@n4)
   > Set RBAN Settings Out(RBAN Settings Out=@n4.RBAN SettingsOut)
   n5 = Branch(execute=@n1.then_1, Condition=Get bIsInFrontEnd())
   > Set Simulation Space(Simulation Space:AnimPhysSimSpaceType Enum=World) -> (Output_Get:AnimPhysSimSpaceType Enum)
   Set Simulation Space(execute=@n5.else)
   n6 = Event OnCharacterCustomizationLoaded()
   > Set HeadComponent(HeadComponent:Skeletal Mesh Component Budgeted Object Reference=Get Array Item with EFort Custom Part Type(Array=@n6.CustomMeshParts, PartType=Head)) -> (Output_Get:Skeletal Mesh Component Budgeted Object Reference)
   n7 = CheckForBackbling()
   > n8 = Branch(Condition=OR Boolean!(A=@n7, B=@n7.bIsBackblingACape, C:bool=@n7.bIsBackblingLong) -> (ReturnValue))
   > Set Component Linear Acc Scale(Component Linear Acc Scale:Vector="0,0,0.1") -> (Output_Get:Vector)
   > Set Component Linear Vel Scale(Component Linear Vel Scale:Vector="0,0,0.1") -> (Output_Get:Vector)
   > Set WearingBackbling(WearingBackbling:bool=true) -> (Output_Get:bool)
   Set Component Linear Acc Scale(execute=@n8.else, Component Linear Acc Scale="0.4,0.2,0.2")
   > Set Component Linear Vel Scale(Component Linear Vel Scale="0.4,0.2,0.2")
   > Set WearingBackbling(WearingBackbling=false)
  ]}
]
variables [
 Component Linear Acc Scale {type=Vector, default=<0.4 0.2 0.2>}
 Component Linear Vel Scale {type=Vector, default=<0.4 0.2 0.2>}
 Gravity Override {type=Vector}
 Gravity Override Params {
  type=GravityOverrideParamsStruct
  default {
   skydive_GoingUp {gravityOverride=<0 -300 1000>, windAmplitude=<150 -500 150>, windFrequency=<1.5 4 5>}
   tactical_Sprint {gravityOverride=<0 -450 -980>}
  }
 }
 Gravity Override Params_Backbling {
  type=GravityOverrideParamsStruct
  default {
   dBNO {jointName=spine_05, gravityOverride=<0 0 -980>}
   dBNO_Carried {jointName=spine_05, gravityOverride=<0 0 -980>}
   riding_Moving {gravityOverride=<0 0 -980>}
   riding_Moving_Targeting {gravityOverride=<0 0 -980>}
   riding_Standing {gravityOverride=<0 0 -980>}
   skydive_DivingDown {jointName=root, gravityOverride=<0 0 -980>, windAmplitude=<0 0 0>, windFrequency=<0 0 0>}
   skydive_FloatingDown {gravityOverride=<0 0 -980>, windAmplitude=<0 0 0>, windFrequency=<0 0 0>, windOffset=<0 0 0>}
   skydive_GoingUp {gravityOverride=<0 0 -980>, windAmplitude=<0 0 0>, windFrequency=<0 0 0>}
   skydive_Parachuting {gravityOverride=<0 0 -980>, windAmplitude=<0 0 0>, windFrequency=<0 0 0>}
   skydive_Surfing {gravityOverride=<0 0 -980>, windAmplitude=<0 0 0>, windFrequency=<0 0 0>}
   swimming {jointName=root, gravityOverride=<0 0 -980>}
   tactical_Sprint {gravityOverride=<0 0 -980>}
  }
 }
 HeadComponent {type=SkeletalMeshComponentBudgeted}
 RBAN Params {
  type=RBANParamsStruct
  default {
   dBNO {
    componentLinearAccScale=<1 1 0.5>
    componentLinearVelScale=<1 1 0.5>
    simSpaceSettings {
     maxAngularAcceleration=40
     maxAngularVelocity=40
     maxLinearAcceleration=20
     maxLinearVelocity=20
     velocityScaleZ=0.15
     worldAlpha=0.35
     externalLinearDragV=<0 0 0>
    }
   }
   dBNO_Carried {
    componentLinearAccScale=<1 1 0.5>
    componentLinearVelScale=<1 1 0.5>
    simSpaceSettings {
     maxAngularAcceleration=40
     maxAngularVelocity=40
     maxLinearAcceleration=20
     maxLinearVelocity=20
     velocityScaleZ=0.15
     worldAlpha=0.35
     externalLinearDragV=<0 0 0>
    }
   }
   emoteOrMelee {
    componentLinearAccScale=<1 1 0.5>
    componentLinearVelScale=<1 1 0.5>
    simSpaceSettings {
     maxAngularAcceleration=40
     maxAngularVelocity=40
     maxLinearAcceleration=20
     maxLinearVelocity=20
     velocityScaleZ=0.15
     worldAlpha=0.15
     externalLinearDragV=<0 0 0>
    }
   }
   front_End {
    componentLinearAccScale=<1 1 0.5>
    componentLinearVelScale=<1 1 0.5>
    simSpaceSettings {
     maxAngularAcceleration=40
     maxAngularVelocity=40
     maxLinearAcceleration=20
     maxLinearVelocity=20
     velocityScaleZ=0.15
     worldAlpha=0.1
     externalLinearDragV=<0 0 0>
    }
   }
   onGround_Crouch_Moving {
    componentLinearAccScale=<1 1 0.5>
    componentLinearVelScale=<1 1 0.5>
    simSpaceSettings {
     maxAngularAcceleration=40
     maxAngularVelocity=40
     maxLinearAcceleration=20
     maxLinearVelocity=20
     velocityScaleZ=0.15
     worldAlpha=0.35
     externalLinearDragV=<0 0 0>
    }
   }
   onGround_Falling {
    componentLinearAccScale=<1 1 0.5>
    componentLinearVelScale=<1 1 0.5>
    simSpaceSettings {
     maxAngularAcceleration=40
     maxAngularVelocity=40
     maxLinearAcceleration=20
     maxLinearVelocity=20
     velocityScaleZ=0.15
     worldAlpha=0.35
     externalLinearDragV=<0 0 0>
    }
   }
   onGround_Moving {
    componentLinearAccScale=<1 1 0.5>
    componentLinearVelScale=<1 1 0.5>
    simSpaceSettings {
     maxAngularAcceleration=40
     maxAngularVelocity=40
     maxLinearAcceleration=20
     maxLinearVelocity=20
     velocityScaleZ=0.15
     worldAlpha=0.35
     externalLinearDragV=<0 0 0>
    }
   }
   onGround_Standing {
    componentLinearAccScale=<1 1 0.5>
    componentLinearVelScale=<1 1 0.5>
    simSpaceSettings {
     maxAngularAcceleration=40
     maxAngularVelocity=40
     maxLinearAcceleration=20
     maxLinearVelocity=20
     velocityScaleZ=0.15
     worldAlpha=0.35
     externalLinearDragV=<0 0 0>
    }
   }
   rail_Grinding {
    componentLinearAccScale=<1 1 0.5>
    componentLinearVelScale=<1 1 0.5>
    simSpaceSettings {
     maxAngularAcceleration=40
     maxAngularVelocity=40
     maxLinearAcceleration=20
     maxLinearVelocity=20
     velocityScaleZ=0.15
     worldAlpha=0.35
     externalLinearDragV=<0 0 0>
    }
   }
   riding_Moving {
    componentLinearAccScale=<1 1 0.5>
    componentLinearVelScale=<1 1 0.5>
    simSpaceSettings {
     maxAngularAcceleration=40
     maxAngularVelocity=40
     maxLinearAcceleration=20
     maxLinearVelocity=20
     velocityScaleZ=0.15
     worldAlpha=0.35
     externalLinearDragV=<0 0 0>
    }
   }
   riding_Moving_Targeting {
    componentLinearAccScale=<1 1 0.5>
    componentLinearVelScale=<1 1 0.5>
    simSpaceSettings {
     maxAngularAcceleration=40
     maxAngularVelocity=40
     maxLinearAcceleration=20
     maxLinearVelocity=20
     velocityScaleZ=0.15
     worldAlpha=0.35
     externalLinearDragV=<0 0 0>
    }
   }
   riding_Standing {
    componentLinearAccScale=<1 1 0.5>
    componentLinearVelScale=<1 1 0.5>
    simSpaceSettings {
     maxAngularAcceleration=40
     maxAngularVelocity=40
     maxLinearAcceleration=20
     maxLinearVelocity=20
     velocityScaleZ=0.15
     worldAlpha=0.35
     externalLinearDragV=<0 0 0>
    }
   }
   skydive_DivingDown {
    componentLinearAccScale=<1 1 0.5>
    componentLinearVelScale=<1 1 0.5>
    simSpaceSettings {
     maxAngularAcceleration=40
     maxAngularVelocity=40
     maxLinearAcceleration=20
     maxLinearVelocity=20
     velocityScaleZ=0.15
     worldAlpha=0.35
     externalLinearDragV=<0 0 0>
    }
   }
   skydive_FloatingDown {
    componentLinearAccScale=<1 1 0.5>
    componentLinearVelScale=<1 1 0.5>
    simSpaceSettings {
     maxAngularAcceleration=40
     maxAngularVelocity=40
     maxLinearAcceleration=20
     maxLinearVelocity=20
     velocityScaleZ=0.15
     worldAlpha=0.35
     externalLinearDragV=<0 0 0>
    }
   }
   skydive_GoingUp {
    componentLinearAccScale=<1 1 0.5>
    componentLinearVelScale=<1 1 0.5>
    simSpaceSettings {
     maxAngularAcceleration=40
     maxAngularVelocity=40
     maxLinearAcceleration=20
     maxLinearVelocity=20
     velocityScaleZ=0.15
     worldAlpha=0.35
     externalLinearDragV=<0 0 0>
    }
   }
   skydive_Parachuting {
    componentLinearAccScale=<1 1 0.5>
    componentLinearVelScale=<1 1 0.5>
    simSpaceSettings {
     maxAngularAcceleration=40
     maxAngularVelocity=40
     maxLinearAcceleration=20
     maxLinearVelocity=20
     velocityScaleZ=0.15
     worldAlpha=0.35
     externalLinearDragV=<0 0 0>
    }
   }
   skydive_Surfing {
    componentLinearAccScale=<1 1 0.5>
    componentLinearVelScale=<1 1 0.5>
    simSpaceSettings {
     maxAngularAcceleration=40
     maxAngularVelocity=40
     maxLinearAcceleration=20
     maxLinearVelocity=20
     velocityScaleZ=0.15
     worldAlpha=0.35
     externalLinearDragV=<0 0 0>
    }
   }
   swimming {
    componentLinearAccScale=<1 1 0.5>
    componentLinearVelScale=<1 1 0.5>
    simSpaceSettings {
     maxAngularAcceleration=40
     maxAngularVelocity=40
     maxLinearAcceleration=20
     maxLinearVelocity=20
     velocityScaleZ=0.15
     worldAlpha=0.35
     externalLinearDragV=<0 0 0>
    }
   }
   tactical_Sprint {
    componentLinearAccScale=<0.3 0.3 0.3>
    componentLinearVelScale=<0.3 0.3 0.3>
    simSpaceSettings {
     maxAngularAcceleration=40
     maxAngularVelocity=40
     maxLinearAcceleration=20
     maxLinearVelocity=20
     velocityScaleZ=0.15
     worldAlpha=0.35
     externalLinearDragV=<0 0 0>
    }
   }
   ziplining {
    componentLinearAccScale=<1 1 0.5>
    componentLinearVelScale=<1 1 0.5>
    simSpaceSettings {
     maxAngularAcceleration=40
     maxAngularVelocity=40
     maxLinearAcceleration=20
     maxLinearVelocity=20
     velocityScaleZ=0.15
     worldAlpha=0.35
     externalLinearDragV=<0 0 0>
    }
   }
  }
 }
 RBAN Params_Backbling {
  type=RBANParamsStruct
  default {
   dBNO {
    componentLinearAccScale=<1 1 0.5>
    componentLinearVelScale=<1 1 0.5>
    simSpaceSettings {
     maxAngularAcceleration=40
     maxAngularVelocity=40
     maxLinearAcceleration=20
     maxLinearVelocity=20
     velocityScaleZ=0.15
     worldAlpha=0.35
     externalLinearDragV=<0 0 0>
    }
   }
   dBNO_Carried {
    componentLinearAccScale=<1 1 0.5>
    componentLinearVelScale=<1 1 0.5>
    simSpaceSettings {
     maxAngularAcceleration=40
     maxAngularVelocity=40
     maxLinearAcceleration=20
     maxLinearVelocity=20
     velocityScaleZ=0.15
     worldAlpha=0.35
     externalLinearDragV=<0 0 0>
    }
   }
   emoteOrMelee {
    componentLinearAccScale=<1 1 0.5>
    componentLinearVelScale=<1 1 0.5>
    simSpaceSettings {
     maxAngularAcceleration=40
     maxAngularVelocity=40
     maxLinearAcceleration=20
     maxLinearVelocity=20
     velocityScaleZ=0.15
     worldAlpha=0.15
     externalLinearDragV=<0 0 0>
    }
   }
   front_End {
    componentLinearAccScale=<1 1 0.5>
    componentLinearVelScale=<1 1 0.5>
    simSpaceSettings {
     maxAngularAcceleration=40
     maxAngularVelocity=40
     maxLinearAcceleration=20
     maxLinearVelocity=20
     velocityScaleZ=0.15
     worldAlpha=0.1
     externalLinearDragV=<0 0 0>
    }
   }
   onGround_Crouch_Moving {
    componentLinearAccScale=<1 1 0.5>
    componentLinearVelScale=<1 1 0.5>
    simSpaceSettings {
     maxAngularAcceleration=40
     maxAngularVelocity=40
     maxLinearAcceleration=20
     maxLinearVelocity=20
     velocityScaleZ=0.15
     worldAlpha=0.35
     externalLinearDragV=<0 0 0>
    }
   }
   onGround_Falling {
    componentLinearAccScale=<1 1 0.5>
    componentLinearVelScale=<1 1 0.5>
    simSpaceSettings {
     maxAngularAcceleration=40
     maxAngularVelocity=40
     maxLinearAcceleration=20
     maxLinearVelocity=20
     velocityScaleZ=0.15
     worldAlpha=0.35
     externalLinearDragV=<0 0 0>
    }
   }
   onGround_Moving {
    componentLinearAccScale=<1 1 0.5>
    componentLinearVelScale=<1 1 0.5>
    simSpaceSettings {
     maxAngularAcceleration=40
     maxAngularVelocity=40
     maxLinearAcceleration=20
     maxLinearVelocity=20
     velocityScaleZ=0.15
     worldAlpha=0.35
     externalLinearDragV=<0 0 0>
    }
   }
   onGround_Standing {
    componentLinearAccScale=<1 1 0.5>
    componentLinearVelScale=<1 1 0.5>
    simSpaceSettings {
     maxAngularAcceleration=40
     maxAngularVelocity=40
     maxLinearAcceleration=20
     maxLinearVelocity=20
     velocityScaleZ=0.15
     worldAlpha=0.35
     externalLinearDragV=<0 0 0>
    }
   }
   rail_Grinding {
    componentLinearAccScale=<1 1 0.5>
    componentLinearVelScale=<1 1 0.5>
    simSpaceSettings {
     maxAngularAcceleration=40
     maxAngularVelocity=40
     maxLinearAcceleration=20
     maxLinearVelocity=20
     velocityScaleZ=0.15
     worldAlpha=0.35
     externalLinearDragV=<0 0 0>
    }
   }
   riding_Moving {
    componentLinearAccScale=<1 1 0.5>
    componentLinearVelScale=<1 1 0.5>
    simSpaceSettings {
     maxAngularAcceleration=40
     maxAngularVelocity=40
     maxLinearAcceleration=20
     maxLinearVelocity=20
     velocityScaleZ=0.15
     worldAlpha=0.35
     externalLinearDragV=<0 0 0>
    }
   }
   riding_Moving_Targeting {
    componentLinearAccScale=<1 1 0.5>
    componentLinearVelScale=<1 1 0.5>
    simSpaceSettings {
     maxAngularAcceleration=40
     maxAngularVelocity=40
     maxLinearAcceleration=20
     maxLinearVelocity=20
     velocityScaleZ=0.15
     worldAlpha=0.35
     externalLinearDragV=<0 0 0>
    }
   }
   riding_Standing {
    componentLinearAccScale=<1 1 0.5>
    componentLinearVelScale=<1 1 0.5>
    simSpaceSettings {
     maxAngularAcceleration=40
     maxAngularVelocity=40
     maxLinearAcceleration=20
     maxLinearVelocity=20
     velocityScaleZ=0.15
     worldAlpha=0.35
     externalLinearDragV=<0 0 0>
    }
   }
   skydive_DivingDown {
    componentLinearAccScale=<1 1 0.5>
    componentLinearVelScale=<1 1 0.5>
    simSpaceSettings {
     maxAngularAcceleration=40
     maxAngularVelocity=40
     maxLinearAcceleration=20
     maxLinearVelocity=20
     velocityScaleZ=0.15
     worldAlpha=0.35
     externalLinearDragV=<0 0 0>
    }
   }
   skydive_FloatingDown {
    componentLinearAccScale=<1 1 0.5>
    componentLinearVelScale=<1 1 0.5>
    simSpaceSettings {
     maxAngularAcceleration=40
     maxAngularVelocity=40
     maxLinearAcceleration=20
     maxLinearVelocity=20
     velocityScaleZ=0.15
     worldAlpha=0.35
     externalLinearDragV=<0 0 0>
    }
   }
   skydive_GoingUp {
    componentLinearAccScale=<1 1 0.5>
    componentLinearVelScale=<1 1 0.5>
    simSpaceSettings {
     maxAngularAcceleration=40
     maxAngularVelocity=40
     maxLinearAcceleration=20
     maxLinearVelocity=20
     velocityScaleZ=0.15
     worldAlpha=0.35
     externalLinearDragV=<0 0 0>
    }
   }
   skydive_Parachuting {
    componentLinearAccScale=<1 1 0.5>
    componentLinearVelScale=<1 1 0.5>
    simSpaceSettings {
     maxAngularAcceleration=40
     maxAngularVelocity=40
     maxLinearAcceleration=20
     maxLinearVelocity=20
     velocityScaleZ=0.15
     worldAlpha=0.35
     externalLinearDragV=<0 0 0>
    }
   }
   skydive_Surfing {
    componentLinearAccScale=<1 1 0.5>
    componentLinearVelScale=<1 1 0.5>
    simSpaceSettings {
     maxAngularAcceleration=40
     maxAngularVelocity=40
     maxLinearAcceleration=20
     maxLinearVelocity=20
     velocityScaleZ=0.15
     worldAlpha=0.35
     externalLinearDragV=<0 0 0>
    }
   }
   swimming {
    componentLinearAccScale=<1 1 0.5>
    componentLinearVelScale=<1 1 0.5>
    simSpaceSettings {
     maxAngularAcceleration=40
     maxAngularVelocity=40
     maxLinearAcceleration=20
     maxLinearVelocity=20
     velocityScaleZ=0.15
     worldAlpha=0.35
     externalLinearDragV=<0 0 0>
    }
   }
   tactical_Sprint {
    componentLinearAccScale=<0.3 0.3 0.3>
    componentLinearVelScale=<0.3 0.3 0.3>
    simSpaceSettings {
     maxAngularAcceleration=40
     maxAngularVelocity=40
     maxLinearAcceleration=20
     maxLinearVelocity=20
     velocityScaleZ=0.15
     worldAlpha=0.35
     externalLinearDragV=<0 0 0>
    }
   }
   ziplining {
    componentLinearAccScale=<1 1 0.5>
    componentLinearVelScale=<1 1 0.5>
    simSpaceSettings {
     maxAngularAcceleration=40
     maxAngularVelocity=40
     maxLinearAcceleration=20
     maxLinearVelocity=20
     velocityScaleZ=0.15
     worldAlpha=0.35
     externalLinearDragV=<0 0 0>
    }
   }
  }
 }
 RBAN Settings Out {
  type=RBANSettingStruct
  default {
   componentLinearAccScale=<0 0 0>
   componentLinearVelScale=<0 0 0>
   simSpaceSettings {
    dampingAlpha=1
    maxAngularAcceleration=1000
    maxAngularVelocity=3
    maxLinearAcceleration=300
    maxLinearVelocity=700
    velocityScaleZ=0
    worldAlpha=1
    externalAngularVelocity=<0 0 0>
    externalLinearDragV=<0 0.5 0>
    externalLinearVelocity=<0 0 0>
   }
  }
 }
 Simulation Space {type=AnimPhysSimSpaceType, default=RootRelative}
 WearingBackbling {type=boolean}
]
